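# MF_UIKit_Outline is a UIKit material function that draws an animated rounded-rectangle outline and glow for buttons, outputting color, alpha and mask.
# A rounded-mask function supplies shape, outline and SDF masks, with padding corrected for aspect ratio.
# Scrolling polar noise, screen-space flow-map noise and a sine sheen add rift-like energy, scaled by Anim parameters for hover and highlight.
# Color blends rift tints with oscillating colors, or uses flat outline color via UseFlatColor.
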MaterialFunction /FortUILibrary/UIKit/Buttons/Materials/MF_UIKit_Outline
graph {
 properties {bExposeToLibrary=true, libraryCategoriesText [UIKit]}
 nodes [
  n1 = NamedRerouteUsage()
  n2 = NamedRerouteUsage()
  FunctionOutput(None=Multiply(A=VertexColor().A, B=Multiply(A=@n1, B=@n2))) {outputName=FinalColor, sortPriority=1}
  FunctionOutput(None=Multiply(A=@n2, B=@n2)) {outputName=FinalAlpha, sortPriority=2}
  FunctionOutput(None=NamedRerouteUsage()) {outputName=MaskAlpha, sortPriority=3}
  n3 = VectorParameter() {parameterName=Corner Radius, defaultValue=<rgba 0.5 0.5 0.5 0.5>}
  n4 = ScalarParameter() {defaultValue=4, group=Outline, parameterName=Outline Thickness}
  n5 = ScalarParameter() {defaultValue=1, group=Anim, parameterName=Anim_OutlineThickness, sliderMax=1}
  n6 = MaterialFunctionCall!(
   UV=NamedRerouteUsage()
   Size=NamedRerouteUsage()
   Corner Radius=AppendVector(A=@n3.RGB, B=@n3.RGB)
   Radius in Pixels=ScalarParameter() {parameterName=Radius in Pixels, sliderMax=1}
   Skew Angle=NamedRerouteDeclaration(None=ScalarParameter() {parameterName=Skew Angle, sliderMax=360}) {name=Skew Angle, nodeColor=<rgba 0 0 0 1>}
   Outline Thickness=NamedRerouteDeclaration(None=LinearInterpolate(B=@n4, Alpha=@n5)) {name=Outline Thickness, nodeColor=<rgba 0 0 0 1>}
  ) {materialFunction=&MF_UI_CoreUI_RoundedMask}
  NamedRerouteDeclaration(None=@n6.Outline Mask) {name=OutlineMask, nodeColor=<rgba 0 0 0 1>}
  n7 = NamedRerouteUsage()
  n8 = MaterialFunctionCall!(Dimensions) {materialFunction=/FortUILibrary/MaterialFunctions/Basics/MF_UI_GetAspectRatio}
  n9 = ComponentMask(None=@n8.Max Ratio) {r=true}
  n10 = ComponentMask(None=@n8.Max Ratio) {g=true}
  n11 = Divide(B=Max(A=@n9, B=@n10))
  n12 = Add(A=NamedRerouteUsage(), B=Multiply(A=OneMinus(None=@n11), B=NamedRerouteUsage()) {constB=0.25})
  n13 = MaterialFunctionCall!(A=@n9, B=@n10) {materialFunction=/FortUILibrary/MaterialFunctions/Basics/MF_UI_BooleanOperations}
  NamedRerouteDeclaration(None=LinearInterpolate(A=AppendVector(A=@n7, B=@n12), B=AppendVector(A=@n12, B=@n7), Alpha=@n13.IsGreater)) {name=Size with Padding, nodeColor=<rgba 0 0 0 1>}
  n14 = MaterialFunctionCall() {materialFunction=&GetUserInterfaceUV}
  NamedRerouteDeclaration(None=@n14.Pixel Size) {name=Pixel Size, nodeColor=<rgba 0 0 0 1>}
  n15 = Multiply(A=LinearInterpolate(Alpha=NamedRerouteUsage()) {constA=0.333}, B=NamedRerouteUsage())
  n16 = Add(
   A=NamedRerouteUsage()
   B=LinearInterpolate(Alpha=ScalarParameter() {group=Anim, parameterName=Anim_Highlight, sliderMax=1}) {constA=1, constB=-1}
  )
  n17 = MaterialFunctionCall("Value (0-1)"=ComponentMask(None=Sine(None=@n16) {period=2}) {r=true}) {materialFunction=&MF_UI_EaseCurves}
  NamedRerouteDeclaration(None=Multiply(A=ScalarParameter() {defaultValue=0.5, group=Alpha, parameterName=Sheen_Alpha, sliderMax=1}, B=Multiply(A=@n15, B=@n17.EaseInCubic))) {name=SheenNoise, nodeColor=<rgba 0 0 0 1>}
  n18 = NamedRerouteUsage()
  n19 = Constant() {desc=UV_Type_Mix, r=0.9}
  n20 = Sine(None=Multiply(A=Time(), B=Constant() {desc=Oscillating Speed, r=0.13}))
  n21 = LinearInterpolate(
   A=NamedRerouteUsage()
   B=AppendVector(A=OneMinus(None=@n18), B=@n18)
   Alpha=LinearInterpolate(A=Add(A=@n19) {constB=0.05}, B=Add(A=@n19) {constB=-0.05}, Alpha=ConstantBiasScale(None=@n20))
  )
  n22 = NamedRerouteUsage()
  n23 = MaterialFunctionCall!(Vector or UVs=NamedRerouteDeclaration(None=MaterialFunctionCall!(Value=@n21, Scale=@n22, Height=@n22) {materialFunction=/FortUILibrary/MaterialFunctions/Basics/MF_UI_PingPong}.Result) {name=PrepPolarUVs, nodeColor=<rgba 0 0 0 1>}) {materialFunction=/FortUILibrary/MaterialFunctions/Basics/MF_UI_CheaperVectorToRadial}
  n24 = ComponentMask(None=@n23.Radial Coordinates) {r=true}
  n25 = Multiply(A=@n24, B=Constant() {r=3})
  n26 = ComponentMask(None=@n23.Radial Coordinates) {g=true}
  n27 = Multiply(A=@n26, B=Constant() {r=2.75})
  n28 = Constant() {r=0.1}
  n29 = Constant() {r=-0.05}
  n30 = Multiply(A=@n26, B=Constant() {r=1.5})
  n31 = Constant() {r=-0.05}
  n32 = Constant() {r=-0.05}
  n33 = NamedRerouteUsage()
  n34 = DDX(Value=@n33)
  n35 = DDY(Value=@n33)
  n36 = TextureSample!(
   UVs=Panner(Coordinate=AppendVector(A=@n24, B=@n30), Time=Time(), Speed=AppendVector(A=@n31, B=@n32))
   Tex=NamedRerouteDeclaration(None=TextureObject() {desc=Rift Noise Texture, samplerType=SAMPLERTYPE_Grayscale, texture=/FortUILibrary/UIKit/Buttons/Textures/LowResBlurredNoise}) {name=Rift Noise Texture, nodeColor=<rgba 0 0 0 1>}
   "DDX(UVs)"=@n34
   "DDY(UVs)"=@n35
   Apply View MipBias
  ) {mipValueMode=TMVM_Derivative}
  n37 = Add(
   A=Panner(Coordinate=AppendVector(A=@n25, B=@n27), Time=Time(), Speed=AppendVector(A=@n28, B=@n29))
   B=Multiply(A=@n36.R, B=Constant() {r=0.5})
  )
  n38 = NamedRerouteUsage()
  n39 = MaterialFunctionCall("Value (0-1)"=DotProduct(A=TextureSample(UVs=@n37, Tex=@n38, "DDX(UVs)"=@n34, "DDY(UVs)"=@n35) {mipValueMode=TMVM_Derivative}.RGB, B=@n36.R)) {materialFunction=&MF_UI_EaseCurves}
  NamedRerouteDeclaration(None=Saturate(None=Multiply(A=@n39.EaseOutCubic, B=@n39.EaseOutCubic))) {name=RiftPulseNoise, nodeColor=<rgba 0 0 0 1>}
  n40 = Rotator(Coordinate=NamedRerouteUsage(), Time=Multiply(A=Time()) {constB=0.333}) {speed=6.2832}
  NamedRerouteDeclaration(None=LinearInterpolate(
    A=VectorParameter() {group=Color, parameterName=Color_Outline, defaultValue=<rgba 1 1 1 1>}.RGB
    B=MaterialFunctionCall!(PrimaryUV, SecondaryUV, UVMix, PixelSize) {materialFunction=/FortUILibrary/MaterialFunctions/CoreUI/MF_UI_CoreUI_OscillatingColor}.Result
    Alpha=ComponentMask(None=@n40) {r=true}
   ) {constA=1}) {name=RiftOscillatingColors, nodeColor=<rgba 0 0 0 1>}
  NamedRerouteDeclaration(None=OneMinus(None=NamedRerouteUsage())) {name=Outline Size, nodeColor=<rgba 0 0 0 1>}
  NamedRerouteDeclaration(None=@n14.Normalized UV) {name=NormalizedUV, nodeColor=<rgba 0 0 0 1>}
  NamedRerouteDeclaration(None=ScalarParameter() {defaultValue=16, group=Padding, parameterName=Padding}) {name=Padding, nodeColor=<rgba 0 0 0 1>}
  n41 = LinearInterpolate(
   A=NamedRerouteUsage()
   B=LinearInterpolate(
    A=VectorParameter() {group=Color, parameterName=Rift Color, defaultValue=<rgba 0.667 1 1 1>}.RGB
    B=NamedRerouteUsage()
    Alpha=Constant() {desc="RiftEnergyOscillatingColorAmount`", r=0.5}
   )
   Alpha=NamedRerouteUsage()
  )
  n42 = LinearInterpolate(
   A=LinearInterpolate(
    A=NamedRerouteUsage()
    B=VectorParameter() {group=Color, parameterName=RiftEnergyColorTint, defaultValue=<rgba 0.5 0.5 1 1>}.RGB
    Alpha=Constant() {desc=RiftEnergyColorTintMix, r=0.8}
   )
   B=VectorParameter() {group=Color, parameterName=Color_Outline, defaultValue=<rgba 1 1 1 1>}.RGB
   Alpha=OneMinus(None=NamedRerouteUsage())
  )
  n43 = NamedRerouteUsage()
  NamedRerouteDeclaration(None=StaticSwitchParameter(
    True=VectorParameter() {group=Color, parameterName=Color_Outline, defaultValue=<rgba 1 1 1 1>}.RGB
    False=LinearInterpolate(A=@n41, B=@n42, Alpha=@n43)
   ) {parameterName=UseFlatColor}) {name=FinalColor, nodeColor=<rgba 0 0 0 1>}
  n44 = ScalarParameter() {defaultValue=1, group=Alpha, parameterName=Glow_Alpha, sliderMax=1}
  n45 = Add(A=@n44) {constB=0.05}
  n46 = Add(A=@n44) {constB=-0.05}
  n47 = Time()
  n48 = Constant() {desc=Glow_Alpha_Change_Speed, r=0.25}
  n49 = ConstantBiasScale(None=Sine(None=Multiply(A=@n47, B=@n48)))
  n50 = Multiply(
   A=LinearInterpolate(
    B=LinearInterpolate(A=@n45, B=@n46, Alpha=@n49)
    Alpha=ScalarParameter() {defaultValue=1, group=Anim, parameterName=Anim_Hover_Fill, sliderMax=1}
   ) {constA=1}
   B=LinearInterpolate(Alpha=NamedRerouteUsage()) {constA=0.5}
  )
  n51 = Multiply(
   A=Multiply(A=NamedRerouteUsage(), B=ScalarParameter() {defaultValue=1, group=Anim, parameterName=Anim_Hover_Energy, sliderMax=1})
   B=LinearInterpolate(Alpha=NamedRerouteUsage()) {constA=0.75}
  )
  n52 = Multiply(
   A=ScalarParameter() {defaultValue=1, group=Alpha, parameterName=Glow_Alpha, sliderMax=1}
   B=Multiply(A=LinearInterpolate(A=@n50, Alpha=@n51), B=NamedRerouteUsage())
  )
  n53 = Add(
   A=Add(A=NamedRerouteDeclaration(None=@n52) {name=CombinedRiftandGlowEnergy, nodeColor=<rgba 0 0 0 1>}, B=NamedRerouteUsage())
   B=NamedRerouteUsage()
  )
  NamedRerouteDeclaration(None=Multiply(A=Saturate(None=@n53), B=ScalarParameter() {defaultValue=1, group=Alpha, parameterName=FInalAlpha, sliderMax=1})) {name=FinalAlpha, nodeColor=<rgba 0 0 0 1>}
  NamedRerouteDeclaration(None=@n6.Shape Mask) {name=ShapeMask, nodeColor=<rgba 0 0 0 1>}
  n54 = SmoothStep(Min=Multiply(A=NamedRerouteUsage()) {constB=-1}, Max=OneMinus(None=NamedRerouteUsage()), Value=NamedRerouteUsage()) {constMax=0.05}
  n55 = OneMinus(None=@n54)
  n56 = OneMinus(None=LinearInterpolate(
    A=NamedRerouteDeclaration(None=@n55) {name=OuterGlowMask, nodeColor=<rgba 0 0 0 1>}
    Alpha=NamedRerouteDeclaration(None=@n54) {name=InnerGlowMask, nodeColor=<rgba 0 0 0 1>}
   ) {constA=1})
  NamedRerouteDeclaration(None=Saturate(None=Multiply(A=@n56) {constB=2})) {name=FinalGlowMask, nodeColor=<rgba 0 0 0 1>}
  n57 = LinearInterpolate(A=Constant() {desc=Outline_Noise_NegativeSpace_Alpha, r=0.5}, Alpha=NamedRerouteUsage())
  n58 = ScalarParameter() {defaultValue=1, group=Anim, parameterName=Anim_Hover_Outline, sliderMax=1}
  NamedRerouteDeclaration(None=Multiply(A=Multiply(A=@n57, B=@n58), B=NamedRerouteUsage())) {name=FinalOutlineAlpha, nodeColor=<rgba 0 0 0 1>}
  NamedRerouteDeclaration(None=@n6.Outline SDF) {name=SDFOutline, nodeColor=<rgba 0 0 0 1>}
  NamedRerouteDeclaration(None=MaterialFunctionCall!(Value PX=NamedRerouteUsage(), UV Pixel Size=NamedRerouteUsage(), X or Y=@n13.IsGreater) {materialFunction=/FortUILibrary/MaterialFunctions/Basics/MF_UI_PixelsToNormalized}.Normalized) {name=Normalized Padding, nodeColor=<rgba 0 0 0 1>}
  n59 = LinearInterpolate(A=ScreenPosition().ViewportUV, B=NamedRerouteUsage()) {constAlpha=0.125}
  n60 = Add(A=OneMinus(None=NamedRerouteUsage()))
  n61 = Multiply(A=Constant() {r=2}, B=@n60)
  n62 = Time()
  n63 = Multiply(A=@n62, B=Constant() {r=-0.1})
  n64 = Multiply(A=@n62, B=Constant() {r=0.05})
  n65 = NamedRerouteDeclaration(None=Add(
    A=MaterialFunctionCall(UVs=@n59, Texture Scale=@n61) {materialFunction=&ScaleUVsByCenter}.UVs
    B=AppendVector(A=@n63, B=@n64)
   )) {name=PanningUVs, nodeColor=<rgba 0 0 0 1>}
  n66 = NamedRerouteUsage()
  n67 = Constant() {r=3}
  n68 = Time()
  n69 = Constant() {r=0.025}
  n70 = Constant() {r=0.05}
  n71 = Add(
   A=MaterialFunctionCall(UVs=@n59, Texture Scale=Multiply(A=@n60, B=@n67)) {materialFunction=&ScaleUVsByCenter}.UVs
   B=AppendVector(A=Multiply(A=@n68, B=@n69), B=Multiply(A=@n68, B=@n70))
  )
  n72 = NamedRerouteDeclaration(None=AppendVector(A=TextureSample!(UVs=@n65, Tex=@n66, Apply View MipBias).R, B=TextureSample(UVs=@n71, Tex=@n66).R)) {name=FlowmapUVs, nodeColor=<rgba 0 0 0 1>}
  n73 = MaterialFunctionCall!(
   UVs=NamedRerouteUsage()
   Texture Object=@n66
   Grayscale Flow Map=TextureSample(UVs=@n72, Tex=@n66).RGB
   Flow Direction=OneMinus(None=@n72)
   Flow Intensity=Constant() {r=0.2}
   Flow Speed=Constant() {r=0.2}
  ) {materialFunction=/FortUILibrary/MaterialFunctions/Effects/MF_UI_GrayscaleFlowMap_New}
  NamedRerouteDeclaration(None=Saturate(None=MaterialFunctionCall("Value (0-1)"=@n73.Distorted Texture) {materialFunction=&MF_UI_EaseCurves}.EaseInOutSine)) {name=ScreenSpaceNoiseMask, nodeColor=<rgba 0 0 0 1>}
 ]
}
