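# M_UI_DragVisualOverlay is an additive UI material in the quest tile folder, named as the overlay for a drag visual.
# It blends two color parameters along the vertical UV axis to form a gradient, with a scalar parameter controlling gradient opacity.
# The shared MF_UIKit_Outline function supplies outline color, alpha and shape mask through named reroutes.
# Max nodes merge gradient and outline, so the brighter or stronger value wins rather than stacking.
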Material /QuestUI/QuestScreen/QuestTile/Materials/M_UI_DragVisualOverlay
graph {
 properties {bAutomaticallySetUsageInEditor=false, blendMode=BLEND_Additive, materialDomain=MD_UI, usedShadingModels=Default Lit}
 nodes [
  Max(A=Multiply(A=NamedRerouteUsage(), B=ScalarParameter() {defaultValue=0.2, parameterName=OverlayGradient_Opacity}), B=NamedRerouteUsage())
  n1 = TextureCoordinate()
  Max(
   A=LinearInterpolate(
    A=VectorParameter() {parameterName=OverlayGradient_Color1, defaultValue=<rgba 0.004 0.0666 0.0343 1>}.RGB
    B=VectorParameter() {parameterName=OverlayGradient_Color2, defaultValue=<rgba 0.0497 0.0048 0.0685 1>}.RGB
    Alpha=ComponentMask(None=@n1) {g=true}
   )
   B=NamedRerouteUsage()
  )
  n2 = MaterialFunctionCall() {materialFunction=/FortUILibrary/UIKit/Buttons/Materials/MF_UIKit_Outline}
  NamedRerouteDeclaration(None=@n2.FinalColor) {name=OutlineColor, nodeColor=<rgba 1 0.6797 0 1>}
  NamedRerouteDeclaration(None=@n2.FinalAlpha) {name=OutlineAlpha, nodeColor=<rgba 1 0.6797 0 1>}
  NamedRerouteDeclaration(None=@n2.MaskAlpha) {name=OutlineShapeMask, nodeColor=<rgba 1 0.6797 0 1>}
 ]
}
