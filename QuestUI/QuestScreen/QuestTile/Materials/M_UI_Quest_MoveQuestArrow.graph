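# Translucent UI material that draws a procedural arrow sliding back and forth horizontally, named for moving a quest in the quest screen tiles.
# A time-driven sine offsets the U coordinate shared by both arrow parts.
# The head is a rotated three-sided MF_UI_RegularPolygon masked by a SmoothStep along U, added to an MF_UI_Box shaft.
# Named reroutes feed the shapes' glow edge limits, and a Color vector parameter is declared.
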Material /QuestUI/QuestScreen/QuestTile/Materials/M_UI_Quest_MoveQuestArrow
graph {
 properties {bAutomaticallySetUsageInEditor=false, blendMode=BLEND_Translucent, materialDomain=MD_UI, usedShadingModels=Default Lit}
 nodes [
  n1 = TextureCoordinate()
  n2 = Sine(None=Multiply(A=Time()))
  n3 = AppendVector(A=Add(A=ComponentMask(None=@n1) {r=true}, B=Multiply(A=@n2) {constB=0.11}), B=ComponentMask(None=@n1) {g=true})
  n4 = Add(A=Multiply(A=@n3, B=Constant2Vector() {g=1, r=1.5}) {desc=Scale}, B=Constant2Vector() {r=-0.35}) {desc=Move}
  n5 = ComponentMask(None=@n4) {r=true}
  n6 = MaterialFunctionCall(UVs=@n4, "Rotation Angle (0-1)"=Constant() {desc=30 degree, r=0.0833}) {materialFunction=&CustomRotator}
  n7 = Constant() {r=0.7}
  n8 = Constant() {r=3}
  n9 = NamedRerouteUsage()
  n10 = NamedRerouteUsage()
  n11 = Constant2Vector() {r=0.06}
  Add(
   A=Multiply(
    A=SmoothStep(Value=@n5) {constMax=0.2, constMin=0.05}
    B=MaterialFunctionCall!(UVs=@n6.Rotated Values, Size=@n7, Sides=@n8, InnerAngle, OutlineThickness, OutlineAlignment, GlowMin=@n9, GlowMax=@n10) {materialFunction=/FortUILibrary/MaterialFunctions/Shapes/MF_UI_RegularPolygon}.Fill
   )
   B=MaterialFunctionCall(UVs=Add(A=@n3, B=@n11) {desc=Move}, Size=Constant2Vector() {g=0.15, r=0.4}, GlowMin=NamedRerouteUsage(), GlowMax=NamedRerouteUsage()) {materialFunction=&MF_UI_Box}.Fill
  )
  n12 = Constant() {r=0.09}
  NamedRerouteDeclaration(None=@n12) {name=SmoothStepMin, nodeColor=<rgba 0 1 0.3906 1>}
  NamedRerouteDeclaration(None=Constant() {r=0.18}) {name=SmoothStepMax, nodeColor=<rgba 1 0.0703 0 1>}
  Reroute(None=@n12)
  VectorParameter() {parameterName=Color, defaultValue=<rgba 1 1 1 1>}
 ]
}
