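; QuestDragDrop is a quest screen asset, filed with the quest tile UI, that by its name relates to dragging and dropping quests.
; It declares one function, ShowDropIcon, which takes a Drop Action of the AthenaDragDropAction enum, defaulting to the first entry.
; The function has no body here, so the file defines only the signature and no icon-switching logic.
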(fn ShowDropIcon
  (params (Drop%20Action :type AthenaDragDropAction%20Enum :default "NewEnumerator0")))
