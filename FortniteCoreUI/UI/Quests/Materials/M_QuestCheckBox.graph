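# UI material for a quest checkbox, where a State scalar represents available, completed or locked.
# State drives three-way lerps that pick between lock, check or blank icon textures and shift the dark-to-white tint and radial-gradient backing.
# Icons are signed distance fields with scalable, rotatable UVs, anti-aliased into fill and optional outline channels that feed color and opacity.
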Material /FortniteCoreUI/UI/Quests/Materials/M_QuestCheckBox
graph {
 properties {
  bAutomaticallySetUsageInEditor=false
  bUsedWithStaticMesh=false
  blendMode=BLEND_Translucent
  materialDomain=MD_UI
  usedShadingModels=Default Lit
 }
 nodes [
  n1 = Constant4Vector() {constant=<rgba 0.0065 0.006 0.01 0.6>}
  n2 = ScalarParameter() {defaultValue=0.5, desc="0 - available, 0.5 - comleted, 1 - locked", parameterName=State}
  n3 = VertexColor()
  n4 = NamedRerouteUsage()
  n5 = Multiply(
   A=NamedRerouteUsage()
   B=Multiply(A=ScalarParameter() {defaultValue=1, group=Color, parameterName=Brightness Percentage, sliderMax=1, sortPriority=2}, B=@n4)
  )
  n6 = ScalarParameter() {defaultValue=1, parameterName=IconOpacity}
  LinearInterpolate(
   A=MaterialFunctionCall(A=@n1.RGBA, B=Constant4Vector() {constant=<rgba 1 1 1 1>}.RGBA, C=@n1.RGBA, Alpha=@n2) {materialFunction=&Lerp_3Color}.Result
   Alpha=Multiply(A=Multiply(A=@n3.A, B=@n5), B=@n6)
  ) {constA=1}
  n7 = Constant() {r=20}
  n8 = Constant() {r=0.5}
  n9 = Constant() {r=1}
  n10 = ScalarParameter() {group=Color, parameterName=Additive Percentage, sliderMax=1, sortPriority=3}
  n11 = Multiply(A=@n6, B=Multiply(A=@n4, B=OneMinus(None=@n10)))
  LinearInterpolate(
   A=LinearInterpolate(
    B=MaterialFunctionCall(A=@n8, B=@n9, C=@n8, Alpha=@n2) {materialFunction=&Lerp_3Color}.Result
    Alpha=MaterialFunctionCall(Density=@n7) {materialFunction=&RadialGradientExponential}.RadialGradientExponential
   ) {constB=0.5}
   B=MaterialFunctionCall(A=@n11, B=Constant(), C=@n11, Alpha=@n2) {materialFunction=&Lerp_3Color}.Result
   Alpha=@n11
  ) {constA=1}
  n12 = NamedRerouteUsage()
  n13 = TextureSampleParameter2D(UVs=@n12) {
   parameterName=IconLock
   samplerSource=SSM_Clamp_WorldGroupSettings
   samplerType=SAMPLERTYPE_LinearColor
   texture=/FortUILibrary/Textures/Icons/T_UI_Icon_UIKit_Locked_SDF
  }
  n14 = TextureSampleParameter2D(UVs=@n12) {
   parameterName=IconCheck
   samplerSource=SSM_Clamp_WorldGroupSettings
   samplerType=SAMPLERTYPE_LinearColor
   texture=/FortUILibrary/Textures/Icons/T_UI_Icon_UIKit_Check_SDF
  }
  n15 = Constant()
  n16 = ScalarParameter() {defaultValue=0.5, desc="0 - available, 0.5 - comleted, 1 - locked", parameterName=State}
  NamedRerouteDeclaration(None=AppendVector(
    A=MaterialFunctionCall(A=@n13.RGB, B=@n14.RGB, C=@n15, Alpha=@n16) {materialFunction=&Lerp_3Color}.Result
    B=Constant()
   )) {name=Texture Channels, nodeColor=<rgba 0 0 0 1>}
  n17 = ScalarParameter() {defaultValue=0.25, group=3 - Outline Settings, parameterName=Outline Opacity, sliderMax=1, sortPriority=34}
  n18 = NamedRerouteUsage()
  NamedRerouteDeclaration(None=LinearInterpolate(A=Multiply(A=@n17, B=@n18), B=ScalarParameter() {defaultValue=1, parameterName=Fill Opacity, sliderMax=1}, Alpha=NamedRerouteUsage()) {constAlpha=1}) {name=Opacity, nodeColor=<rgba 0 0 0 1>}
  n19 = VectorParameter() {group=3 - Outline Settings, parameterName=Outline Color, sortPriority=33, defaultValue=<rgba 1 0.8111 0 1>}
  n20 = Multiply(A=Constant() {r=1}, B=NamedRerouteUsage())
  n21 = Constant() {r=1}
  n22 = NamedRerouteUsage()
  NamedRerouteDeclaration(None=LinearInterpolate(
    A=LinearInterpolate(B=@n19.RGB, Alpha=@n20)
    B=VectorParameter() {parameterName=IconColor, defaultValue=<rgba 1 1 1 1>}.RGB
    Alpha=Multiply(A=@n21, B=@n22)
   )) {name=Separated Colors, nodeColor=<rgba 0 0 0 1>}
  n23 = NamedRerouteUsage()
  n24 = ScalarParameter() {parameterName=Smoothing, sliderMax=1}
  n25 = ComponentMask(None=@n23) {g=true}
  n26 = OneMinus(None=ScalarParameter() {defaultValue=0.35, group=3 - Outline Settings, parameterName=Outline Scale, sliderMax=1})
  n27 = AppendVector(
   A=MaterialFunctionCall(
    DistanceField=ComponentMask(None=@n23) {r=true}
    Threshold=ScalarParameter() {defaultValue=0.68, parameterName=Fill Aliasing}
    Smoothing=@n24
   ) {materialFunction=&MF_UI_AntiAliasing}.Outer
   B=StaticSwitchParameter(
    True=MaterialFunctionCall(DistanceField=@n25, Threshold=@n26, Smoothing=@n24) {materialFunction=&MF_UI_AntiAliasing}.Outer
    False=Constant()
   ) {group=3 - Outline Settings, parameterName=Display Outline}
  )
  n28 = NamedRerouteUsage()
  n29 = MaterialFunctionCall(Float2=NamedRerouteDeclaration(None=StaticSwitchParameter(True=@n27, False=@n28) {defaultValue=true, group=2 - Distance Field, parameterName=Use Distance Field, sortPriority=0}) {name=Distance Field Pass, nodeColor=<rgba 0 0 0 1>}) {materialFunction=&BreakOutFloat2Components}
  NamedRerouteDeclaration(None=@n29.R) {name=Red Channel, nodeColor=<rgba 0 0 0 1>}
  NamedRerouteDeclaration(None=@n29.G) {name=Green Channel, nodeColor=<rgba 0 0 0 1>}
  n30 = MaterialFunctionCall() {materialFunction=&GetUserInterfaceUV}
  n31 = AppendVector(
   A=Subtract(A=ComponentMask(None=@n30.Normalized UV) {r=true}, B=ScalarParameter() {group=1 - Transform, parameterName=OffsetX, sortPriority=48})
   B=Add(A=ComponentMask(None=@n30.Normalized UV) {g=true}, B=ScalarParameter() {group=1 - Transform, parameterName=OffsetY, sortPriority=48})
  )
  n32 = ScalarParameter() {defaultValue=1.2, group=1 - Transform, parameterName=Scale, sortPriority=0}
  NamedRerouteDeclaration(None=MaterialFunctionCall(
    UVs=MaterialFunctionCall(UVs=@n31, Texture Scale=@n32) {materialFunction=&ScaleUVsByCenter}.UVs
    "Rotation Angle (0-1)"=ScalarParameter() {group=1 - Transform, parameterName=Rotation}
   ) {materialFunction=&CustomRotator}.Rotated Values) {name=UVs, nodeColor=<rgba 0 0 0 1>}
 ]
}
