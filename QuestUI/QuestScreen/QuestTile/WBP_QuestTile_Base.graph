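# Base quest tile button widget, built on UIKitModularButton with no widgets of its own, that handles focus visuals and sounds.
# Hover, selection and non-touch focus set focused state, then call the empty MarkSeen hook and play a hover sound; their opposites unfocus.
# SetFocusState uses an XOR guard to request instant or animated block transitions only on state changes.
# Clicking plays the pressed sound.
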WidgetBlueprint /QuestUI/QuestScreen/QuestTile/WBP_QuestTile_Base
graphs [
 EventGraph {nodes [
   Event On Focused() {is=title K2Node_Event}
   > n1 = Branch(Condition=IsUsingTouch(PlayerController=GetOwningPlayer()))
   SetFocusState!(execute=[@n4.then, @n1.else, @n6.then], self, IsFocused:bool=true, IsInstant:bool=false) -> (then)
   > MarkSeen()
   > PlaySoundHovered()
   Event On Unfocused() {is=title K2Node_Event}
   > n2 = Branch(Condition=IsUsingTouch(PlayerController=GetOwningPlayer()))
   SetFocusState(execute=[@n3.then, @n2.else, @n5.then], IsFocused=false, IsInstant=false)
   n3 = Event On Deselected()
   n4 = Event On Hovered()
   n5 = Event On Unhovered()
   n6 = Event On Selected()
   Event On Clicked()
   > PlaySoundPressed()
  ]}
 MarkSeen {nodes [MarkSeen() {is=title K2Node_FunctionEntry}]}
 PlaySoundHovered {nodes [
   PlaySoundHovered() {is=title K2Node_FunctionEntry}
   > n1 = Get!(execute:Exec) -> (then:Exec, else:Exec, HoveredSound:Sound Base Object Reference)
   > SpawnSound2D!(
    execute
    Sound:Sound Base Object Reference=@n1
    VolumeMultiplier:float=1
    PitchMultiplier:float=1
    StartTime:float=0
    ConcurrencySettings:Sound Concurrency Object Reference
    bPersistAcrossLevelTransition:bool=false
    bAutoDestroy:bool=true
   ) -> (then, ReturnValue:Audio Component Object Reference)
  ]}
 PlaySoundPressed {nodes [
   PlaySoundPressed() {is=title K2Node_FunctionEntry}
   > n1 = Get!(execute) -> (then, else, PressedSound:Sound Base Object Reference)
   > SpawnSound2D(Sound=@n1, VolumeMultiplier=1, PitchMultiplier=1, StartTime=0, bPersistAcrossLevelTransition=false, bAutoDestroy=true)
  ]}
 SetFocusState {nodes [
   n1 = SetFocusState!() -> (then, IsFocused:bool=~, IsInstant:bool=~) {is=title K2Node_FunctionEntry}
   > Branch(Condition=XOR Boolean!(A:bool=@n1, B:bool=Get IsButtonFocused() -> (IsButtonFocused:bool)) -> (ReturnValue:bool))
   > n2 = Set IsButtonFocused(IsButtonFocused:bool=Get IsFocused() -> (IsFocused:bool)) -> (Output_Get:bool)
   > n3 = Branch(Condition=@n2)
   > n4 = Branch(Condition=Get IsInstant() -> (IsInstant:bool))
   > RequestInstantTransitionOnBlocks!(execute, self:UIKit Modular Button Object Reference, State:EUIKitBlockInstantTransitionState Enum=InstantFocused) -> (then)
   RequestTransitionOnBlocks!(execute=@n4.else, self:UIKit Modular Button Object Reference, State:EUIKitBlockVisualState Enum=Focused) -> (then)
   n5 = Branch(execute=@n3.else, Condition=Get IsInstant())
   > RequestInstantTransitionOnBlocks(State=InstantUnfocused)
   RequestTransitionOnBlocks(execute=@n5.else, State=Unfocused)
  ]}
]
variables [
 HoveredSound {type=SoundBase, default=/Game/Sounds/Fort_UI_Sounds/FrontEnd/New_FrontEnd/DefaultSounds/PressedButtons/FN_UI_Default_Button_Hovered_Cue}
 IsButtonFocused {type=boolean}
 PressedSound {type=SoundBase, default=/Game/Sounds/Fort_UI_Sounds/FrontEnd/New_FrontEnd/DefaultSounds/PressedButtons/FN_UI_Default_Button_Pressed_Cue}
]
widgets {info {inheritedWidgetCount=0, parentClass=/Script/FortniteUI.UIKitModularButton, rootWidgetClass=null, widgetCount=0}, widgets []}
