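# Quest screen counter widget that shows how many quests are tracked against the maximum allowed, as "count/max" on a backplate.
# SetOrderedQuests stores the length of the quest view model array it receives in NumberTrackedQuests, then calls Display, which also runs on Construct.
# Display writes that number to one text block and the limit from the player view model's TrackerVM to a second, dimmed "/{max}" block.
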WidgetBlueprint /QuestUI/QuestScreen/WBP_TrackedQuestsCounter
graphs [
 Display {nodes [
   Display() {is=title K2Node_FunctionEntry}
   > Set Text(
    Text="To Text (Integer)"(Value=Get NumberTrackedQuests() -> (NumberTrackedQuests:int=0))
    self=Get Text_TrackedCount() -> (Text_TrackedCount:Common Text Object Reference)
   )
   n1 = Get TrackerVM!(self:Quest UI Player View Model Object Reference=Get QuestUIPlayerVM() -> (QuestUIPlayerVM:Quest UI Player View Model Object Reference)) -> (TrackerVM:Quest UI Tracker View Model Object Reference)
   > Set Text(
    Text=Format Text!(Format="LOCTEXT(\"/{max}\")", max:int=GetMaxTrackedQuestCount!(self:Quest UI Tracker View Model Object Reference=@n1) -> (ReturnValue:int=0)) -> (Result)
    self=Get Text_TotalCount() -> (Text_TotalCount:Common Text Object Reference)
   )
  ]}
 EventGraph {nodes [
   Event Construct()
   > Display() {is=title K2Node_CallFunction}
  ]}
 SetOrderedQuests {nodes [
   n1 = SetOrderedQuests!() -> (then, OrderedQuests:Array of Quest UI Base Content View Model Object References) {is=title K2Node_FunctionEntry}
   > Set NumberTrackedQuests(NumberTrackedQuests:int=Length(TargetArray:Array of Quest UI Base Content View Model Object References=@n1)) -> (Output_Get:int=0)
   > Display()
  ]}
]
variables [NumberTrackedQuests {type=integer}]
widgets {
 info {inheritedWidgetCount=0, parentClass=&UserWidget, rootWidgetClass=&CommonBorder, widgetCount=4}
 widgets [
  Border_QuestCounterBackground {
   widgetClassPath=&CommonBorder
   bIsVariable=true
   widget=Border_QuestCounterBackground
   properties {
    horizontalAlignment=HAlign_Left
    verticalAlignment=VAlign_Center
    background {
     bIsDynamicallyLoaded=false
     drawAs=Image
     imageType=FullColor
     mirroring=NoMirror
     resourceName=None
     resourceObject=/QuestUI/QuestScreen/QuestTile/Materials/MI_UI_TrackedQuestsCounter_Backplate
     tiling=NoTile
     imageSize=<32 32>
     margin=<ltrb 0 0 0 0>
     outlineSettings {
      bUseBrushTransparency=false
      roundingType=HalfHeightRadius
      width=0
      color {colorUseRule=UseColor_Specified, specifiedColor=<rgba 0 0 0 0>}
      cornerRadii=<0 0 0 0>
     }
     tintColor {colorUseRule=UseColor_Specified, specifiedColor=<rgba 1 1 1 1>}
     uVRegion {bIsValid=false, max=<0 0>, min=<0 0>}
    }
    padding=<ltrb 12 0 12 2>
   }
  }
  HorizontalBox_0 {
   widgetClassPath=&HorizontalBox
   parent=Border_QuestCounterBackground
   widget=HorizontalBox_0
   slot {class=BorderSlot, properties {horizontalAlignment=HAlign_Left, verticalAlignment=VAlign_Center, padding=<ltrb 12 0 12 2>}}
  }
  Text_TrackedCount {
   widgetClassPath=&CommonTextBlock
   bIsVariable=true
   parent=HorizontalBox_0
   widget=Text_TrackedCount
   properties {
    lineHeightPercentage=0.9
    mobileFontSizeMultiplier=1.25
    text=0
    textTransformPolicy=ToUpper
    wrapTextAt=500
    font {
     bForceMonospaced=true
     bMaterialIsStencil=false
     fontMaterial=null
     fontObject=&HeadingNow
     letterSpacing=0
     monospacedWidth=1
     size=15.75
     skewAmount=0
     typefaceFontName=76 Bold Italic
     outlineSettings {
      bApplyOutlineToDropShadows=false
      bMiteredCorners=false
      bSeparateFillAlpha=false
      outlineMaterial=null
      outlineSize=0
      outlineColor=<rgba 0 0 0 1>
     }
    }
    shadowOffset=<0 0>
   }
   slot {class=HorizontalBoxSlot}
  }
  Text_TotalCount {
   widgetClassPath=&CommonTextBlock
   bIsVariable=true
   parent=HorizontalBox_0
   widget=Text_TotalCount
   properties {
    lineHeightPercentage=0.9
    mobileFontSizeMultiplier=1.25
    text=/25
    textTransformPolicy=ToUpper
    wrapTextAt=500
    colorAndOpacity {colorUseRule=UseColor_Specified, specifiedColor=<rgba 1 1 1 0.4>}
    font {
     bForceMonospaced=true
     bMaterialIsStencil=false
     fontMaterial=null
     fontObject=&HeadingNow
     letterSpacing=0
     monospacedWidth=1
     size=15.75
     skewAmount=0
     typefaceFontName=76 Bold Italic
     outlineSettings {
      bApplyOutlineToDropShadows=false
      bMiteredCorners=false
      bSeparateFillAlpha=false
      outlineMaterial=null
      outlineSize=0
      outlineColor=<rgba 0 0 0 1>
     }
    }
    shadowOffset=<0 0>
   }
   slot {class=HorizontalBoxSlot}
  }
 ]
}
