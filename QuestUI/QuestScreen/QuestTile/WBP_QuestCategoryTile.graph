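# Quest screen category tile, derived from WBP_QuestTile_Base, showing a category's header, progress, rewards, expiration or external-link highlight, and new, tracked and availability states.
# View model setters drive visibility switchers; a progress-changed delegate is bound for checkbox icons or a material-driven bar.
# The lazy-loaded background takes size and debug flags as dynamic material parameters.
# A tracked-quests mode collapses category details and adds a counter widget at runtime.
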WidgetBlueprint /QuestUI/QuestScreen/QuestTile/WBP_QuestCategoryTile
graphs [
 AddTrackedQuestCounter {nodes [
   AddTrackedQuestCounter() {is=title K2Node_FunctionEntry}
   > n1 = Create Widget!(
    execute:Exec
    Class:User Widget Class Reference=/QuestUI/QuestScreen/WBP_TrackedQuestsCounter.WBP_TrackedQuestsCounter_C
    OwningPlayer:Player Controller Object Reference
   ) -> (then:Exec, ReturnValue:WBP Tracked Quests Counter Object Reference)
   > n2 = Set TrackedQuestCounter(TrackedQuestCounter:WBP Tracked Quests Counter Object Reference=@n1) -> (Output_Get:WBP Tracked Quests Counter Object Reference)
   > n3 = AddChild(self=Get GridPanel_Root() -> (GridPanel_Root:Grid Panel Object Reference), Content=@n2)
   > n4 = Cast To GridSlot!(execute, Object=@n3) -> (then, CastFailed, AsGrid Slot:Grid Slot Object Reference) {is=title K2Node_DynamicCast}
   > SetPadding!(execute, self:Grid Slot Object Reference=@n4, InPadding:Margin Structure=Make Margin(Left=0, Top=8, Right=16, Bottom=8)) -> (then)
   > SetColumn!(execute, self:Grid Slot Object Reference=@n4, InColumn:int=1) -> (then)
   > SetLayer!(execute, self:Grid Slot Object Reference=@n4, InLayer:int=10) -> (then)
   > SetHorizontalAlignment(self:Grid Slot Object Reference=@n4, InHorizontalAlignment=HAlign_Right)
  ]}
 BranchCategoryTesting {nodes [
   Inputs!() -> (In:Exec)
   > n1 = Branch(Condition=NOT Boolean(A=Get IsTesting() -> (IsTesting:bool)))
   > Outputs!(NotTesting:Exec, Testing:Exec=@n1.else)
  ]}
 DisplayBackgroundMaterial {nodes [
   n1 = DisplayBackgroundMaterial!() -> (then, Material:Soft Object Reference) {is=title K2Node_FunctionEntry}
   > SetBrushFromLazyDisplayAsset!(
    execute
    self:Common Lazy Image Object Reference=Get Lazy_Background() -> (Lazy_Background:Fort Lazy Image Object Reference)
    LazyObject:"Soft Object Reference (by ref)"=@n1
    bMatchTextureSize:bool=false
   ) -> (then)
   > n2 = GetDynamicMaterial(self=Get Lazy_Background())
   > SetScalarParameterValue(
    self=@n2
    ParameterName=Get Param_IsDebugDisplay() -> (Param_IsDebugDisplay:Name=None)
    Value="To Float (Boolean)"(InBool=Get IsDebugDisplay() -> (IsDebugDisplay:bool))
   )
   > SetScalarParameterValue(
    self=@n2
    ParameterName=Get Param_IsHalfTile() -> (Param_IsHalfTile:Name=None)
    Value="To Float (Boolean)"(InBool=Get IsHalfTile() -> (IsHalfTile:bool))
   )
   n3 = Get QuestUICategoryVM() -> (QuestUICategoryVM:Quest UI Category View Model Object Reference)
   > Is Valid(InputObject=@n3)
   > n4 = GetDynamicMaterial(self=Get Lazy_Background())
   n5 = Get CategorySize!(self:Quest UI Category Interface View Model Object Reference=@n3) -> (CategorySize:EQuestCategorySize Enum=Small)
   > SetScalarParameterValue(
    self=@n4
    ParameterName=Get Param_IsLarge() -> (Param_IsLarge:Name=None)
    Value="To Float (Boolean)"(InBool="Equal (Enum)"(A:EQuestCategorySize Enum=@n5, B:EQuestCategorySize Enum=Large))
   )
  ]}
 DisplayTrackedQuestStyle {nodes [
   DisplayTrackedQuestStyle() {is=title K2Node_FunctionEntry}
   > n1 = Branch(Condition=Get IsTrackedQuestTile() -> (IsTrackedQuestTile:bool))
   > SetHeader!(execute, self, Text:Text="LOCTEXT(\"My Tracked Quests\")") -> (then)
   > SetHeightOverride(self=Get TileSizeBox() -> (TileSizeBox:Size Box Object Reference), InHeightOverride=48)
   > UpdateBackgroundMaterial!(execute, self, NewParam:Soft Object Reference=/QuestUI/QuestScreen/QuestTile/Materials/MI_UI_TrackedQuests_Backplate) -> (then)
   > SetVisibility(
    self=[
     Get DebugCategoryID() -> (DebugCategoryID:Common Text Object Reference)
     Get Switcher_CategoryState() -> (Switcher_CategoryState:Common Visibility Switcher Object Reference)
     Get Rewards() -> (Rewards:Grid Panel Object Reference)
     Get StatusIndicator() -> (StatusIndicator:WBP UIKit Status Indicator Object Reference)
     Get TrackedImage() -> (TrackedImage:Image Object Reference)
    ]
    InVisibility=Collapsed
   )
   > n2 = Cast To GridSlot(Object=Get Slot!(self:Widget Object Reference=Get Content() -> (Content:Grid Panel Object Reference)) -> (Slot:Panel Slot Object Reference))
   > SetPadding!(execute, self=@n2, InPadding_Left:float=16, InPadding_Top:float=0, InPadding_Right:float=16, InPadding_Bottom:float=9) -> (then)
   > SetIsHalfTile!(execute, self, HalfTile:bool=true) -> (then)
   > AddTrackedQuestCounter() {is=title K2Node_CallFunction}
   RemoveWidget!(
    execute=@n1.else
    Widget:Widget Object Reference=Get TrackedQuestCounter() -> (TrackedQuestCounter:WBP Tracked Quests Counter Object Reference)
   ) -> (then)
   > Set TrackedQuestCounter()
  ]}
 EventGraph {nodes [
   n1 = UpdateBackgroundMaterial!() -> (OutputDelegate, then, NewParam:Soft Object Reference) {is=title K2Node_CustomEvent}
   > Branch Category Testing!(In:Exec) -> (NotTesting:Exec, Testing:Exec) {is=title K2Node_MacroInstance}
   > DisplayBackgroundMaterial!(execute, self, Material:Soft Object Reference=@n1.NewParam) -> (then) {is=title K2Node_CallFunction}
   Event PreConstruct()
   > RegisterModularBlock!(
    execute
    self:UIKit Modular Button Object Reference
    Block:UIKit Block Object Reference=Get Block_Outline() -> (Block_Outline:WBP UIKit Block Outline Object Reference)
   ) -> (then)
   > SetCompactUIStyle() {is=title K2Node_CallFunction}
   > DisplayTrackedQuestStyle() {is=title K2Node_CallFunction}
   > UpdateExpirationSwitcher()
   > n2 = Branch(Condition=InEditor!() -> (ReturnValue:bool))
   > SetText!(
    execute
    self:WBP UIKit Highlight Base Object Reference=Get Highlight_Expiration() -> (Highlight_Expiration:WBP UIKit Highlight Object Reference)
    Text:Text="INVTEXT(\"LEAVING TODAY\")"
   ) -> (then)
   > n3 = SetText(
    self=Get Highlight_CustomExpiration() -> (Highlight_CustomExpiration:WBP UIKit Highlight Object Reference)
    Text="INVTEXT(\"[PH] REFRESHES DAILY AT 14:00:00\")"
   )
   n4 = Branch Category Testing(In=[@n3.then, @n2.else])
   DisplayBackgroundMaterial(execute=@n4.Testing, Material=/FortniteCoreUI/Items/QuestItems/Categories/Art/MI_UI_CategoryTest)
   Event Construct()
   > SetDebugCategoryID!(execute, self, CategoryID:"String (by ref)"=MakeLiteralString()) -> (then)
  ]}
 HasCinematicRewards {nodes [
   n1 = HasCinematicRewards!() -> (then, HasCinematics:bool=~) {is=title K2Node_FunctionEntry}
   > SetActiveWidgetIndex(
    self:Common Visibility Switcher Object Reference=Get Switcher_Reward() -> (Switcher_Reward:Common Visibility Switcher Object Reference)
    Index=Select(Option 0:int=0, Option 1:int=2, Index=@n1) -> (ReturnValue:int)
   )
  ]}
 HasCosmeticRewards {nodes [
   n1 = HasCosmeticRewards!() -> (then, HasItems:bool=~) {is=title K2Node_FunctionEntry}
   > SetActiveWidgetIndex(self=Get Switcher_Reward(), Index=Select(Option 0=0, Option 1=1, Index=@n1))
  ]}
 HasExpirationTime {nodes [
   HasExpirationTime!() -> (then, HasExpirationTime:bool=~) {is=title K2Node_FunctionEntry}
   > UpdateExpirationSwitcher()
  ]}
 IsAvailableInCurrentGM {nodes [
   n1 = IsAvailableInCurrentGM!() -> (then, IsAvailable:bool=~) {is=title K2Node_FunctionEntry}
   > SetRenderOpacity(self=[Get Rewards(), Get Content(), Get Lazy_Background()], InOpacity=SelectFloat(A=1, B=0.5, bPickA=@n1))
  ]}
 IsCategoryTracked {nodes [
   n1 = IsCategoryTracked!() -> (then, IsShowMark:bool=~) {is=title K2Node_FunctionEntry}
   > SetVisibility(
    self=Get TrackedImage()
    InVisibility=Select(Option 0:ESlateVisibility Enum=Collapsed, Option 1:ESlateVisibility Enum=SelfHitTestInvisible, Index=@n1) -> (ReturnValue:ESlateVisibility Enum)
   )
  ]}
 IsNew {nodes [
   n1 = IsNew!() -> (then, IsShowIndIcator:bool=~) {is=title K2Node_FunctionEntry}
   > Set Bang State!(execute, self:Fort Bang Wrapper NUI Object Reference=Get StatusIndicator(), bEnabled:bool=@n1, Count:int=0) -> (then) {is=title K2Node_CallFunction}
  ]}
 PopulateIcons {nodes [
   PopulateIcons!() -> (then, CompletedCount:int, TotalCount:int, LockedCount:int) {is=title K2Node_FunctionEntry}
   > ClearChildren(self=Get CategoryProgress() -> (CategoryProgress:Wrap Box Object Reference))
   > n1 = For Loop(FirstIndex=1, LastIndex=Get TotalCount() -> (TotalCount:int=0))
   > n2 = Construct Object from Class(Class=&Image) -> (ReturnValue:Image Object Reference)
   > SetBrushFromMaterial(self=@n2, Material=/FortniteCoreUI/UI/Quests/Materials/M_QuestCheckBox)
   > n3 = GetDynamicMaterial(self=@n2)
   > n4 = Branch(Condition="integer <= integer"(A=@n1, B=Get CompletedCount() -> (CompletedCount:int=0)))
   > n5 = SetScalarParameterValue(self=@n3, ParameterName=State, Value=0.5)
   AddChild(execute=[@n5.then, @n9.then, @n10.then], self=Get CategoryProgress(), Content=@n2)
   n6 = Get TotalCount() -> (TotalCount=0)
   n7 = Get LockedCount() -> (LockedCount:int=0)
   n8 = Branch(execute=@n4.else, Condition="integer <= integer"(A=@n1, B=int - int(A=@n6, B=@n7)))
   > n9 = SetScalarParameterValue(self=@n3, ParameterName=State, Value=1)
   n10 = SetScalarParameterValue(execute=@n8.else, self=@n3, ParameterName=State, Value=0)
  ]}
 SetCategoryVM {nodes [
   n1 = SetCategoryVM!() -> (then, CategoryVM:Quest UI Category View Model Object Reference) {is=title K2Node_FunctionEntry}
   > Is Valid(InputObject=@n1)
   > Bind Event to On Category Progress Changed(self:Quest UI Category View Model Object Reference=@n1, Delegate=Create Event()) {is=title K2Node_AddDelegate}
   > UpdateProgress()
   > UpdateExpirationSwitcher()
  ]}
 SetCategoryXPAmount {nodes [
   n1 = SetCategoryXPAmount!() -> (then, QuestUIRewardVMs:Array of Quest UI Reward VM Object References) {is=title K2Node_FunctionEntry}
   n2 = "Get (a copy)"(Array:Array of Quest UI Reward VM Object References=@n1) -> (Output:Quest UI Reward VM Object Reference)
   > n3 = Is Valid(InputObject=@n2)
   > n4 = Branch(Condition=Get bIsSeasonalXP!(self:Quest UI Reward VM Object Reference=@n2) -> (bIsSeasonalXP:bool))
   > n5 = Get!(execute:Exec) -> (then:Exec, else:Exec, QuestUICategoryVM:Quest UI Category View Model Object Reference)
   n6 = Get bHasCinematicRewards!(self:Quest UI Category View Model Object Reference=@n5) -> (bHasCinematicRewards:bool)
   n7 = Get bHasCosmeticRewards!(self:Quest UI Category View Model Object Reference=@n5) -> (bHasCosmeticRewards:bool)
   n8 = Get ItemType!(self:Quest UI Reward VM Object Reference=@n2) -> (ItemType:EFortItemType Enum=WorldItem)
   n9 = Get Quantity!(self:Quest UI Reward VM Object Reference=@n2) -> (Quantity:int=0)
   n10 = AND Boolean(
    A=NOT Boolean(A=OR Boolean(A=@n7, B=@n6))
    B=AND Boolean(A="Not Equal (Integer)"(A=@n9, B=1), B="Equal (Enum)"(A:EFortItemType Enum=@n8, B:EFortItemType Enum=AccountResource))
   )
   n11 = "To Text (Integer)"(Value=@n9)
   n12 = int / int(A=@n9, B=1000)
   n13 = Format Text!(Format="LOCTEXT(\"{Amount}k\")", Amount:int=@n12) -> (Result)
   n14 = "integer >= integer"(A=@n12, B=1)
   > Set Text(
    Text=Select(Option 0:Text, Option 1:Text=Select(Option 0=@n11, Option 1=@n13, Index=@n14), Index=@n10) -> (ReturnValue:Text)
    self=Get Text_XPAmount() -> (Text_XPAmount:Common Text Object Reference)
   )
   > Set Visibility(
    Visibility=Select(Option 0:ESlateVisibility Enum=Collapsed, Option 1:ESlateVisibility Enum=SelfHitTestInvisible, Index=@n10) -> (ReturnValue:ESlateVisibility Enum)
    self=Get Image_XP() -> (Image_XP:Image Object Reference)
   )
   Set Text(execute=[@n5.else, @n4.else, @n3.Is Not Valid], self=Get Text_XPAmount())
   > Set Visibility(Visibility=Collapsed, self=Get Image_XP())
  ]}
 SetCompactUIStyle {nodes [
   SetCompactUIStyle() {is=title K2Node_FunctionEntry}
   > Branch(Condition=IsMobileGame())
   > Set TileHeight(TileHeight:double=Get CompactUITileHeight() -> (CompactUITileHeight:double)) -> (Output_Get:double)
   > Set TileHeightLarge(TileHeightLarge:double=Get CompactUITileHeightLarge() -> (CompactUITileHeightLarge:double)) -> (Output_Get:double)
   > SetDesiredSizeOverride!(execute, self:Image Object Reference=Get TrackedImage(), DesiredSize:Vector 2D Structure=(X=32,Y=32)) -> (then)
  ]}
 SetCustomExpirationText {nodes [
   SetCustomExpirationText!() -> (then, CustomExpirationText:Text) {is=title K2Node_FunctionEntry}
   > UpdateExpirationSwitcher()
  ]}
 SetDebugCategoryID {nodes [
   n1 = SetDebugCategoryID!() -> (then, CategoryID:"String (by ref)") {is=title K2Node_FunctionEntry}
   "SetText (Text)"(
    execute=[@n3.else, @n1.then, @n2.else]
    self=Get DebugCategoryID()
    InText="To Text (String)"(InString=Get CategoryID() -> (CategoryID:"String (by ref)"))
   )
   n2 = Get()
   > n3 = Branch(Condition=Get bHasExpirationTime!(self:Quest UI Category View Model Object Reference=@n2) -> (bHasExpirationTime:bool))
   n4 = "To Text (String)"(InString=Get CategoryID())
   > n5 = Branch(Condition=TextIsEmpty(InText=@n4))
   n6 = Get ExpirationTime!(self:Quest UI Category View Model Object Reference=@n2) -> (ExpirationTime:Date Time Structure)
   "SetText (Text)"(
    execute=@n5.else
    self=Get DebugCategoryID()
    InText=Format Text!(
     Format="INVTEXT(\"{0}; Ends {1}\")"
     0:Text=@n4
     1:Text=As DateTime!(In:"Date Time Structure (by ref)"=@n6, InDateStyle:EDateTimeStyle Enum=Default, InTimeStyle:EDateTimeStyle Enum=Default) -> (ReturnValue:Text)
    ) -> (Result)
   )
  ]}
 SetExpirationTime {nodes [
   n1 = SetExpirationTime!() -> (then, ExpirationTime:Date Time Structure) {is=title K2Node_FunctionEntry}
   n2 = GetFirstLocalFortPlayerController!(bIncludeNonSpectators:bool=true, bIncludeSpectators:bool=false) -> (ReturnValue:Fort Player Controller Object Reference)
   > Is Valid(InputObject=@n2)
   > n3 = GetMcpRelativeUtcNow!(execute, FortPC:Fort Player Controller Object Reference=@n2) -> (then, ReturnValue:Date Time Structure)
   n4 = Get Total Hours!(A:Timespan Structure=DateTime - DateTime!(A:Date Time Structure=@n1, B:Date Time Structure=@n3) -> (ReturnValue:Timespan Structure)) -> (ReturnValue:double) {is=title K2Node_CallFunction}
   n5 = "float < float"(A=@n4, B=24)
   n6 = "float > float"(A=@n4)
   > SetText(
    self=Get Highlight_Expiration()
    Text=Select(Option 0:Text, Option 1:Text="LOCTEXT(\"LEAVING TODAY\")", Index=AND Boolean(A=@n5, B=@n6)) -> (ReturnValue:Text)
   )
  ]}
 SetExternalLinkAllowsText {nodes [
   SetExternalLinkAllowsText!() -> (then, Allowed:bool=~) {is=title K2Node_FunctionEntry}
   > UpdateExpirationSwitcher()
  ]}
 SetExternalLinkVM {nodes [
   n1 = SetExternalLinkVM!() -> (then, LinkVM:Quest UIExternal Link VM Object Reference) {is=title K2Node_FunctionEntry}
   > Set QuestUIExternalLinkVM(QuestUIExternalLinkVM:Quest UIExternal Link VM Object Reference=@n1) -> (Output_Get:Quest UIExternal Link VM Object Reference)
   > UpdateExpirationSwitcher()
  ]}
 SetHeader {nodes [
   n1 = SetHeader!() -> (then, Text:Text) {is=title K2Node_FunctionEntry}
   > n2 = Is Valid(InputObject=Get QuestUICategoryVM())
   "SetText (Text)"(execute=@n2.Is Not Valid, self=Get Header() -> (Header:Common Text Object Reference), InText="LOCTEXT(\"My Tracked Quests\")")
   n3 = GetCategoryName!(self:Quest UI Category View Model Object Reference=Get QuestUICategoryVM()) -> (ReturnValue:Text)
   "SetText (Text)"(
    execute=@n2.Is Valid
    self=Get Header()
    InText=Select(
     Option 0=@n1
     Option 1=Format Text!(Format="LOCTEXT(\"Empty data \\\"{short}\\\"\")", short:Text=@n3) -> (Result)
     Index=TextIsEmpty(InText=@n1)
    )
   )
   n4 = Make Array!("[0]"=(TagName="StyleSheet.Style.Subheading"), "[1]":Gameplay Tag Structure=(TagName="StyleSheet.Style.20")) -> (Array)
   n5 = Make Array("[0]"=(TagName="StyleSheet.Style.Subheading"), "[1]"=(TagName="StyleSheet.Style.05"))
   n6 = Get IsDebugDisplay()
   UpdateWidgetStyleTags!(
    execute
    InWidget:Widget Object Reference=Get Header()
    InAddedTags:Array of Gameplay Tag Structures=Select(Option 0:Array of Gameplay Tag Structures=@n4, Option 1:Array of Gameplay Tag Structures=@n5, Index=@n6) -> (ReturnValue:Array of Gameplay Tag Structures)
    InRemovedTags:Array of Gameplay Tag Structures=Select(Option 0=@n5, Option 1=@n4, Index=@n6)
   ) -> (then, ReturnValue:bool)
  ]}
 SetIsHalfTile {nodes [
   n1 = SetIsHalfTile!() -> (then, HalfTile:bool=~) {is=title K2Node_FunctionEntry}
   > Set IsHalfTile(IsHalfTile:bool=@n1) -> (Output_Get:bool)
   > n2 = GetDynamicMaterial(self=Get Lazy_Background())
   > Is Valid(InputObject=@n2)
   n3 = Get IsTrackedQuestTile()
   n4 = Get IsHalfTile()
   > SetScalarParameterValue(self=@n2, ParameterName=Get Param_IsHalfTile() -> (Param_IsHalfTile=None), Value="To Float (Boolean)"(InBool=OR Boolean(A=@n4, B=@n3)))
  ]}
 SetTrackedQuestStyle {nodes [
   n1 = SetTrackedQuestStyle!() -> (then, IsTrackedQuestTile:bool=~) {is=title K2Node_FunctionEntry}
   > Set IsTrackedQuestTile(IsTrackedQuestTile:bool=@n1) -> (Output_Get:bool)
   > DisplayTrackedQuestStyle()
  ]}
 UpdateExpirationSwitcher {nodes [
   UpdateExpirationSwitcher() {is=title K2Node_FunctionEntry}
   n1 = Get QuestUIExternalLinkVM() -> (QuestUIExternalLinkVM:Quest UIExternal Link VM Object Reference)
   > n2 = Is Valid(InputObject=@n1)
   n3 = Get QuestUICategoryVM()
   n4 = Is Valid(exec=[@n11.else, @n2.Is Not Valid], InputObject=@n3)
   > n5 = Branch(Condition=Get bHasExpirationTime(self=@n3))
   n6 = Get Switcher_Expiration() -> (Switcher_Expiration:Common Visibility Switcher Object Reference)
   > n7 = SetActiveWidgetIndex(self=@n6, Index=0)
   SetVisibility(execute=[@n10.then, @n7.then, @n12.then], self=@n6, InVisibility=SelfHitTestInvisible)
   n8 = Get CustomExpirationText!(self:Quest UI Category View Model Object Reference=@n3) -> (CustomExpirationText:Text)
   n9 = Branch(execute=@n5.else, Condition=NOT Boolean(A=TextIsEmpty(InText=@n8)))
   > SetActiveWidgetIndex(self=@n6, Index=1)
   > n10 = SetText(self=Get Highlight_CustomExpiration(), Text=@n8)
   SetVisibility(execute=[@n9.else, @n4.Is Not Valid], self=@n6, InVisibility=Collapsed)
   n11 = Branch(execute=@n2.Is Valid, Condition=Get bOwnerAllowsText!(self:Quest UIExternal Link VM Object Reference=@n1) -> (bOwnerAllowsText:bool))
   > SetActiveWidgetIndex(self=@n6, Index=2)
   > n12 = Set ColorAndOpacity!(
    execute
    ColorAndOpacity:Slate Color Structure=Make Slate Color(SpecifiedColor=(R=1,G=1,B=1,A=1))
    self:Text Object Reference=Get Text_Highlight!(self:WBP UIKit Highlight Base Object Reference=Get HIghlight_BP() -> (HIghlight_BP:WBP UIKit Highlight Object Reference)) -> (Text_Highlight:Common Text Object Reference)
   ) -> (then, Output_Get:Slate Color Structure)
  ]}
 UpdateProgress {nodes [
   UpdateProgress() {is=title K2Node_FunctionEntry}
   > n1 = Switch on EQuestCategoryProgressReportType!(
    execute
    Selection:EQuestCategoryProgressReportType Enum=Get CategoryProgressReportType!(self:Quest UI Category View Model Object Reference=Get QuestUICategoryVM()) -> (CategoryProgressReportType:EQuestCategoryProgressReportType Enum=CategoryCompletionOnly)
   ) -> (CategoryCompletionOnly:Exec, Icons:Exec, ProgressBar:Exec) {is=title K2Node_SwitchEnum}
   SetActiveWidgetIndex(execute=@n1.Icons, self=Get Switcher_CategoryState(), Index=0)
   > SetVisibility(self=Get Switcher_CategoryState(), InVisibility=SelfHitTestInvisible)
   n2 = Get QuestUICategoryVM()
   n3 = Get CompletedQuestCount!(self:Quest UI Category View Model Object Reference=@n2) -> (CompletedQuestCount:int=0)
   n4 = Get TotalQuestCount!(self:Quest UI Category View Model Object Reference=@n2) -> (TotalQuestCount:int=0)
   > PopulateIcons!(
    execute
    self
    CompletedCount:int=@n3
    TotalCount:int=@n4
    LockedCount:int=Get LockedQuestCount!(self:Quest UI Category View Model Object Reference=@n2) -> (LockedQuestCount:int=0)
   ) -> (then) {is=title K2Node_CallFunction}
   SetActiveWidgetIndex(execute=@n1.ProgressBar, self=Get Switcher_CategoryState(), Index=1)
   > SetVisibility(self=Get Switcher_CategoryState(), InVisibility=SelfHitTestInvisible)
   > "SetText (Text)"(
    self=Get CommonText_CategoryProgress() -> (CommonText_CategoryProgress:Common Text Object Reference)
    InText=Format Text!(Format="LOCTEXT(\"{Completed}/{Total}\")", Completed:int=@n3, Total:int=@n4) -> (Result)
   )
   > n5 = GetDynamicMaterial(self=Get Progress() -> (Progress:Image Object Reference))
   > SetScalarParameterValue(self=@n5, ParameterName=AnimateProgress, Value=float / float(A:int=@n3, B:int=@n4))
   n6 = Get QuestUICategoryVM()
   n7 = Branch(
    execute=@n1.CategoryCompletionOnly
    Condition="integer >= integer"(A=Get CompletedQuestCount(self=@n6) -> (CompletedQuestCount=0), B=Get TotalQuestCount(self=@n6) -> (TotalQuestCount=0))
   )
   > SetActiveWidgetIndex(self=Get Switcher_CategoryState(), Index=0)
   > SetVisibility(self=Get Switcher_CategoryState(), InVisibility=SelfHitTestInvisible)
   > PopulateIcons(CompletedCount=1, TotalCount=1, LockedCount=0)
   SetVisibility(execute=@n7.else, self=Get Switcher_CategoryState(), InVisibility=Collapsed)
  ]}
 UpdateSize {nodes [
   n1 = UpdateSize!() -> (then, Type:EQuestCategorySize Enum=Small) {is=title K2Node_FunctionEntry}
   n2 = NOT Boolean(A=Get IsDebugDisplay())
   n3 = "Equal (Enum)"(A:EQuestCategorySize Enum=@n1, B:EQuestCategorySize Enum=Large)
   > SetMinDesiredHeight!(
    execute
    self:Size Box Object Reference=Get TileSizeBox()
    InMinDesiredHeight:float=SelectFloat(A=Get TileHeightLarge() -> (TileHeightLarge:double), B=Get TileHeight() -> (TileHeight:double), bPickA=AND Boolean(A=@n3, B=@n2))
   ) -> (then)
   > n4 = GetDynamicMaterial(self=Get Lazy_Background())
   n5 = Get Type() -> (Type:EQuestCategorySize Enum=Small)
   > SetScalarParameterValue(self=@n4, ParameterName=Get Param_IsLarge() -> (Param_IsLarge=None), Value="To Float (Boolean)"(InBool="Equal (Enum)"(A=@n5, B=Large)))
   n6 = Get Type() -> (Type=Small)
   SetFont!(
    execute
    self=Get Header()
    InFontInfo_FontObject:Object Reference=&HeadingNow
    InFontInfo_FontMaterial:Object Reference
    InFontInfo_OutlineSettings:Font Outline Settings Structure
    InFontInfo_TypefaceFontName:Name=None
    InFontInfo_Size:float=Select(Option 0:float=24, Option 1:float=27, Index="Equal (Enum)"(A=@n6, B=Large)) -> (ReturnValue:float)
    InFontInfo_LetterSpacing:int=0
    InFontInfo_SkewAmount:float=0
    InFontInfo_bForceMonospaced:bool=false
    InFontInfo_bMaterialIsStencil:bool=false
    InFontInfo_MonospacedWidth:float=1
   ) -> (then)
   "Equal (Enum)"(A=@n1, B=Wide)
   Get CategorySize(self=Get QuestUICategoryVM()) -> (CategorySize=Small)
  ]}
]
variables [
 CompactUITileHeight {type=number, default=140}
 CompactUITileHeightLarge {type=number, default=140}
 ExpirationTime {type=DateTime, default=0001.01.01-00.00.00}
 IsDebugDisplay {type=boolean}
 IsHalfTile {type=boolean}
 IsLeavingToday {type=boolean}
 IsTesting {type=boolean}
 IsTrackedQuestTile {type=boolean}
 Param_IsDebugDisplay {type=string, default=IsDebugDisplay}
 Param_IsHalfTile {type=string, default=IsHalfTile}
 Param_IsLarge {type=string, default=IsLargeTile}
 TileHeight {type=number, default=140}
 TileHeightLarge {type=number, default=140}
 TrackedQuestCounter {type=WBP_TrackedQuestsCounter_C, default=null}
]
widgets {
 info {
  inheritedWidgetCount=1
  parentClass=&/WBP_QuestTile_Base.WBP_QuestTile_Base_C
  rootWidgetClass=&SizeBox
  widgetCount=29
 }
 widgets [
  TileSizeBox {
   widgetClassPath=&SizeBox
   bIsVariable=true
   widget=TileSizeBox
   properties {bOverride_MinDesiredHeight=true, minDesiredHeight=100}
  }
  GridPanel_Root {
   widgetClassPath=&GridPanel
   bIsVariable=true
   parent=TileSizeBox
   widget=GridPanel_Root
   properties {columnFill [1,0], rowFill [1]}
   slot {class=SizeBoxSlot}
  }
  StatusIndicator {
   widgetClassPath=/FortUILibrary/UIKit/StatusIndicator/WBP_UIKit_StatusIndicator
   bIsVariable=true
   parent=GridPanel_Root
   widget=StatusIndicator
   properties {type=Attention, configuration=Dot}
   slot {class=GridSlot, properties {column=1, horizontalAlignment=HAlign_Right, layer=10, verticalAlignment=VAlign_Top, padding=<ltrb 0 8 8 0>}}
  }
  Block_Outline {
   widgetClassPath=/FortUILibrary/UIKit/Buttons/Blocks/WBP_UIKit_Block_Outline
   bIsVariable=true
   parent=GridPanel_Root
   widget=Block_Outline
   properties {bottomLeftCornerRadius=24, bottomRightCornerRadius=24, radiusInPixels=true, topLeftCornerRadius=24, topRightCornerRadius=24}
   slot {class=GridSlot, properties {columnSpan=2}}
  }
  Lazy_Background {
   widgetClassPath=&FortLazyImage
   bIsVariable=true
   parent=GridPanel_Root
   widget=Lazy_Background
   properties {
    brush {
     bIsDynamicallyLoaded=false
     drawAs=Image
     imageType=FullColor
     mirroring=NoMirror
     resourceName=None
     resourceObject=/FortniteCoreUI/Items/QuestItems/Categories/Art/MI_UI_CategoryLoading
     tiling=NoTile
     imageSize=<32 32>
     margin=<ltrb 0 0 0 0>
     outlineSettings {
      bUseBrushTransparency=false
      roundingType=HalfHeightRadius
      width=0
      color {colorUseRule=UseColor_Specified, specifiedColor=<rgba 0 0 0 0>}
      cornerRadii=<0 0 0 0>
     }
     tintColor {colorUseRule=UseColor_Specified, specifiedColor=<rgba 1 1 1 1>}
     uVRegion {bIsValid=false, max=<0 0>, min=<0 0>}
    }
    loadingBackgroundBrush {
     bIsDynamicallyLoaded=false
     drawAs=Image
     imageType=FullColor
     mirroring=NoMirror
     resourceName=None
     resourceObject=&MI_UI_CategoryLoading
     tiling=NoTile
     imageSize=<32 32>
     margin=<ltrb 0 0 0 0>
     outlineSettings {
      bUseBrushTransparency=false
      roundingType=HalfHeightRadius
      width=0
      color {colorUseRule=UseColor_Specified, specifiedColor=<rgba 0 0 0 0>}
      cornerRadii=<0 0 0 0>
     }
     tintColor {colorUseRule=UseColor_Specified, specifiedColor=<rgba 1 1 1 0.5>}
     uVRegion {bIsValid=false, max=<0 0>, min=<0 0>}
    }
   }
   slot {class=GridSlot, properties {columnSpan=2}}
  }
  Content {
   widgetClassPath=&GridPanel
   bIsVariable=true
   parent=GridPanel_Root
   widget=Content
   properties {clipping=ClipToBounds, columnFill [1], rowFill [0,0,1]}
   slot {class=GridSlot, properties {columnSpan=2, layer=3, padding=<ltrb 16 12 80 16>}}
  }
  HorizontalBox_0 {widgetClassPath=&HorizontalBox, parent=Content, widget=HorizontalBox_0, slot {class=GridSlot}}
  TrackedImage {
   widgetClassPath=&Image
   bIsVariable=true
   parent=HorizontalBox_0
   widget=TrackedImage
   properties {brush {
     bIsDynamicallyLoaded=false
     drawAs=Image
     imageType=FullColor
     mirroring=NoMirror
     resourceName=None
     resourceObject=/QuestUI/QuestScreen/Materials/T_UI_QuestPin_Tracked
     tiling=NoTile
     imageSize=<32 32>
     margin=<ltrb 0 0 0 0>
     outlineSettings {
      bUseBrushTransparency=false
      roundingType=HalfHeightRadius
      width=0
      color {colorUseRule=UseColor_Specified, specifiedColor=<rgba 0 0 0 0>}
      cornerRadii=<0 0 0 0>
     }
     tintColor {colorUseRule=UseColor_Specified, specifiedColor=<rgba 1 1 1 1>}
     uVRegion {bIsValid=false, max=<0 0>, min=<0 0>}
    }}
   slot {class=HorizontalBoxSlot}
  }
  Switcher_Expiration {
   widgetClassPath=&CommonVisibilitySwitcher
   bIsVariable=true
   parent=HorizontalBox_0
   widget=Switcher_Expiration
   slot {class=HorizontalBoxSlot, properties {horizontalAlignment=HAlign_Left, size {sizeRule=Fill, value=1}}}
  }
  Highlight_Expiration {
   widgetClassPath=/FortUILibrary/UIKit/Highlight/WBP_UIKit_Highlight
   bIsVariable=true
   parent=Switcher_Expiration
   widget=Highlight_Expiration
   properties {highlightSize=Compact, highlightText="", highlightType=Urgent}
   slot {class=CommonVisibilitySwitcherSlot}
  }
  Highlight_CustomExpiration {
   widgetClassPath=&WBP_UIKit_Highlight
   bIsVariable=true
   parent=Switcher_Expiration
   widget=Highlight_CustomExpiration
   properties {highlightSize=Compact, highlightText="", highlightType=Deal}
   slot {class=CommonVisibilitySwitcherSlot}
  }
  HIghlight_BP {
   widgetClassPath=&WBP_UIKit_Highlight
   bIsVariable=true
   parent=Switcher_Expiration
   widget=HIghlight_BP
   properties {highlightSize=Compact, highlightText=BATTLE PASS REQUIRED, highlightType=Urgent}
   slot {class=CommonVisibilitySwitcherSlot}
  }
  Header {
   widgetClassPath=&CommonTextBlock
   bIsVariable=true
   parent=Content
   widget=Header
   properties {
    scrollStyle=&TextScrollStyle-Base_C
    text=Quest title
    textTransformPolicy=ToUpper
    colorAndOpacity {colorUseRule=UseColor_Specified, specifiedColor=<rgba 1 1 1 0.9>}
    font {
     bForceMonospaced=false
     bMaterialIsStencil=false
     fontMaterial=null
     fontObject=&HeadingNow
     letterSpacing=0
     monospacedWidth=1
     size=18
     skewAmount=0
     typefaceFontName=86 Bold
     outlineSettings {
      bApplyOutlineToDropShadows=false
      bMiteredCorners=false
      bSeparateFillAlpha=false
      outlineMaterial=null
      outlineSize=0
      outlineColor=<rgba 0 0 0 1>
     }
    }
    margin=<ltrb 0 0 4 0>
    shadowOffset=<0 0>
   }
   slot {class=GridSlot, properties {horizontalAlignment=HAlign_Left, row=5, padding=<ltrb 0 0 64 0>}}
  }
  DebugCategoryID {
   widgetClassPath=&CommonTextBlock
   bIsVariable=true
   parent=Content
   widget=DebugCategoryID
   properties {
    bAutoCollapseWithEmptyText=true
    text=Debug Category ID
    colorAndOpacity {colorUseRule=UseColor_Specified, specifiedColor=<rgba 0.5776 0.5776 0.5776 1>}
    font {
     bForceMonospaced=false
     bMaterialIsStencil=false
     fontMaterial=null
     fontObject=&HeadingNow
     letterSpacing=0
     monospacedWidth=1
     size=10.5
     skewAmount=0
     typefaceFontName=74 Regular
     outlineSettings {
      bApplyOutlineToDropShadows=false
      bMiteredCorners=false
      bSeparateFillAlpha=false
      outlineMaterial=null
      outlineSize=0
      outlineColor=<rgba 0 0 0 1>
     }
    }
    shadowOffset=<0 0>
   }
   slot {class=GridSlot, properties {row=2, verticalAlignment=VAlign_Bottom}}
  }
  Switcher_CategoryState {
   widgetClassPath=&CommonVisibilitySwitcher
   bIsVariable=true
   parent=Content
   widget=Switcher_CategoryState
   slot {class=GridSlot, properties {horizontalAlignment=HAlign_Left, layer=2, row=4, verticalAlignment=VAlign_Bottom}}
  }
  CategoryProgress {
   widgetClassPath=&WrapBox
   bIsVariable=true
   parent=Switcher_CategoryState
   widget=CategoryProgress
   properties {bExplicitWrapSize=true, horizontalAlignment=HAlign_Fill, wrapSize=600, innerSlotPadding=<7 7>}
   slot {class=CommonVisibilitySwitcherSlot}
  }
  DebugImage {
   widgetClassPath=&Image
   parent=CategoryProgress
   widget=DebugImage
   properties {brush {
     bIsDynamicallyLoaded=false
     drawAs=Image
     imageType=FullColor
     mirroring=NoMirror
     resourceName=None
     resourceObject=&M_QuestCheckBox
     tiling=NoTile
     imageSize=<32 32>
     margin=<ltrb 0 0 0 0>
     outlineSettings {
      bUseBrushTransparency=false
      roundingType=HalfHeightRadius
      width=0
      color {colorUseRule=UseColor_Specified, specifiedColor=<rgba 0 0 0 0>}
      cornerRadii=<0 0 0 0>
     }
     tintColor {colorUseRule=UseColor_Specified, specifiedColor=<rgba 1 1 1 1>}
     uVRegion {bIsValid=false, max=<0 0>, min=<0 0>}
    }}
   slot {class=WrapBoxSlot}
  }
  CategoryQuests {
   widgetClassPath=&GridPanel
   parent=Switcher_CategoryState
   widget=CategoryQuests
   properties {columnFill [1]}
   slot {class=CommonVisibilitySwitcherSlot, properties {horizontalAlignment=HAlign_Left, verticalAlignment=VAlign_Bottom, padding=<ltrb 0 0 64 4>}}
  }
  Progress {
   widgetClassPath=&Image
   bIsVariable=true
   parent=CategoryQuests
   widget=Progress
   properties {brush {
     bIsDynamicallyLoaded=false
     drawAs=Image
     imageType=FullColor
     mirroring=NoMirror
     resourceName=None
     resourceObject=/FortniteCoreUI/UI/Quests/MaterialInstances/MI_UI_QuestProgressBar
     tiling=NoTile
     imageSize=<244 16>
     margin=<ltrb 0.5 0.5 0.5 0.5>
     outlineSettings {
      bUseBrushTransparency=false
      roundingType=HalfHeightRadius
      width=0
      color {colorUseRule=UseColor_Specified, specifiedColor=<rgba 0 0 0 0>}
      cornerRadii=<0 0 0 0>
     }
     tintColor {colorUseRule=UseColor_Specified, specifiedColor=<rgba 1 1 1 1>}
     uVRegion {bIsValid=false, max=<0 0>, min=<0 0>}
    }}
   slot {class=GridSlot, properties {horizontalAlignment=HAlign_Left, verticalAlignment=VAlign_Center}}
  }
  CommonText_CategoryProgress {
   widgetClassPath=&CommonTextBlock
   bIsVariable=true
   parent=CategoryQuests
   widget=CommonText_CategoryProgress
   properties {
    mobileFontSizeMultiplier=1.14
    text=1/10
    visibility=Collapsed
    font {
     bForceMonospaced=false
     bMaterialIsStencil=false
     fontMaterial=null
     fontObject=&HeadingNow
     letterSpacing=0
     monospacedWidth=1
     size=15.75
     skewAmount=0
     typefaceFontName=75 Medium
     outlineSettings {
      bApplyOutlineToDropShadows=false
      bMiteredCorners=false
      bSeparateFillAlpha=false
      outlineMaterial=null
      outlineSize=0
      outlineColor=<rgba 0 0 0 1>
     }
    }
    shadowOffset=<0 0>
   }
   slot {class=GridSlot, properties {column=1, verticalAlignment=VAlign_Center, padding=<ltrb 8 0 0 0>}}
  }
  Spacer_161 {widgetClassPath=&Spacer, parent=Content, widget=Spacer_161, properties {size=<600 0>}, slot {class=GridSlot}}
  Rewards {
   widgetClassPath=&GridPanel
   bIsVariable=true
   parent=GridPanel_Root
   widget=Rewards
   properties {columnFill [1], rowFill [1,0]}
   slot {class=GridSlot, properties {column=1, horizontalAlignment=HAlign_Right, layer=10, padding=<ltrb 0 0 16 16>}}
  }
  Switcher_Reward {
   widgetClassPath=&CommonVisibilitySwitcher
   bIsVariable=true
   parent=Rewards
   widget=Switcher_Reward
   slot {class=GridSlot, properties {horizontalAlignment=HAlign_Center, verticalAlignment=VAlign_Bottom}}
  }
  XP {
   widgetClassPath=&VerticalBox
   parent=Switcher_Reward
   widget=XP
   slot {
    class=CommonVisibilitySwitcherSlot
    properties {horizontalAlignment=HAlign_Center, verticalAlignment=VAlign_Bottom, padding=<ltrb -8 0 -8 0>}
   }
  }
  Image_XP {
   widgetClassPath=&Image
   bIsVariable=true
   parent=XP
   widget=Image_XP
   properties {
    visibility=SelfHitTestInvisible
    brush {
     bIsDynamicallyLoaded=false
     drawAs=Image
     imageType=FullColor
     mirroring=NoMirror
     resourceName=None
     resourceObject=/Game/UI/Foundation/Textures/Icons/Items/UniversalXP/T_UI_UniversalXP_Icon_Standard_128
     tiling=NoTile
     imageSize=<64 64>
     margin=<ltrb 0 0 0 0>
     outlineSettings {
      bUseBrushTransparency=false
      roundingType=HalfHeightRadius
      width=0
      color {colorUseRule=UseColor_Specified, specifiedColor=<rgba 0 0 0 0>}
      cornerRadii=<0 0 0 0>
     }
     tintColor {colorUseRule=UseColor_Specified, specifiedColor=<rgba 1 1 1 1>}
     uVRegion {bIsValid=false, max=<0 0>, min=<0 0>}
    }
   }
   slot {class=VerticalBoxSlot, properties {horizontalAlignment=HAlign_Center, padding=<ltrb 0 0 0 -8>}}
  }
  Text_XPAmount {
   widgetClassPath=&CommonTextBlock
   bIsVariable=true
   parent=XP
   widget=Text_XPAmount
   properties {
    bAutoCollapseWithEmptyText=true
    font {
     bForceMonospaced=false
     bMaterialIsStencil=false
     fontMaterial=null
     fontObject=&HeadingNow
     letterSpacing=0
     monospacedWidth=1
     size=15.75
     skewAmount=0
     typefaceFontName=75 Medium
     outlineSettings {
      bApplyOutlineToDropShadows=false
      bMiteredCorners=false
      bSeparateFillAlpha=false
      outlineMaterial=null
      outlineSize=0
      outlineColor=<rgba 0 0 0 1>
     }
    }
    shadowOffset=<0 0>
   }
   slot {class=VerticalBoxSlot, properties {horizontalAlignment=HAlign_Center}}
  }
  Image_Item {
   widgetClassPath=&Image
   bIsVariable=true
   parent=Switcher_Reward
   widget=Image_Item
   properties {
    visibility=SelfHitTestInvisible
    brush {
     bIsDynamicallyLoaded=false
     drawAs=Image
     imageType=FullColor
     mirroring=NoMirror
     resourceName=None
     resourceObject=/FortUILibrary/Textures/Icons/T_UI_Icon_UIKit_Gift
     tiling=NoTile
     imageSize=<48 48>
     margin=<ltrb 0 0 0 0>
     outlineSettings {
      bUseBrushTransparency=false
      roundingType=HalfHeightRadius
      width=0
      color {colorUseRule=UseColor_Specified, specifiedColor=<rgba 0 0 0 0>}
      cornerRadii=<0 0 0 0>
     }
     tintColor {colorUseRule=UseColor_Specified, specifiedColor=<rgba 1 1 1 1>}
     uVRegion {bIsValid=false, max=<0 0>, min=<0 0>}
    }
   }
   slot {class=CommonVisibilitySwitcherSlot, properties {horizontalAlignment=HAlign_Center, verticalAlignment=VAlign_Bottom}}
  }
  Image_Cinematic {
   widgetClassPath=&Image
   bIsVariable=true
   parent=Switcher_Reward
   widget=Image_Cinematic
   properties {
    visibility=SelfHitTestInvisible
    brush {
     bIsDynamicallyLoaded=false
     drawAs=Image
     imageType=FullColor
     mirroring=NoMirror
     resourceName=None
     resourceObject=/QuestUI/QuestScreen/Materials/MI_UI_QuestIcon_CinematicReward_VFX
     tiling=NoTile
     imageSize=<64 64>
     margin=<ltrb 0 0 0 0>
     outlineSettings {
      bUseBrushTransparency=false
      roundingType=HalfHeightRadius
      width=0
      color {colorUseRule=UseColor_Specified, specifiedColor=<rgba 0 0 0 0>}
      cornerRadii=<0 0 0 0>
     }
     tintColor {colorUseRule=UseColor_Specified, specifiedColor=<rgba 1 1 1 1>}
     uVRegion {bIsValid=false, max=<0 0>, min=<0 0>}
    }
   }
   slot {
    class=CommonVisibilitySwitcherSlot
    properties {horizontalAlignment=HAlign_Center, verticalAlignment=VAlign_Bottom, padding=<ltrb -8 0 -8 -8>}
   }
  }
  InputActionWidget {widgetClassPath=&CommonActionWidget, bInherited=true, widget=null}
 ]
}
