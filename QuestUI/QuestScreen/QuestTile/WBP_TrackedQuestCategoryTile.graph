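# Widget Blueprint for a tracked quest category tile in the quest screen UI, derived from WBP_QuestCategoryTile.
# It adds no widgets, variables or event logic of its own, and its event graph is empty.
# Any layout and behavior come from the parent tile, leaving this asset as a distinct subclass for the tracked category.
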WidgetBlueprint /QuestUI/QuestScreen/QuestTile/WBP_TrackedQuestCategoryTile
graphs [EventGraph {links [], nodes []}]
variables []
widgets {
 info {
  inheritedWidgetCount=0
  parentClass=&/WBP_QuestCategoryTile.WBP_QuestCategoryTile_C
  rootWidgetClass=null
  widgetCount=0
 }
 widgets []
}
