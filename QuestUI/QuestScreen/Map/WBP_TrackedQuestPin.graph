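# WBP_TrackedQuestPin is a quest pin widget, built from one lazy-loaded image, showing index, selection, enabled state and location type.
# Setters feed DisplayPin, which picks the normal or selected material, sets size and dims disabled pins.
# Index and location type drive scalar parameters on the dynamic material, with the index reapplied once lazy loading finishes.
# Helpers toggle tracked visibility and read the multi-pinning CVar.
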WidgetBlueprint /QuestUI/QuestScreen/Map/WBP_TrackedQuestPin
graphs [
 DisplayIcon {nodes [
   DisplayIcon() {is=title K2Node_FunctionEntry}
   > SetBrushFromLazyMaterial!(
    execute
    self:Common Lazy Image Object Reference=Get LazyImg_PinBG() -> (LazyImg_PinBG:Fort Lazy Image Object Reference)
    LazyMaterial:"Material Interface Soft Object Reference (by ref)"=Select(
     Option 0:"Material Interface Soft Object Reference (by ref)"=Get IconMI() -> (IconMI:Material Interface Soft Object Reference)
     Option 1:"Material Interface Soft Object Reference (by ref)"=Get IconMI_Selected() -> (IconMI_Selected:Material Interface Soft Object Reference)
     Index=Get IsSelected() -> (IsSelected:bool)
    ) -> (ReturnValue:"Material Interface Soft Object Reference (by ref)")
   ) -> (then)
   > UpdateQuestLocationMaterial!(execute, self, Index:EQuestLocationType Enum=Get Quest Location() -> (Quest Location:EQuestLocationType Enum=ExactLocation)) -> (then)
   > SetDesiredSizeOverride!(execute, self:Image Object Reference=Get LazyImg_PinBG(), DesiredSize:Vector 2D Structure=Get IconSize() -> (IconSize:Vector 2D Structure)) -> (then)
  ]}
 DisplayIndex {nodes [
   DisplayIndex() {is=title K2Node_FunctionEntry}
   > n1 = GetDynamicMaterial(self=Get LazyImg_PinBG())
   n2 = Get IndexUnavailable() -> (IndexUnavailable:int=0)
   n3 = Get IsQuestEnabled() -> (IsQuestEnabled:bool)
   n4 = Get IconIndex() -> (IconIndex:int=0)
   > SetScalarParameterValue(
    self=@n1
    ParameterName=Get Param_BackingTypeSelector() -> (Param_BackingTypeSelector:Name=None)
    Value="To Float (Integer)"(InInt=Select(Option 0:int=@n2, Option 1:int=@n4, Index=@n3) -> (ReturnValue:int))
   )
   > SetColorAndOpacity(
    self:Image Object Reference=Get LazyImg_PinBG()
    InColorAndOpacity:Linear Color Structure=Select(
     Option 0:Linear Color Structure=(R=0.6939,G=0.6939,B=0.6939,A=0.4)
     Option 1:Linear Color Structure=(R=1,G=1,B=1,A=1)
     Index=Get IsQuestEnabled()
    ) -> (ReturnValue:Linear Color Structure)
   )
  ]}
 DisplayPin {nodes [
   DisplayPin() {is=title K2Node_FunctionEntry}
   > DisplayIcon() {is=title K2Node_CallFunction}
   > DisplayIndex() {is=title K2Node_CallFunction}
   > Bind Event to On Loading State Changed(self:Common Lazy Image Object Reference=Get LazyImg_PinBG(), Delegate=Create Event()) {is=title K2Node_AddDelegate}
  ]}
 EventGraph {nodes [
   n1 = Event PreConstruct()
   > n2 = Branch(Condition=@n1.IsDesignTime)
   > DisplayPin() {is=title K2Node_CallFunction}
   SetBrushFromLazyMaterial(
    execute=@n2.else
    self=Get LazyImg_PinBG()
    LazyMaterial=Select(
     Option 0:"Material Interface Soft Object Reference (by ref)"=Get IconMI()
     Option 1:"Material Interface Soft Object Reference (by ref)"=Get IconMI_Selected()
     Index=Get IsSelected()
    ) -> (ReturnValue:"Material Interface Soft Object Reference (by ref)")
   )
   > UpdateQuestLocationMaterial(Index=Get Quest Location() -> (Quest Location=ExactLocation))
  ]}
 GetShowMultipinning {nodes [
   GetShowMultipinning() {is=title K2Node_FunctionEntry}
   n1 = GetNumericCVar!(CVar:String=QuestUI.EnableMultiPinning) -> (OutValue:float, ReturnValue:bool)
   > Return Node!(
    execute
    ReturnValue:bool=Select(Option 0:bool=false, Option 1:bool="Equal (Float)"(A=@n1, B=1), Index=@n1.ReturnValue) -> (ReturnValue:bool)
   )
  ]}
 HandleDisplayPinFinishedLoading {nodes [
   n1 = HandleDisplayPinFinishedLoading!() -> (then, bIsLoading:bool=~) {is=title K2Node_FunctionEntry}
   > Branch(Condition=NOT Boolean(A=@n1))
   > DisplayIndex()
   > Unbind Event from On Loading State Changed(self:Common Lazy Image Object Reference=Get LazyImg_PinBG(), Delegate=Create Event()) {is=title K2Node_RemoveDelegate}
  ]}
 InitializeQuestPin {nodes [
   n1 = InitializeQuestPin!() -> (then, IconIndex:int, IsSelected:bool=~) {is=title K2Node_FunctionEntry}
   > Set IconIndex(IconIndex:int=@n1) -> (Output_Get:int=0)
   > Set IsSelected(IsSelected:bool=@n1.IsSelected) -> (Output_Get:bool)
   > DisplayPin()
  ]}
 SetIcon {nodes [
   n1 = SetIcon!() -> (then, IconMI:Material Interface Soft Object Reference, IconMI_Selected:Material Interface Soft Object Reference) {is=title K2Node_FunctionEntry}
   > Set IconMI(IconMI:Material Interface Soft Object Reference=@n1) -> (Output_Get:Material Interface Soft Object Reference)
   > n2 = Set IconMI_Selected(IconMI_Selected:Material Interface Soft Object Reference=@n1.IconMI_Selected) -> (Output_Get:Material Interface Soft Object Reference)
   > n3 = Branch(Condition=IsValidSoftObjectReference(SoftObjectReference=@n2))
   DisplayIcon(execute=[@n3.then, @n4.then])
   n4 = Set IconMI_Selected(execute=@n3.else, IconMI_Selected=Get IconMI())
  ]}
 SetIndex {nodes [
   n1 = SetIndex!() -> (then, IconIndex:int) {is=title K2Node_FunctionEntry}
   > Set IconIndex(IconIndex=@n1) -> (Output_Get=0)
   > DisplayPin()
  ]}
 SetIsContentEnabled {nodes [
   n1 = SetIsContentEnabled!() -> (then, IsEnabled:bool=~) {is=title K2Node_FunctionEntry}
   > Set IsQuestEnabled(IsQuestEnabled:bool=@n1) -> (Output_Get:bool)
   > DisplayPin()
  ]}
 SetIsTracked {nodes [
   n1 = SetIsTracked!() -> (then, IsTracked:bool=~) {is=title K2Node_FunctionEntry}
   > SetVisibility(
    self=Get LazyImg_PinBG()
    InVisibility=Select(Option 0:ESlateVisibility Enum=Collapsed, Option 1:ESlateVisibility Enum=SelfHitTestInvisible, Index=@n1) -> (ReturnValue:ESlateVisibility Enum)
   )
  ]}
 SetPinIndex {nodes [
   SetPinIndex!() -> (then, Index:int) {is=title K2Node_FunctionEntry}
   n1 = Get Index() -> (Index:int=0)
   n2 = Get Index() -> (Index=0)
   > Set IconIndex(IconIndex=Select(Option 0:int=Get IndexStar() -> (IndexStar:int=0), Option 1:int=int + int(A=@n2), Index="integer >= integer"(A=@n1)) -> (ReturnValue:int)) -> (Output_Get=0)
   > DisplayPin()
  ]}
 SetPinMaterial {nodes [
   n1 = SetPinMaterial!() -> (then, PinMaterialInstance:Material Interface Soft Object Reference) {is=title K2Node_FunctionEntry}
   > Set IconMI(IconMI=@n1)
   > SetBrushFromLazyMaterial(self=Get LazyImg_PinBG(), LazyMaterial=Get IconMI())
   > UpdateQuestLocationMaterial(Index=Get Quest Location() -> (Quest Location=ExactLocation))
  ]}
 SetQuestLocation {nodes [
   n1 = SetQuestLocation!() -> (then, QuestLocation:EQuestLocationType Enum=ExactLocation) {is=title K2Node_FunctionEntry}
   > n2 = Set Quest Location(Quest Location:EQuestLocationType Enum=@n1) -> (Output_Get:EQuestLocationType Enum=ExactLocation)
   > UpdateQuestLocationMaterial(Index=@n2)
  ]}
 SetSelected {nodes [
   n1 = SetSelected!() -> (then, IsSelected:bool=~) {is=title K2Node_FunctionEntry}
   > Set IsSelected(IsSelected=@n1)
   > DisplayPin()
  ]}
 UpdateQuestLocationMaterial {nodes [
   n1 = UpdateQuestLocationMaterial!() -> (then, Index:EQuestLocationType Enum=ExactLocation) {is=title K2Node_FunctionEntry}
   > n2 = GetDynamicMaterial(self=Get LazyImg_PinBG())
   > SetScalarParameterValue(
    self=@n2
    ParameterName=LocationType
    Value=Select!(ExactLocation:float=0, AreaBased:float=1, NoLocation:float=2, Index:EQuestLocationType Enum=@n1) -> (ReturnValue:float)
   )
  ]}
]
variables [
 BackUpIconMI_List {type="array<MaterialInstance>", default []}
 IconIndex {type=integer, default=4}
 IconMI {type=MaterialInterface, default=/QuestUI/QuestScreen/Materials/MI_UI_QuestPin_Tile}
 IconMI_Selected {type=MaterialInterface, default=/Game/Athena/UI/Challenges/Art/MapIcons/MI_UI_QuestPin_Selected}
 IconSize {type=Vector2D, default=<32 32>}
 IndexBlank {type=integer}
 IndexStar {type=integer, default=4}
 IndexUnavailable {type=integer, default=-1}
 IsQuestEnabled {type=boolean}
 IsSelected {type=boolean}
 Param_BackingTypeSelector {type=string, default=BackingTypeSelector}
 Quest Location {type=EQuestLocationType, default=ExactLocation}
]
widgets {
 info {inheritedWidgetCount=0, parentClass=&UserWidget, rootWidgetClass=&FortLazyImage, widgetCount=1}
 widgets [LazyImg_PinBG {
   widgetClassPath=&FortLazyImage
   bIsVariable=true
   widget=LazyImg_PinBG
   properties {brush {
     bIsDynamicallyLoaded=false
     drawAs=Image
     imageType=FullColor
     mirroring=NoMirror
     resourceName=None
     resourceObject=&MI_UI_QuestPin_Tile
     tiling=NoTile
     imageSize=<32 32>
     margin=<ltrb 0 0 0 0>
     outlineSettings {
      bUseBrushTransparency=false
      roundingType=HalfHeightRadius
      width=0
      color {colorUseRule=UseColor_Specified, specifiedColor=<rgba 0 0 0 0>}
      cornerRadii=<0 0 0 0>
     }
     tintColor {colorUseRule=UseColor_Specified, specifiedColor=<rgba 1 1 1 1>}
     uVRegion {bIsValid=false, max=<0 0>, min=<0 0>}
    }}
  }]
}
