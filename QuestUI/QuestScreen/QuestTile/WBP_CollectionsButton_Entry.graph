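# Collections tile button for the quest screen, derived from WBP_QuestTile_Base, that toggles the full-screen map on click after setting it to the active product context.
# PreConstruct registers background and outline as modular button blocks.
# SetIcon async-loads a material instance into the background; it and SetName fall back to Collections defaults.
# UpdateSize sets height from its Type argument, but the material's large-tile parameter from the Type variable.
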WidgetBlueprint /QuestUI/QuestScreen/QuestTile/WBP_CollectionsButton_Entry
graphs [
 EventGraph {nodes [
   n1 = LoadIcon!() -> (OutputDelegate, then, Icon:Soft Object Reference) {is=title K2Node_CustomEvent}
   > n2 = Async Load Asset(Asset=@n1.Icon)
   n3 = Cast To MaterialInstance!(execute=@n2.Completed, Object=@n2) -> (then, CastFailed, AsMaterial Instance:Material Instance Object Reference) {is=title K2Node_DynamicCast}
   > SetBackgroundMaterial!(
    execute
    self:WBP UIKit Block Background Base Object Reference=Get Block_Background() -> (Block_Background:WBP UIKit Block Background Generic Object Reference)
    Material Instance:"Material Instance Object Reference (by ref)"=@n3
   ) -> (then)
   Event PreConstruct()
   > RegisterModularBlock!(execute, self:UIKit Modular Button Object Reference, Block:UIKit Block Object Reference=Get Block_Background()) -> (then)
   > RegisterModularBlock(Block=Get Block_Outline() -> (Block_Outline:WBP UIKit Block Outline Object Reference))
   Event On Clicked()
   > Parent: On Clicked() {is=title K2Node_CallParentFunction}
   n4 = GetComponentByClass!(
    self:Actor Object Reference=GetOwningPlayer()
    ComponentClass:Actor Component Class Reference=/Script/FortniteGame.FortControllerComponent_Minimap
   ) -> (ReturnValue:Fort Controller Component Minimap Object Reference)
   > SetActiveMapToActiveProductContext(self:Fort Controller Component Minimap Object Reference=@n4)
   > ToggleFullScreenMap(self:Fort Controller Component Minimap Object Reference=@n4)
  ]}
 SetIcon {nodes [
   n1 = SetIcon!() -> (then, IconOverride:Soft Object Reference) {is=title K2Node_FunctionEntry}
   > LoadIcon!(
    execute
    self
    Icon:Soft Object Reference=Select(
     Option 0:Soft Object Reference=/FortniteCoreUI/Items/QuestItems/Categories/Art/MI_UI_QB_Collections
     Option 1:Soft Object Reference=@n1
     Index=IsValidSoftObjectReference(SoftObjectReference=@n1)
    ) -> (ReturnValue:Soft Object Reference)
   ) -> (then) {is=title K2Node_CallFunction}
  ]}
 SetName {nodes [
   n1 = SetName!() -> (then, DisplayNameOverride:Text) {is=title K2Node_FunctionEntry}
   > "SetText (Text)"(
    self=Get Header() -> (Header:Common Text Object Reference)
    InText=Select(Option 0:Text=@n1, Option 1:Text="LOCTEXT(\"Collections\")", Index=TextIsEmpty(InText=@n1)) -> (ReturnValue:Text)
   )
  ]}
 UpdateSize {nodes [
   n1 = UpdateSize!() -> (then, Type:EQuestCategorySize Enum=Small) {is=title K2Node_FunctionEntry}
   > SetMinDesiredHeight!(
    execute
    self:Size Box Object Reference=Get SizeBox_Collections() -> (SizeBox_Collections:Size Box Object Reference)
    InMinDesiredHeight:float=SelectFloat(
     A=Get TileHeightLarge() -> (TileHeightLarge:double)
     B=Get TileHeight() -> (TileHeight:double)
     bPickA="Equal (Enum)"(A:EQuestCategorySize Enum=@n1, B:EQuestCategorySize Enum=Large)
    )
   ) -> (then)
   > n2 = GetDynamicMaterial(self=Get Fill!(self:WBP UIKit Block Background Generic Object Reference=Get Block_Background()) -> (Fill:Fort Mobile Image Object Reference))
   n3 = Get Type() -> (Type:EQuestCategorySize Enum=Small)
   > SetScalarParameterValue(self=@n2, ParameterName=Get Param_IsLarge() -> (Param_IsLarge:Name=None), Value="To Float (Boolean)"(InBool="Equal (Enum)"(A=@n3, B=Large)))
  ]}
]
variables [
 Param_IsLarge {type=string, default=IsLargeTile}
 TileHeight {type=number, default=140}
 TileHeightLarge {type=number, default=140}
]
widgets {
 info {
  inheritedWidgetCount=1
  parentClass=&/WBP_QuestTile_Base.WBP_QuestTile_Base_C
  rootWidgetClass=&SizeBox
  widgetCount=6
 }
 widgets [
  SizeBox_Collections {
   widgetClassPath=&SizeBox
   bIsVariable=true
   widget=SizeBox_Collections
   properties {bOverride_MinDesiredHeight=true, minDesiredHeight=200}
  }
  Root {
   widgetClassPath=&GridPanel
   parent=SizeBox_Collections
   widget=Root
   properties {columnFill [1], rowFill [1]}
   slot {class=SizeBoxSlot}
  }
  Block_Outline {
   widgetClassPath=/FortUILibrary/UIKit/Buttons/Blocks/WBP_UIKit_Block_Outline
   bIsVariable=true
   parent=Root
   widget=Block_Outline
   properties {bottomLeftCornerRadius=24, bottomRightCornerRadius=24, radiusInPixels=true, topLeftCornerRadius=24, topRightCornerRadius=24}
   slot {class=GridSlot, properties {layer=10}}
  }
  Block_Background {
   widgetClassPath=/FortUILibrary/Systems/ModularButton/Blocks/WBP_UIKit_Block_BackgroundGeneric
   bIsVariable=true
   parent=Root
   widget=Block_Background
   properties {backgroundMaterial=&MI_UI_QB_Collections}
   slot {class=GridSlot}
  }
  Header {
   widgetClassPath=&CommonTextBlock
   bIsVariable=true
   parent=Root
   widget=Header
   properties {
    scrollStyle=&TextScrollStyle-Base_C
    text=Collections
    textTransformPolicy=ToUpper
    colorAndOpacity {colorUseRule=UseColor_Specified, specifiedColor=<rgba 1 1 1 0.9>}
    font {
     bForceMonospaced=false
     bMaterialIsStencil=false
     fontMaterial=null
     fontObject=&HeadingNow
     letterSpacing=0
     monospacedWidth=1
     size=20.25
     skewAmount=0
     typefaceFontName=76 Bold
     outlineSettings {
      bApplyOutlineToDropShadows=false
      bMiteredCorners=false
      bSeparateFillAlpha=false
      outlineMaterial=null
      outlineSize=0
      outlineColor=<rgba 0 0 0 1>
     }
    }
    shadowOffset=<0 0>
   }
   slot {class=GridSlot, properties {layer=1, verticalAlignment=VAlign_Top, padding=<ltrb 16 16 32 0>}}
  }
  InputActionWidget {widgetClassPath=&CommonActionWidget, bInherited=true, widget=null}
 ]
}
