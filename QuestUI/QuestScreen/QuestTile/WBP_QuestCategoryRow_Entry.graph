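# Quest screen list row widget that shows category tiles and collections buttons side by side in a Dynamic Entry Box.
# It rebuilds entries from category interface viewmodels, choosing widget class by cast and flagging category tiles as half tiles when sharing a row.
# Button hover, click and focus are relayed as dispatchers, and clicks also select the row in its owning list view.
# On receiving focus, it focuses the entry at the clamped ActiveColumn, supporting column-aware navigation between rows.
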WidgetBlueprint /QuestUI/QuestScreen/QuestTile/WBP_QuestCategoryRow_Entry
graphs [
 BindButtonEvents {nodes [
   n1 = BindButtonEvents!() -> (then, ButtonInstance:Common Button Base Object Reference, Bind:bool=~) {is=title K2Node_FunctionEntry}
   > Is Valid(InputObject=@n1)
   > n2 = Branch(Condition=@n1.Bind)
   n3 = Create Event()
   > Bind Event to On Hovered(self:Common Button Base Object Reference=@n1, Delegate=@n3) {is=title K2Node_AddDelegate}
   n4 = Create Event()
   > Bind Event to On Clicked(self:Common Button Base Object Reference=@n1, Delegate=@n4) {is=title K2Node_AddDelegate}
   n5 = Create Event()
   > Bind Event to On Focused(self:Common Button Base Object Reference=@n1, Delegate=@n5) {is=title K2Node_AddDelegate}
   Unbind Event from On Hovered(execute=@n2.else, self:Common Button Base Object Reference=@n1, Delegate=@n3) {is=title K2Node_RemoveDelegate}
   > Unbind Event from On Clicked(self:Common Button Base Object Reference=@n1, Delegate=@n4) {is=title K2Node_RemoveDelegate}
   > Unbind Event from On Focused(self:Common Button Base Object Reference=@n1, Delegate=@n5) {is=title K2Node_RemoveDelegate}
  ]}
 ClearAllEntries {nodes [
   ClearAllEntries() {is=title K2Node_FunctionEntry}
   > n1 = For Each Loop(Array:Array of User Widget Object References=GetAllEntries!(self:Dynamic Entry Box Base Object Reference=Get DynamicEntryBox_Content() -> (DynamicEntryBox_Content:Dynamic Entry Box Object Reference)) -> (ReturnValue:Array of User Widget Object References)) -> (Array Element:User Widget Object Reference)
   > n2 = Cast To CommonButtonBase!(execute, Object=@n1) -> (then, CastFailed, AsCommon Button Base:Common Button Base Object Reference) {is=title K2Node_DynamicCast}
   > BindButtonEvents!(execute, self, ButtonInstance:Common Button Base Object Reference=@n2, Bind:bool=false) -> (then) {is=title K2Node_CallFunction}
   > Reset!(execute, self:Dynamic Entry Box Object Reference=Get DynamicEntryBox_Content(), bDeleteWidgets:bool=false) -> (then)
  ]}
 EventGraph {nodes [
   n1 = Event OnListItemObjectSet()
   > n2 = Cast To QuestUICategoryRowVM!(execute, Object=@n1.ListItemObject) -> (then, CastFailed, AsQuest UI Category Row View Model:Quest UI Category Row View Model Object Reference) {is=title K2Node_DynamicCast}
   > Set QuestUICategoryRowVM(QuestUICategoryRowVM:Quest UI Category Row View Model Object Reference=@n2) -> (Output_Get:Quest UI Category Row View Model Object Reference)
  ]}
 HandleButtonClicked {nodes [
   n1 = HandleButtonClicked!() -> (then, ButtonInstance:Common Button Base Object Reference) {is=title K2Node_FunctionEntry}
   > n2 = Cast To CommonListView!(execute, Object=GetOwningListView!(UserListEntry:User List Entry Interface) -> (ReturnValue:List View Base Object Reference)) -> (then, CastFailed, AsCommon List View:Common List View Object Reference) {is=title K2Node_DynamicCast}
   > Set Selected Item!(
    execute
    self:List View Object Reference=@n2
    Item:Object Reference=GetListItemObject!(UserObjectListEntry:User Object List Entry Interface) -> (ReturnValue:Object Reference)
   ) -> (then) {is=title K2Node_CallFunction}
   > Call On Entry Clicked!(execute:Exec, self:WBP Quest Category Row Entry Object Reference, Entry:Common Button Base Object Reference=@n1) -> (then:Exec) {is=title K2Node_CallDelegate}
  ]}
 HandleButtonFocused {nodes [
   n1 = HandleButtonFocused!() -> (then, ButtonInstance:Common Button Base Object Reference) {is=title K2Node_FunctionEntry}
   n2 = GetAllEntries(self=Get DynamicEntryBox_Content())
   > Branch(Condition="integer > integer"(A=Length(TargetArray:Array of User Widget Object References=@n2), B=Get ActiveColumn() -> (ActiveColumn:int=0)))
   > Call On Column Focused!(
    execute:Exec
    self:WBP Quest Category Row Entry Object Reference
    Column:int=Find Item!(TargetArray:Array of User Widget Object References=@n2, ItemToFind:"User Widget Object Reference (by ref)"=@n1) -> (ReturnValue)
   ) -> (then:Exec) {is=title K2Node_CallDelegate}
  ]}
 HandleButtonHovered {nodes [
   n1 = HandleButtonHovered!() -> (then, ButtonInstance:Common Button Base Object Reference) {is=title K2Node_FunctionEntry}
   > Call On Entry Hovered!(execute:Exec, self:WBP Quest Category Row Entry Object Reference, Entry:Common Button Base Object Reference=@n1) -> (then:Exec) {is=title K2Node_CallDelegate}
  ]}
 OnColumnFocused {nodes [OnColumnFocused!() -> (then, Column:int) {is=title K2Node_FunctionEntry}]}
 OnEntryClicked {nodes [OnEntryClicked!() -> (then, Entry:Common Button Base Object Reference) {is=title K2Node_FunctionEntry}]}
 OnEntryHovered {nodes [OnEntryHovered!() -> (then, Entry:Common Button Base Object Reference) {is=title K2Node_FunctionEntry}]}
 OnFocusReceived {nodes [
   OnFocusReceived!() -> (then, MyGeometry:Geometry Structure, InFocusEvent:Focus Event Structure) {is=title K2Node_FunctionEntry}
   n1 = GetAllEntries(self=Get DynamicEntryBox_Content())
   n2 = Length(TargetArray=@n1)
   > n3 = Branch(Condition="integer > integer"(A=@n2))
   n4 = int - int(A=@n2)
   n5 = Get ActiveColumn() -> (ActiveColumn=0)
   > n6 = SetFocus(self="Get (a copy)"(Array:Array of User Widget Object References=@n1, Dimension 1="Clamp (Integer)"(Value=@n5, Min=0, Max=@n4) -> (ReturnValue=0)) -> (Output:User Widget Object Reference))
   Return Node!(
    execute=[@n6.then, @n3.else]
    ReturnValue:Event Reply Structure=Make Event Reply!() -> (EventReply:Event Reply Structure) {is=title K2Node_MakeStruct}
   )
  ]}
 SetActiveColumn {nodes [
   n1 = SetActiveColumn!() -> (then, Column:int) {is=title K2Node_FunctionEntry}
   > Set ActiveColumn(ActiveColumn:int=@n1) -> (Output_Get:int=0)
  ]}
 SetCategoryInterfaceVMs {nodes [
   n1 = SetCategoryInterfaceVMs!() -> (then, CategoryInterfaceVMs:Array of Quest UI Category Interface View Model Object References) {is=title K2Node_FunctionEntry}
   > ClearAllEntries() {is=title K2Node_CallFunction}
   > n2 = For Each Loop(Array:Array of Quest UI Category Interface View Model Object References=@n1) -> (Array Element:Quest UI Category Interface View Model Object Reference)
   > n3 = Cast To QuestUICategoryVM!(execute, Object=@n2) -> (then, CastFailed, AsQuest UI Category View Model:Quest UI Category View Model Object Reference) {is=title K2Node_DynamicCast}
   > n4 = Create Entry of Class!(
    execute
    self:Dynamic Entry Box Object Reference=Get DynamicEntryBox_Content()
    EntryClass:User Widget Class Reference=&/WBP_QuestCategoryTile.WBP_QuestCategoryTile_C
   ) -> (then, ReturnValue:WBP Quest Category Tile Object Reference)
   > SetQuestUICategoryVM!(execute, self:WBP Quest Category Tile Object Reference=@n4, Viewmodel:Quest UI Category View Model Object Reference=@n3) -> (then)
   > BindButtonEvents(ButtonInstance=@n4, Bind=true)
   n5 = Get CategoryInterfaceVMs() -> (CategoryInterfaceVMs:Array of Quest UI Category Interface View Model Object References)
   > SetIsHalfTile!(
    execute
    self:WBP Quest Category Tile Object Reference=@n4
    HalfTile:bool="integer > integer"(A=Length(TargetArray:Array of Quest UI Category Interface View Model Object References=@n5), B=1)
   ) -> (then)
   n6 = Cast To QuestUICollectionsVM!(execute=@n3.CastFailed, Object=@n2) -> (then, CastFailed, AsQuest UI Collections View Model:Quest UI Collections View Model Object Reference) {is=title K2Node_DynamicCast}
   > n7 = Create Entry of Class(self=Get DynamicEntryBox_Content(), EntryClass=&/WBP_CollectionsButton_Entry.WBP_CollectionsButton_Entry_C) -> (ReturnValue:WBP Collections Button Entry Object Reference)
   > SetQuestUICollectionsVM!(execute, self:WBP Collections Button Entry Object Reference=@n7, Viewmodel:Quest UI Collections View Model Object Reference=@n6) -> (then)
   > BindButtonEvents(ButtonInstance=@n7, Bind=true)
  ]}
]
variables [
 ActiveColumn {type=integer}
 OnColumnFocused {}
 OnEntryClicked {}
 OnEntryHovered {}
]
widgets {
 info {inheritedWidgetCount=0, parentClass=&CommonUserWidget, rootWidgetClass=&DynamicEntryBox, widgetCount=1}
 widgets [DynamicEntryBox_Content {
   widgetClassPath=&DynamicEntryBox
   bIsVariable=true
   widget=DynamicEntryBox_Content
   properties {
    entryWidgetClass=&WBP_QuestCategoryTile_C
    numDesignerPreviewEntries=2
    entrySizeRule {sizeRule=Fill, value=1}
    entrySpacing=<24 0>
   }
  }]
}
