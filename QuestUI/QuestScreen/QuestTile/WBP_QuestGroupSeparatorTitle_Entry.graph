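# Quest screen list entry that shows separator and title rows for group headers, locked quests, compatibility notices and quest messages.
# On item set, it casts the unwrapped view model to pick a style, and UpdateStyle adjusts font, color, icon, divider lines and plain or rich title.
# Locked rows format unlock text by type (countdown, quests remaining, prerequisite), preferring runtime then config overrides.
# It also has debug ID fields and a selection outline.
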WidgetBlueprint /QuestUI/QuestScreen/QuestTile/WBP_QuestGroupSeparatorTitle_Entry
graphs [
 EventGraph {nodes [
   n1 = SetIcon!() -> (OutputDelegate, then, Icon:Material Instance Soft Object Reference) {is=title K2Node_CustomEvent}
   > SetBrushFromSoftMaterial!(
    execute
    self:Image Object Reference=Get Icon() -> (Icon:Image Object Reference)
    SoftMaterial:Material Interface Soft Object Reference=@n1.Icon
   ) -> (then)
   n2 = Event PreConstruct()
   > Branch(Condition=@n2.IsDesignTime)
   > SetLockedData()
   Event On Entry Released() {is=title K2Node_Event}
   > Set QuestUILockedQuestVM(QuestUILockedQuestVM:Quest UI Locked Quest View Model Object Reference) -> (Output_Get:Quest UI Locked Quest View Model Object Reference)
   > Set DebugQuestUIHeaderVM(DebugQuestUIHeaderVM:Debug Quest UIHeader VM Object Reference) -> (Output_Get:Debug Quest UIHeader VM Object Reference)
   > Set QuestUICompatibilityVM(QuestUICompatibilityVM:Quest UI Compatibility View Model Object Reference) -> (Output_Get:Quest UI Compatibility View Model Object Reference)
   > Set QuestUIMessageVM(QuestUIMessageVM:Quest UI Message View Model Object Reference) -> (Output_Get:Quest UI Message View Model Object Reference)
   n3 = Event OnListItemObjectSet()
   n4 = MakeLiteralString()
   > SetDebugSourceID!(execute, self, DebugString:"String (by ref)"=@n4) -> (then)
   > SetDebugRepresentedID!(execute, self, DebugString:"String (by ref)"=@n4) -> (then)
   n5 = UnwrapItem!(Item:Object Reference=@n3.ListItemObject) -> (ReturnValue:Object Reference)
   > n6 = Cast To DebugQuestUIHeaderVM!(execute, Object=@n5) -> (then, CastFailed, AsDebug Quest UIHeader VM:Debug Quest UIHeader VM Object Reference) {is=title K2Node_DynamicCast}
   > n7 = Set DebugQuestUIHeaderVM(DebugQuestUIHeaderVM=@n6)
   > UpdateStyle!(
    execute
    self
    Style:E_Seperator Enum=Select(
     Option 0:E_Seperator Enum=NewEnumerator4
     Option 1:E_Seperator Enum=NewEnumerator0
     Index=Get IsGameModeTitle!(self:Debug Quest UIHeader VM Object Reference=@n7) -> (IsGameModeTitle:bool)
    ) -> (ReturnValue:E_Seperator Enum)
   ) -> (then)
   n8 = Cast To QuestUILockedQuestVM!(execute=@n6.CastFailed, Object=@n5) -> (then, CastFailed, AsQuest UI Locked Quest View Model:Quest UI Locked Quest View Model Object Reference) {is=title K2Node_DynamicCast}
   > Set QuestUILockedQuestVM(QuestUILockedQuestVM=@n8)
   > UpdateStyle(Style=NewEnumerator1)
   n9 = Cast To QuestUICompatibilityVM!(execute=@n8.CastFailed, Object=@n5) -> (then, CastFailed, AsQuest UI Compatibility View Model:Quest UI Compatibility View Model Object Reference) {is=title K2Node_DynamicCast}
   > Set QuestUICompatibilityVM(QuestUICompatibilityVM=@n9)
   > UpdateStyle(Style=NewEnumerator2)
   n10 = Cast To QuestUIMessageVM!(execute=@n9.CastFailed, Object=@n5) -> (then, CastFailed, AsQuest UI Message View Model:Quest UI Message View Model Object Reference) {is=title K2Node_DynamicCast}
   > Set QuestUIMessageVM(QuestUIMessageVM=@n10)
   > UpdateStyle(Style=NewEnumerator3)
   SetVisibility(
    self=Get SelectedStateOutline() -> (SelectedStateOutline:Image Object Reference)
    InVisibility=Select(
     Option 0:ESlateVisibility Enum=Collapsed
     Option 1:ESlateVisibility Enum=SelfHitTestInvisible
     Index=Event On Item Selection Changed!() -> (OutputDelegate, then, bIsSelected:bool) {is=title K2Node_Event}.bIsSelected
    ) -> (ReturnValue:ESlateVisibility Enum)
   )
   Event Construct()
   Event Destruct()
  ]}
 GetFinalCategoryName {nodes [
   GetFinalCategoryName() {is=title K2Node_FunctionEntry}
   n1 = Break Locked Quest Runtime Data!(LockedQuestRuntimeData:"Locked Quest Runtime Data Structure (by ref)"=Get LockedQuestData!(self:Quest UI Locked Quest View Model Object Reference=Get QuestUILockedQuestVM() -> (QuestUILockedQuestVM:Quest UI Locked Quest View Model Object Reference)) -> (LockedQuestData:Locked Quest Runtime Data Structure)) -> (ConfigData:Locked Quest Config Data Structure, RuntimeOverrideData:Locked Quest Override Data Structure) {is=title K2Node_BreakStruct}
   n2 = Break Locked Quest Override Data!(LockedQuestOverrideData:"Locked Quest Override Data Structure (by ref)"=@n1.RuntimeOverrideData) -> (CategoryName:Text) {is=title K2Node_BreakStruct}
   > n3 = Branch(Condition=TextIsEmpty(InText=@n2))
   > Return Node!(
    execute
    CategoryName:Text=Break Locked Quest Override Data(LockedQuestOverrideData=Break Locked Quest Config Data!(LockedQuestConfigData:"Locked Quest Config Data Structure (by ref)"=@n1) -> (bConsiderHiddenContent:bool=~, bOnlyShowWhileProductActive:bool=~, bShowRewards:bool=~, OverrideData:Locked Quest Override Data Structure) {is=title K2Node_BreakStruct}.OverrideData)
   )
   Return Node(execute=@n3.else, CategoryName=@n2)
  ]}
 GetFinalDisplayText {nodes [
   n1 = GetFinalDisplayText!() -> (then, DefaultText:Text) {is=title K2Node_FunctionEntry}
   n2 = Break Locked Quest Runtime Data(LockedQuestRuntimeData=Get LockedQuestData(self=Get QuestUILockedQuestVM()))
   n3 = Break Locked Quest Override Data!(LockedQuestOverrideData=@n2.RuntimeOverrideData) -> (DisplayText:Text)
   > n4 = Branch(Condition=TextIsEmpty(InText=@n3))
   n5 = Break Locked Quest Override Data(LockedQuestOverrideData=Break Locked Quest Config Data(LockedQuestConfigData=@n2) -> (bConsiderHiddenContent=~, bOnlyShowWhileProductActive=~, bShowRewards=~).OverrideData)
   > n6 = Branch(Condition=TextIsEmpty(InText=@n5))
   > Return Node!(execute, DisplayText:Text=@n1)
   Return Node(execute=@n6.else, DisplayText=@n5)
   Return Node(execute=@n4.else, DisplayText=@n3)
  ]}
 SetDebugRepresentedID {nodes [
   n1 = SetDebugRepresentedID!() -> (then, DebugString:"String (by ref)") {is=title K2Node_FunctionEntry}
   > n2 = Branch(Condition=IsEmpty(InString=@n1))
   n3 = Get DebugFieldRight() -> (DebugFieldRight:Common Text Object Reference)
   > "SetText (Text)"(self=@n3)
   "SetText (Text)"(execute=@n2.else, self=@n3, InText=Format Text!(Format="LOCTEXT(\"Represented: {0}\")", 0:"String (by ref)"=@n1) -> (Result))
  ]}
 SetDebugSourceID {nodes [
   n1 = SetDebugSourceID!() -> (then, DebugString:"String (by ref)") {is=title K2Node_FunctionEntry}
   > n2 = Branch(Condition=IsEmpty(InString=@n1))
   n3 = Get DebugFieldLeft() -> (DebugFieldLeft:Common Text Object Reference)
   > "SetText (Text)"(self=@n3)
   "SetText (Text)"(execute=@n2.else, self=@n3, InText=Format Text(Format="LOCTEXT(\"Source: {0}\")", 0=@n1))
  ]}
 SetDebugText {nodes [
   n1 = SetDebugText!() -> (then, DebugText:"Text (by ref)") {is=title K2Node_FunctionEntry}
   > "SetText (Text)"(self=Get DebugFieldLeft(), InText=@n1)
  ]}
 SetLockedData {nodes [
   SetLockedData() {is=title K2Node_FunctionEntry}
   > SetIcon!(execute, self, Icon:Material Instance Soft Object Reference=/FortUILibrary/Textures/MaterialInstances/MI_UIKit_Icon_Locked) -> (then) {is=title K2Node_CallFunction}
   n1 = Get QuestUILockedQuestVM()
   > n2 = Switch on ELockedWidgetUnlockType!(
    execute
    Selection:ELockedWidgetUnlockType Enum=Break Locked Quest Runtime Data!(LockedQuestRuntimeData=Get LockedQuestData(self=@n1)) -> (UnlockType:ELockedWidgetUnlockType Enum)
   ) -> (
    QuestDaysFromEventStart:Exec
    QuestChallengesCompleteUntilUnlocked:Exec
    BundleDaysFromEventStart:Exec
    PrerequisiteQuest:Exec
    CalendarEventEnd:Exec
   ) {is=title K2Node_SwitchEnum}
   n3 = Get QuestUILockedQuestVM()
   n4 = Timespan - Timespan(
    A:Timespan Structure=Break Locked Quest Runtime Data!(LockedQuestRuntimeData=Get LockedQuestData(self=@n3)) -> (UnlockTime:Timespan Structure)
    B:Timespan Structure=GetTimeSinceEventStart!(self:Quest UI Locked Quest View Model Object Reference=@n3) -> (ReturnValue:Timespan Structure)
   ) -> (ReturnValue:Timespan Structure) {is=title K2Node_PromotableOperator}
   n5 = "Timespan <= Timespan"(A:Timespan Structure=@n4, B:Timespan Structure=Get Timespan() -> (Timespan:Timespan Structure)) -> (ReturnValue:bool=~) {is=title K2Node_PromotableOperator}
   n6 = Get Zero() -> (Zero:Timespan Structure)
   n7 = Select(
    Option 0:"Timespan Structure (by ref)"=@n4
    Option 1:"Timespan Structure (by ref)"=@n6
    Index="Timespan <= Timespan"(A=@n4, B=@n6) -> (ReturnValue=~)
   ) -> (ReturnValue:"Timespan Structure (by ref)")
   SetTitleText!(
    execute=[@n2.BundleDaysFromEventStart, @n2.CalendarEventEnd, @n2.QuestDaysFromEventStart]
    self
    Text:Text=Format Text!(
     Format=GetFinalDisplayText!(DefaultText:Text="LOCTEXT(\"Additional Quests available in {Time}\")") -> (DisplayText:Text) {is=title K2Node_CallFunction}
     DisplayName:Text=GetFinalCategoryName!() -> (CategoryName:Text) {is=title K2Node_CallFunction}
     Time:Text=GetTimespanAsText!(
      Timespan:"Timespan Structure (by ref)"=@n7
      bShowHoursWhenZero:bool=true
      bShowSeconds:bool=@n5
      FormatType:ETimespanAsTextFormat Enum=DaysHoursMinutesSeconds
      UnitDisplayType:ETimespanUnitDisplayFormat Enum=Abbreviated
      bUseSentenceCase:bool=false
     ) -> (ReturnValue:Text)
    ) -> (Result)
   ) -> (then)
   > SetIcon(Icon=/FortUILibrary/Textures/MaterialInstances/MI_UIKit_Icon_UpdatePending)
   n8 = Break Locked Quest Runtime Data!(LockedQuestRuntimeData=Get LockedQuestData(self=Get QuestUILockedQuestVM())) -> (UnlockNumber:int)
   n9 = "To Integer64 (Integer)"!(InInt:int=@n8) -> (ReturnValue:Integer64=0)
   n10 = "integer > integer"(A=@n8, B=1)
   SetTitleText(
    execute=@n2.QuestChallengesCompleteUntilUnlocked
    Text=Format Text!(
     Format=GetFinalDisplayText(DefaultText="LOCTEXT(\"Complete {Number} more {Quest} to advance.\")")
     Number:int="To Integer (Integer64)"!(InInt:Integer64=@n9) -> (ReturnValue:int=0)
     DisplayName=GetFinalCategoryName()
     Quest:Text=Select(Option 0:Text="LOCTEXT(\"Quest\")", Option 1:Text="LOCTEXT(\"Quests\")", Index=@n10) -> (ReturnValue:Text)
    ) -> (Result)
   )
   n11 = Get QuestUILockedQuestVM()
   SetTitleText(
    execute=@n2.PrerequisiteQuest
    Text=Format Text!(
     Format=GetFinalDisplayText(DefaultText="LOCTEXT(\"Complete {PrereqQuest} to advance.\")")
     PrereqQuest:Text=GetSourceDescription!(self:Quest UI Locked Quest View Model Object Reference=@n11) -> (ReturnValue:Text)
    ) -> (Result)
   )
  ]}
 SetMessageType {nodes [
   n1 = SetMessageType!() -> (then, MessageType:EQuestUIMessageType Enum=Custom) {is=title K2Node_FunctionEntry}
   > n2 = Switch on EQuestUIMessageType!(execute, Selection:EQuestUIMessageType Enum=@n1) -> (Custom:Exec, PartyUp:Exec, AllComplete:Exec, NoTrackedQuests:Exec) {is=title K2Node_SwitchEnum}
   SetTitleText(execute=@n2.AllComplete, Text="LOCTEXT(\"All quests completed!\")")
   SetTitleText(execute=@n2.PartyUp, Text="LOCTEXT(\"Party up with friends to share Daily Quests! \")")
  ]}
 SetTitleText {nodes [
   n1 = SetTitleText!() -> (then, Text:Text) {is=title K2Node_FunctionEntry}
   > SetText(self=Get RichTitle() -> (RichTitle:Common Rich Text Block Object Reference), InText=@n1)
   > "SetText (Text)"(self=Get Title() -> (Title:Common Text Object Reference), InText=@n1)
  ]}
 UpdateStyle {nodes [
   UpdateStyle!() -> (then, Style:E_Seperator Enum=NewEnumerator4) {is=title K2Node_FunctionEntry}
   n1 = Get MessageFont() -> (MessageFont:Slate Font Info Structure)
   n2 = Get Title()
   > SetFont(
    self=@n2
    InFontInfo=Select!(
     NewEnumerator4:Slate Font Info Structure=@n1
     NewEnumerator0:Slate Font Info Structure=Get Header Font() -> (Header Font:Slate Font Info Structure)
     NewEnumerator1:Slate Font Info Structure=@n1
     NewEnumerator2:Slate Font Info Structure=Get SeparatorTitleFont() -> (SeparatorTitleFont:Slate Font Info Structure)
     NewEnumerator3:Slate Font Info Structure=@n1
     Index:E_Seperator Enum=Get Style() -> (Style:E_Seperator Enum=NewEnumerator4)
    ) -> (ReturnValue:Slate Font Info Structure)
   )
   n3 = Get Style() -> (Style=NewEnumerator4)
   > SetTextCase!(
    execute
    self:Common Text Object Reference=@n2
    bUseAllCaps:bool=OR Boolean(A="Equal (Enum)"(A:E_Seperator Enum=@n3, B:E_Seperator Enum=NewEnumerator2), B="Equal (Enum)"(A=@n3, B=NewEnumerator0))
   ) -> (then)
   n4 = Get Style() -> (Style=NewEnumerator4)
   > SetDesiredSizeOverride!(
    execute
    self:Image Object Reference=Get Icon()
    DesiredSize:Vector 2D Structure=Select!(Option 0:Vector 2D Structure=(X=20,Y=20), Option 1:Vector 2D Structure=(X=48,Y=48), Index:bool="Equal (Enum)"(A=@n4, B=NewEnumerator0)) -> (ReturnValue:Vector 2D Structure)
   ) -> (then)
   n5 = Get Text Color Diabled() -> (Text Color Diabled:Slate Color Structure)
   n6 = Get Text Color White() -> (Text Color White:Slate Color Structure)
   > SetColorAndOpacity(
    self=Get Title()
    InColorAndOpacity=Select!(
     NewEnumerator4:Slate Color Structure=@n5
     NewEnumerator0:Slate Color Structure=@n6
     NewEnumerator1:Slate Color Structure=@n5
     NewEnumerator2:Slate Color Structure=@n5
     NewEnumerator3:Slate Color Structure=@n6
     Index:E_Seperator Enum=@n4
    ) -> (ReturnValue:Slate Color Structure)
   )
   n7 = Get Icon Color Disabled() -> (Icon Color Disabled:Linear Color Structure)
   n8 = Get Icon Color White() -> (Icon Color White:Linear Color Structure)
   > Set ColorAndOpacity!(
    execute
    ColorAndOpacity:Linear Color Structure=Select(
     NewEnumerator4:Linear Color Structure=@n7
     NewEnumerator0:Linear Color Structure=@n8
     NewEnumerator1:Linear Color Structure=@n7
     NewEnumerator2:Linear Color Structure=@n7
     NewEnumerator3:Linear Color Structure=@n8
     Index=Get Style() -> (Style=NewEnumerator4)
    ) -> (ReturnValue:Linear Color Structure)
    self:Image Object Reference=Get Icon()
   ) -> (then, Output_Get:Linear Color Structure)
   n9 = Get Style() -> (Style=NewEnumerator4)
   n10 = "Equal (Enum)"(A=@n9, B=NewEnumerator2)
   > SetVisibility(
    self=[Get RightLine() -> (RightLine:Image Object Reference), Get LeftLine() -> (LeftLine:Image Object Reference)]
    InVisibility=Select!(Option 0:ESlateVisibility Enum=Collapsed, Option 1:ESlateVisibility Enum=SelfHitTestInvisible, Index:bool=@n10) -> (ReturnValue:ESlateVisibility Enum)
   )
   n11 = SelectInt(A=490, B=SelectInt(A=680, B=820, bPickA=IsUsingCompactUI()), bPickA=@n10)
   > SetWrapTextWidth!(execute, self:Common Text Object Reference=Get Title(), InWrapTextAt:int=@n11) -> (then)
   > Set WrapTextAt!(execute, WrapTextAt:float="To Float (Integer)"(InInt=@n11), self:Text Layout Widget Object Reference=Get RichTitle()) -> (then, Output_Get:float)
   n12 = "Equal (Enum)"(A=@n9, B=NewEnumerator1)
   n13 = NOT Boolean(A=@n12)
   > SetVisibility(self=Get Icon(), InVisibility=Select(Option 0=SelfHitTestInvisible, Option 1=Collapsed, Index=AND Boolean(A=@n10, B=@n13)))
   > SetVisibility(self=Get Spacer() -> (Spacer:Spacer Object Reference), InVisibility=Select(Option 0=Collapsed, Option 1=SelfHitTestInvisible, Index=@n12))
   > SetActiveWidget(
    self=Get TitleSwitcher() -> (TitleSwitcher:Common Visibility Switcher Object Reference)
    Widget=Select(Option 0:Widget Object Reference=Get Title(), Option 1:Widget Object Reference=Get RichTitle(), Index=@n12) -> (ReturnValue:Widget Object Reference)
   )
   > n14 = Switch on E_Seperator!(execute, Selection:E_Seperator Enum=Get Style() -> (Style=NewEnumerator4)) -> (NewEnumerator4:Exec, NewEnumerator0:Exec, NewEnumerator1:Exec, NewEnumerator2:Exec, NewEnumerator3:Exec) {is=title K2Node_SwitchEnum}
   SetLockedData(execute=@n14.NewEnumerator1) {is=title K2Node_CallFunction}
   SetPadding!(
    execute=@n14.NewEnumerator2
    self:User Widget Object Reference
    InPadding:Margin Structure=Get NotAvailablePadding() -> (NotAvailablePadding:Margin Structure)
   ) -> (then)
   > SetHorizontalAlignment(
    self:Horizontal Box Slot Object Reference=SlotAsHorizontalBoxSlot!(Widget:Widget Object Reference=Get TitleSwitcher()) -> (ReturnValue:Horizontal Box Slot Object Reference)
    InHorizontalAlignment=HAlign_Center
   )
   n15 = "Equal (Enum)"(
    A:EQuestUIMessageType Enum=Get MessageType!(self:Quest UI Message View Model Object Reference=Get QuestUIMessageVM() -> (QuestUIMessageVM:Quest UI Message View Model Object Reference)) -> (MessageType:EQuestUIMessageType Enum=Custom)
    B:EQuestUIMessageType Enum=PartyUp
   )
   SetVisibility(
    execute=@n14.NewEnumerator3
    self=Get Icon()
    InVisibility=Select(Option 0:ESlateVisibility Enum=Collapsed, Option 1:ESlateVisibility Enum=SelfHitTestInvisible, Index=@n15) -> (ReturnValue:ESlateVisibility Enum)
   )
   > Branch(Condition=@n15)
   > SetIcon(Icon=/FortUILibrary/Textures/MaterialInstances/MI_UIKit_Icon_PartyUp)
   n16 = Branch(Condition=Get IsCompatibility() -> (IsCompatibility:bool))
   > n17 = SetPadding(InPadding=Get NotAvailablePadding())
   SetHorizontalAlignment(
    execute=[@n17.then, @n16.else]
    self=SlotAsHorizontalBoxSlot(Widget=Get TitleSwitcher())
    InHorizontalAlignment=Select(Option 0:EHorizontalAlignment Enum=HAlign_Left, Option 1:EHorizontalAlignment Enum=HAlign_Center, Index=Get IsCompatibility()) -> (ReturnValue:EHorizontalAlignment Enum)
   )
  ]}
]
variables [
 NotAvailablePadding {type=Margin, default=<ltrb 0 64 0 0>}
 Timespan {type=Timespan, default=+00000000.00:01:00.000000000}
]
widgets {
 info {inheritedWidgetCount=0, parentClass=&CommonUserWidget, rootWidgetClass=&VerticalBox, widgetCount=16}
 widgets [
  Root {widgetClassPath=&VerticalBox, widget=Root, properties {bIsEnabled=false, bOverride_Cursor=true}}
  Overlay_0 {widgetClassPath=&Overlay, parent=Root, widget=Overlay_0, slot {class=VerticalBoxSlot, properties {padding=<ltrb 0 10 0 0>}}}
  HorizontalBox_0 {
   widgetClassPath=&HorizontalBox
   parent=Overlay_0
   widget=HorizontalBox_0
   slot {class=OverlaySlot, properties {horizontalAlignment=HAlign_Fill, verticalAlignment=VAlign_Fill, padding=<ltrb 4 4 4 4>}}
  }
  RightLine {
   widgetClassPath=&Image
   bIsVariable=true
   parent=HorizontalBox_0
   widget=RightLine
   properties {brush {
     bIsDynamicallyLoaded=false
     drawAs=Image
     imageType=NoImage
     mirroring=NoMirror
     resourceName=None
     resourceObject=null
     tiling=NoTile
     imageSize=<32 2>
     margin=<ltrb 0 0 0 0>
     outlineSettings {
      bUseBrushTransparency=false
      roundingType=HalfHeightRadius
      width=0
      color {colorUseRule=UseColor_Specified, specifiedColor=<rgba 0 0 0 0>}
      cornerRadii=<0 0 0 0>
     }
     tintColor {colorUseRule=UseColor_Specified, specifiedColor=<rgba 1 1 1 0.5>}
     uVRegion {bIsValid=false, max=<0 0>, min=<0 0>}
    }}
   slot {class=HorizontalBoxSlot, properties {verticalAlignment=VAlign_Center, padding=<ltrb 0 0 3 0>, size {sizeRule=Fill, value=0.5}}}
  }
  ContentArea {widgetClassPath=&Overlay, parent=HorizontalBox_0, widget=ContentArea, slot {class=HorizontalBoxSlot}}
  Content {
   widgetClassPath=&GridPanel
   bIsVariable=true
   parent=ContentArea
   widget=Content
   properties {columnFill [0,1], rowFill [1]}
   slot {class=OverlaySlot, properties {horizontalAlignment=HAlign_Center, verticalAlignment=VAlign_Center, padding=<ltrb 4 1 4 1>}}
  }
  Icon {
   widgetClassPath=&Image
   bIsVariable=true
   parent=Content
   widget=Icon
   properties {
    visibility=SelfHitTestInvisible
    brush {
     bIsDynamicallyLoaded=false
     drawAs=Image
     imageType=FullColor
     mirroring=NoMirror
     resourceName=None
     resourceObject=/FortUILibrary/Textures/Material/MI_UIKit_Icon_Base
     tiling=NoTile
     imageSize=<20 20>
     margin=<ltrb 0 0 0 0>
     outlineSettings {
      bUseBrushTransparency=false
      roundingType=HalfHeightRadius
      width=0
      color {colorUseRule=UseColor_Specified, specifiedColor=<rgba 0 0 0 0>}
      cornerRadii=<0 0 0 0>
     }
     tintColor {colorUseRule=UseColor_Specified, specifiedColor=<rgba 1 1 1 1>}
     uVRegion {bIsValid=false, max=<0 0>, min=<0 0>}
    }
    colorAndOpacity=<rgba 0.4 0.4 0.4 1>
   }
   slot {class=GridSlot, properties {horizontalAlignment=HAlign_Left, verticalAlignment=VAlign_Center}}
  }
  Spacer {
   widgetClassPath=&Spacer
   bIsVariable=true
   parent=HorizontalBox_0
   widget=Spacer
   properties {visibility=Collapsed, size=<6 1>}
   slot {class=HorizontalBoxSlot}
  }
  TitleSwitcher {
   widgetClassPath=&CommonVisibilitySwitcher
   bIsVariable=true
   parent=HorizontalBox_0
   widget=TitleSwitcher
   slot {class=HorizontalBoxSlot, properties {verticalAlignment=VAlign_Center, padding=<ltrb 0 0 8 0>}}
  }
  Title {
   widgetClassPath=&CommonTextBlock
   bIsVariable=true
   parent=TitleSwitcher
   widget=Title
   properties {
    bIsEnabled=false
    bIsScrollingEnabled=false
    mobileFontSizeMultiplier=1.3
    scrollStyle=&TextScrollStyle-Base_C
    text=Additional Quests available in 0s
    wrapTextAt=490
    colorAndOpacity {colorUseRule=UseColor_Specified, specifiedColor=<rgba 0.4 0.4 0.4 1>}
    font {
     bForceMonospaced=false
     bMaterialIsStencil=false
     fontMaterial=null
     fontObject=&HeadingNow
     letterSpacing=0
     monospacedWidth=1
     size=13.5
     skewAmount=0
     typefaceFontName=74 Regular
     outlineSettings {
      bApplyOutlineToDropShadows=false
      bMiteredCorners=false
      bSeparateFillAlpha=false
      outlineMaterial=null
      outlineSize=0
      outlineColor=<rgba 0 0 0 1>
     }
    }
    shadowOffset=<0 0>
   }
   slot {class=CommonVisibilitySwitcherSlot}
  }
  RichTitle {
   widgetClassPath=&CommonRichTextBlock
   bIsVariable=true
   parent=TitleSwitcher
   widget=RichTitle
   properties {
    bIsScrollingEnabled=false
    bOverrideDefaultStyle=true
    mobileTextBlockScale=1.3
    text="Additional \\nQuests available in 0s"
    textStyleSet=/Game/Athena/HUD/QuestIcons/DT_RichText_QuestTileSeperator
    wrapTextAt=490
    decoratorClasses [/Game/Athena/HUD/QuestIcons/BP_QuestTile_IconDecorator.BP_QuestTile_IconDecorator_C]
    defaultTextStyleOverride {
     overflowPolicy=Clip
     transformPolicy=None
     colorAndOpacity {colorUseRule=UseColor_Specified, specifiedColor=<rgba 0.4 0.4 0.4 1>}
     font {
      bForceMonospaced=false
      bMaterialIsStencil=false
      fontMaterial=null
      fontObject=&HeadingNow
      letterSpacing=0
      monospacedWidth=1
      size=13.5
      skewAmount=0
      typefaceFontName=74 Regular
      outlineSettings {
       bApplyOutlineToDropShadows=false
       bMiteredCorners=false
       bSeparateFillAlpha=false
       outlineMaterial=null
       outlineSize=0
       outlineColor=<rgba 0 0 0 1>
      }
     }
     highlightColor {colorUseRule=UseColor_Specified, specifiedColor=<rgba 0 0 0 1>}
     highlightShape {
      bIsDynamicallyLoaded=false
      drawAs=Image
      imageType=NoImage
      mirroring=NoMirror
      resourceName=None
      resourceObject=null
      tiling=NoTile
      imageSize=<32 32>
      margin=<ltrb 0 0 0 0>
      outlineSettings {
       bUseBrushTransparency=false
       roundingType=HalfHeightRadius
       width=0
       color {colorUseRule=UseColor_Specified, specifiedColor=<rgba 0 0 0 0>}
       cornerRadii=<0 0 0 0>
      }
      tintColor {colorUseRule=UseColor_Specified, specifiedColor=<rgba 1 1 1 1>}
      uVRegion {bIsValid=false, max=<0 0>, min=<0 0>}
     }
     selectedBackgroundColor {colorUseRule=UseColor_Foreground, specifiedColor=<rgba 1 0 1 1>}
     shadowColorAndOpacity=<rgba 0 0 0 0>
     shadowOffset=<0 0>
     strikeBrush {
      bIsDynamicallyLoaded=false
      drawAs=Image
      imageType=NoImage
      mirroring=NoMirror
      resourceName=None
      resourceObject=null
      tiling=NoTile
      imageSize=<32 32>
      margin=<ltrb 0 0 0 0>
      outlineSettings {
       bUseBrushTransparency=false
       roundingType=HalfHeightRadius
       width=0
       color {colorUseRule=UseColor_Specified, specifiedColor=<rgba 0 0 0 0>}
       cornerRadii=<0 0 0 0>
      }
      tintColor {colorUseRule=UseColor_Specified, specifiedColor=<rgba 1 1 1 1>}
      uVRegion {bIsValid=false, max=<0 0>, min=<0 0>}
     }
     underlineBrush {
      bIsDynamicallyLoaded=false
      drawAs=Image
      imageType=NoImage
      mirroring=NoMirror
      resourceName=None
      resourceObject=null
      tiling=NoTile
      imageSize=<32 32>
      margin=<ltrb 0 0 0 0>
      outlineSettings {
       bUseBrushTransparency=false
       roundingType=HalfHeightRadius
       width=0
       color {colorUseRule=UseColor_Specified, specifiedColor=<rgba 0 0 0 0>}
       cornerRadii=<0 0 0 0>
      }
      tintColor {colorUseRule=UseColor_Specified, specifiedColor=<rgba 1 1 1 1>}
      uVRegion {bIsValid=false, max=<0 0>, min=<0 0>}
     }
    }
   }
   slot {class=CommonVisibilitySwitcherSlot, properties {verticalAlignment=VAlign_Top}}
  }
  LeftLine {
   widgetClassPath=&Image
   bIsVariable=true
   parent=HorizontalBox_0
   widget=LeftLine
   properties {brush {
     bIsDynamicallyLoaded=false
     drawAs=Image
     imageType=NoImage
     mirroring=NoMirror
     resourceName=None
     resourceObject=null
     tiling=NoTile
     imageSize=<32 2>
     margin=<ltrb 0 0 0 0>
     outlineSettings {
      bUseBrushTransparency=false
      roundingType=HalfHeightRadius
      width=0
      color {colorUseRule=UseColor_Specified, specifiedColor=<rgba 0 0 0 0>}
      cornerRadii=<0 0 0 0>
     }
     tintColor {colorUseRule=UseColor_Specified, specifiedColor=<rgba 1 1 1 0.5>}
     uVRegion {bIsValid=false, max=<0 0>, min=<0 0>}
    }}
   slot {class=HorizontalBoxSlot, properties {verticalAlignment=VAlign_Center, padding=<ltrb 2 0 0 0>, size {sizeRule=Fill, value=0.5}}}
  }
  SelectedStateOutline {
   widgetClassPath=&Image
   bIsVariable=true
   parent=Overlay_0
   widget=SelectedStateOutline
   properties {
    visibility=Collapsed
    brush {
     bIsDynamicallyLoaded=false
     drawAs=Image
     imageType=FullColor
     mirroring=NoMirror
     resourceName=None
     resourceObject=/QuestUI/QuestScreen/QuestTile/Materials/MI_UI_Outline_Selected
     tiling=NoTile
     imageSize=<32 32>
     margin=<ltrb 0 0 0 0>
     outlineSettings {
      bUseBrushTransparency=false
      roundingType=HalfHeightRadius
      width=0
      color {colorUseRule=UseColor_Specified, specifiedColor=<rgba 0 0 0 0>}
      cornerRadii=<0 0 0 0>
     }
     tintColor {colorUseRule=UseColor_Specified, specifiedColor=<rgba 1 1 1 1>}
     uVRegion {bIsValid=false, max=<0 0>, min=<0 0>}
    }
   }
   slot {class=OverlaySlot, properties {horizontalAlignment=HAlign_Fill, verticalAlignment=VAlign_Fill}}
  }
  HorizontalBox_1 {widgetClassPath=&HorizontalBox, parent=Root, widget=HorizontalBox_1, slot {class=VerticalBoxSlot}}
  DebugFieldLeft {
   widgetClassPath=&CommonTextBlock
   bIsVariable=true
   parent=HorizontalBox_1
   widget=DebugFieldLeft
   properties {
    bAutoCollapseWithEmptyText=true
    scrollStyle=&TextScrollStyle-Base_C
    text=Debug Left
    colorAndOpacity {colorUseRule=UseColor_Specified, specifiedColor=<rgba 0.5776 0.5776 0.5776 1>}
    font {
     bForceMonospaced=false
     bMaterialIsStencil=false
     fontMaterial=null
     fontObject=&HeadingNow
     letterSpacing=0
     monospacedWidth=1
     size=10.5
     skewAmount=0
     typefaceFontName=74 Regular
     outlineSettings {
      bApplyOutlineToDropShadows=false
      bMiteredCorners=false
      bSeparateFillAlpha=false
      outlineMaterial=null
      outlineSize=0
      outlineColor=<rgba 0 0 0 1>
     }
    }
    shadowOffset=<0 0>
   }
   slot {class=HorizontalBoxSlot, properties {size {sizeRule=Fill, value=1}}}
  }
  DebugFieldRight {
   widgetClassPath=&CommonTextBlock
   bIsVariable=true
   parent=HorizontalBox_1
   widget=DebugFieldRight
   properties {
    bAutoCollapseWithEmptyText=true
    justification=Right
    scrollStyle=&TextScrollStyle-Base_C
    text=Debug Right
    colorAndOpacity {colorUseRule=UseColor_Specified, specifiedColor=<rgba 0.5776 0.5776 0.5776 1>}
    font {
     bForceMonospaced=false
     bMaterialIsStencil=false
     fontMaterial=null
     fontObject=&HeadingNow
     letterSpacing=0
     monospacedWidth=1
     size=10.5
     skewAmount=0
     typefaceFontName=74 Regular
     outlineSettings {
      bApplyOutlineToDropShadows=false
      bMiteredCorners=false
      bSeparateFillAlpha=false
      outlineMaterial=null
      outlineSize=0
      outlineColor=<rgba 0 0 0 1>
     }
    }
    shadowOffset=<0 0>
   }
   slot {class=HorizontalBoxSlot, properties {size {sizeRule=Fill, value=1}}}
  }
 ]
}
