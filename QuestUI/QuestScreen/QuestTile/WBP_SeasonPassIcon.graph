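# Quest tile badge widget that shows a season pass's small icon over a pink circle.
# SetSeasonItem casts an allowed, valid season item's definition to SeasonPassItemDefinition and passes its soft PassData to LoadPassData; otherwise it collapses the widget.
# LoadPassData async-loads the data, rechecks IsBattlePass on completion so a later rejected item stays hidden, then sets the lazy image and makes the widget visible.
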WidgetBlueprint /QuestUI/QuestScreen/QuestTile/WBP_SeasonPassIcon
graphs [
 EventGraph {nodes [
   n1 = LoadPassData!() -> (OutputDelegate, then, PassDataSoftObject:Season Pass Data Base Soft Object Reference) {is=title K2Node_CustomEvent}
   > n2 = Async Load Asset(Asset=@n1.PassDataSoftObject)
   Branch(execute=@n2.Completed, Condition=Get IsBattlePass() -> (IsBattlePass:bool))
   > n3 = Cast To SeasonPassData!(execute, Object=@n2) -> (then, CastFailed, AsSeason Pass Data:Season Pass Data Object Reference) {is=title K2Node_DynamicCast}
   > SetBrushFromLazyDisplayAsset!(
    execute
    self:Common Lazy Image Object Reference=Get Icon() -> (Icon:Fort Lazy Image Object Reference)
    LazyObject:"Soft Object Reference (by ref)"=Get SmallIcon!(self:Season Pass Data Object Reference=@n3) -> (SmallIcon:Soft Object Reference)
    bMatchTextureSize:bool=true
   ) -> (then)
   > SetVisibility(InVisibility=HitTestInvisible)
  ]}
 SetSeasonItem {nodes [
   n1 = SetSeasonItem!() -> (then, SeasonItem:Fort Season Item Object Reference, IsAllowed:bool=~) {is=title K2Node_FunctionEntry}
   > Set IsBattlePass(IsBattlePass:bool=false) -> (Output_Get:bool)
   > n2 = Is Valid(InputObject=@n1)
   > n3 = Branch(Condition=@n1.IsAllowed)
   > n4 = Cast To SeasonPassItemDefinition!(execute, Object=GetItemDefinition!(self:Fort Item Object Reference=@n1) -> (ReturnValue:Fort Item Definition Object Reference)) -> (then, CastFailed, AsSeason Pass Item Definition:Season Pass Item Definition Object Reference) {is=title K2Node_DynamicCast}
   > Set IsBattlePass(IsBattlePass=true)
   > LoadPassData!(
    execute
    self
    PassDataSoftObject:Season Pass Data Base Soft Object Reference=Get PassData!(self:Season Pass Item Definition Object Reference=@n4) -> (PassData:Season Pass Data Base Soft Object Reference)
   ) -> (then) {is=title K2Node_CallFunction}
   Set IsBattlePass(execute=[@n4.CastFailed, @n3.else, @n2.Is Not Valid], IsBattlePass=false)
   > SetVisibility(InVisibility=Collapsed)
  ]}
]
variables [IsBattlePass {type=boolean}]
widgets {
 info {inheritedWidgetCount=0, parentClass=&UserWidget, rootWidgetClass=&Overlay, widgetCount=3}
 widgets [
  Overlay_29 {widgetClassPath=&Overlay, widget=Overlay_29}
  BG {
   widgetClassPath=&Image
   parent=Overlay_29
   widget=BG
   properties {
    brush {
     bIsDynamicallyLoaded=false
     drawAs=Image
     imageType=FullColor
     mirroring=NoMirror
     resourceName=None
     resourceObject=/LauncherSocial/circle
     tiling=NoTile
     imageSize=<24 24>
     margin=<ltrb 0 0 0 0>
     outlineSettings {
      bUseBrushTransparency=false
      roundingType=HalfHeightRadius
      width=0
      color {colorUseRule=UseColor_Specified, specifiedColor=<rgba 0 0 0 0>}
      cornerRadii=<0 0 0 0>
     }
     tintColor {colorUseRule=UseColor_Specified, specifiedColor=<rgba 1 1 1 1>}
     uVRegion {bIsValid=false, max=<0 0>, min=<0 0>}
    }
    colorAndOpacity=<rgba 0.807 0.0203 0.3419 1>
   }
   slot {class=OverlaySlot, properties {horizontalAlignment=HAlign_Fill, verticalAlignment=VAlign_Fill}}
  }
  Icon {
   widgetClassPath=&FortLazyImage
   bIsVariable=true
   parent=Overlay_29
   widget=Icon
   properties {brush {
     bIsDynamicallyLoaded=false
     drawAs=Image
     imageType=NoImage
     mirroring=NoMirror
     resourceName=None
     resourceObject=null
     tiling=NoTile
     imageSize=<20 20>
     margin=<ltrb 0 0 0 0>
     outlineSettings {
      bUseBrushTransparency=false
      roundingType=HalfHeightRadius
      width=0
      color {colorUseRule=UseColor_Specified, specifiedColor=<rgba 0 0 0 0>}
      cornerRadii=<0 0 0 0>
     }
     tintColor {colorUseRule=UseColor_Specified, specifiedColor=<rgba 1 1 1 1>}
     uVRegion {bIsValid=false, max=<0 0>, min=<0 0>}
    }}
   slot {class=OverlaySlot, properties {horizontalAlignment=HAlign_Center, verticalAlignment=VAlign_Center}}
  }
 ]
}
