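# Quest reward tile widget showing a reward icon, quantity text, season pass badge and page indicator.
# SetRewardVMs and SetCategoryRewardVMs keep rewards eligible for small display; category mode shows one, preferring cinematic rewards.
# A looping timer is started for multiple rewards, and UpdateShownReward advances the shown reward and page.
# Cinematic icons vary by availability and watched state; account-resource quantities other than one display as text, abbreviated in thousands.
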WidgetBlueprint /QuestUI/QuestScreen/QuestTile/WBP_QuestReward
graphs [
 DisplayCinematicRewardVM {nodes [
   n1 = DisplayCinematicRewardVM!() -> (then, CinematicRewardVM:Quest UI Reward Cinematic VM Object Reference) {is=title K2Node_FunctionEntry}
   > Set QuestUIRewardVM(QuestUIRewardVM:Quest UI Reward VM Object Reference=@n1) -> (Output_Get:Quest UI Reward VM Object Reference)
   n2 = Get CinematicRewardVM() -> (CinematicRewardVM:Quest UI Reward Cinematic VM Object Reference)
   n3 = Get bIsWatched!(self:Quest UI Reward Cinematic VM Object Reference=@n2) -> (bIsWatched:bool)
   n4 = Get CinematicRewardIcon_VFX() -> (CinematicRewardIcon_VFX:Material Interface Soft Object Reference)
   n5 = Get CinematicRewardIcon() -> (CinematicRewardIcon:Material Interface Soft Object Reference)
   > UpdateIcon!(
    execute
    self
    LazyObject:"Soft Object Reference (by ref)"=Select(
     Option 0:Material Interface Soft Object Reference=Get CinematicRewardIcon()
     Option 1:Material Interface Soft Object Reference=Select(Option 0=@n4, Option 1=@n5, Index=@n3)
     Index=Get bIsRewardContentAvailable!(self:Quest UI Reward VM Object Reference=@n2) -> (bIsRewardContentAvailable:bool)
    ) -> (ReturnValue:Material Interface Soft Object Reference)
   ) -> (then)
   > SetText!(execute, self:Self Object Reference, Text:Text) -> (then)
  ]}
 DisplayContainer {nodes [
   n1 = DisplayContainer!() -> (then, NumRewards:int) {is=title K2Node_FunctionEntry}
   > SetVisibility(InVisibility=Select(Option 0:ESlateVisibility Enum=SelfHitTestInvisible, Option 1:ESlateVisibility Enum=Collapsed, Index="Equal (Integer)"(A=@n1)) -> (ReturnValue:ESlateVisibility Enum))
  ]}
 DisplayContentEnabled {nodes [
   DisplayContentEnabled() {is=title K2Node_FunctionEntry}
   > SetOpacity!(
    execute
    self:Image Object Reference=Get LazyIcon() -> (LazyIcon:Fort Lazy Image Object Reference)
    InOpacity:float=Select(Option 0:float=0.5, Option 1:float=1, Index=Get Is Content Enabled() -> (Is Content Enabled:bool)) -> (ReturnValue:float)
   ) -> (then)
   > SetColorAndOpacity(
    self:User Widget Object Reference=Get Block_Text() -> (Block_Text:WBP UIKit Block Text Object Reference)
    InColorAndOpacity:Linear Color Structure=Select(
     Option 0:Linear Color Structure=(R=0.6939,G=0.6939,B=0.6939,A=0.4)
     Option 1:Linear Color Structure=(R=1,G=1,B=1,A=1)
     Index=Get Is Content Enabled()
    ) -> (ReturnValue:Linear Color Structure)
   )
  ]}
 DisplayPages {nodes [
   n1 = DisplayPages!() -> (then, NumRewards:int) {is=title K2Node_FunctionEntry}
   n2 = "integer > integer"(A=@n1, B=1)
   > SetVisibility(
    self=Get Pages() -> (Pages:WBP Quest Pages Object Reference)
    InVisibility=Select(Option 0:ESlateVisibility Enum=Collapsed, Option 1:ESlateVisibility Enum=SelfHitTestInvisible, Index=@n2) -> (ReturnValue:ESlateVisibility Enum)
   )
   > SetOptionsAmount!(execute, self:WBP Quest Pages Object Reference=Get Pages(), PagesAmount:int=Get NumRewards() -> (NumRewards:int=0)) -> (then)
   > Branch(Condition=@n2)
   > n3 = Set Timer by Event(Delegate=Create Event(), Time=2, bLooping=true)
   > Set TimerHandle(TimerHandle:Timer Handle Structure=@n3) -> (Output_Get:Timer Handle Structure)
  ]}
 DisplayRewardVM {nodes [
   n1 = DisplayRewardVM!() -> (then, RewardVM:Quest UI Reward VM Object Reference) {is=title K2Node_FunctionEntry}
   > n2 = Set QuestUIRewardVM(QuestUIRewardVM=@n1)
   > SetSeasonItem!(
    execute
    self:WBP Season Pass Icon Object Reference=Get SeasonPassIcon() -> (SeasonPassIcon:WBP Season Pass Icon Object Reference)
    SeasonItem:Fort Season Item Object Reference=GetRequiredSeasonItem!(self:Quest UI Reward VM Object Reference=@n2) -> (ReturnValue:Fort Season Item Object Reference)
    IsAllowed:bool=NOT Boolean(A=Get bIsRewardContentAvailable(self=@n2))
   ) -> (then)
   > Call On Reward Updated!(execute:Exec, self:WBP Quest Reward Object Reference, RewardVM:Quest UI Reward VM Object Reference=@n2) -> (then:Exec) {is=title K2Node_CallDelegate}
  ]}
 DisplayRewardVMs {nodes [
   DisplayRewardVMs() {is=title K2Node_FunctionEntry}
   n1 = Length(TargetArray:Array of Quest UI Reward VM Object References=Get CustomWidgetRewards() -> (CustomWidgetRewards:Array of Quest UI Reward VM Object References))
   > DisplayContainer!(execute, self, NumRewards:int=@n1) -> (then) {is=title K2Node_CallFunction}
   > DisplayPages!(execute, self, NumRewards:int=Select(Option 0:int=@n1, Option 1:int=1, Index=Get IsCategory() -> (IsCategory:bool)) -> (ReturnValue:int)) -> (then) {is=title K2Node_CallFunction}
   > n2 = For Each Loop with Break!(Exec:Exec, Array:Array of Quest UI Reward VM Object References=Get CustomWidgetRewards(), Break:Exec=@n7.then) -> (LoopBody:Exec, Array Element:Quest UI Reward VM Object Reference, Array Index:int, Completed:Exec)
   n3 = Branch(execute=@n2.Completed, Condition=NOT Boolean(A=Get FoundCinematicReward() -> (FoundCinematicReward:bool)))
   n4 = Get CustomWidgetRewards()
   > n5 = Branch(Condition="integer > integer"(A=Length(TargetArray=@n4)))
   > DisplayRewardVM!(
    execute
    self
    RewardVM:Quest UI Reward VM Object Reference="Get (a copy)"(Array:Array of Quest UI Reward VM Object References=Get CustomWidgetRewards()) -> (Output:Quest UI Reward VM Object Reference)
   ) -> (then) {is=title K2Node_CallFunction}
   > Return Node!(execute, FoundCinematic:bool=Get FoundCinematicReward())
   Return Node(execute=@n5.else, FoundCinematic=Get FoundCinematicReward())
   Return Node(execute=@n3.else, FoundCinematic=Get FoundCinematicReward())
   n6 = Cast To QuestUIRewardCinematicVM!(execute=@n2.LoopBody, Object=@n2) -> (then, CastFailed, AsQuest UI Reward Cinematic VM:Quest UI Reward Cinematic VM Object Reference) {is=title K2Node_DynamicCast}
   > DisplayCinematicRewardVM!(execute, self, CinematicRewardVM:Quest UI Reward Cinematic VM Object Reference=@n6) -> (then) {is=title K2Node_CallFunction}
   > n7 = Set FoundCinematicReward(FoundCinematicReward:bool=true) -> (Output_Get:bool)
  ]}
 EventGraph {nodes [
   Event PreConstruct()
   > SetVisibility(
    self=Get Pages()
    InVisibility=Select(
     Option 0:ESlateVisibility Enum=Collapsed
     Option 1:ESlateVisibility Enum=SelfHitTestInvisible
     Index=Get ShowPageInEditor() -> (ShowPageInEditor:bool)
    ) -> (ReturnValue:ESlateVisibility Enum)
   )
   > SetText(Text="LOCTEXT(\"100k\")")
   > SetStyle()
   > SetMaterialTextureParamName!(execute, self:Common Lazy Image Object Reference=Get LazyIcon(), TextureParamName:Name=Item) -> (then)
   > SetVisibility(
    self=Get Background() -> (Background:Image Object Reference)
    InVisibility=Select(Option 0=SelfHitTestInvisible, Option 1=Hidden, Index=Get bHideBackgroundOverride() -> (bHideBackgroundOverride:bool))
   )
   > SetBrushFromMaterial(
    self=Get Background()
    Material=Select(
     Option 0:Material Interface Object Reference=/QuestUI/QuestScreen/MaterialInstances/MI_UI_Quest_RewardBackground
     Option 1:Material Interface Object Reference=/QuestUI/QuestScreen/MaterialInstances/MI_UI_Quest_RewardBackground_Black
     Index=Get bUseBlackBackground() -> (bUseBlackBackground:bool)
    ) -> (ReturnValue:Material Interface Object Reference)
   )
   > SetPadding!(
    execute
    self:User Widget Object Reference=Get Block_Text()
    InPadding_Left:float=0
    InPadding_Top:float=0
    InPadding_Right:float=0
    InPadding_Bottom:float=Select(Option 0:float=-15, Option 1:float=5, Index=Get QuestTextInside() -> (QuestTextInside:bool)) -> (ReturnValue:float)
   ) -> (then)
  ]}
 OnRewardUpdated {nodes [OnRewardUpdated!() -> (then, RewardVM:Quest UI Reward VM Object Reference) {is=title K2Node_FunctionEntry}]}
 SetCategoryRewardVMs {nodes [
   n1 = SetCategoryRewardVMs!() -> (then, CustomWidgetRewards:Array of Quest UI Reward VM Object References) {is=title K2Node_FunctionEntry}
   > Set IsCategory(IsCategory:bool=true) -> (Output_Get:bool)
   > Clear(TargetArray:Array of Quest UI Reward VM Object References=Get CustomWidgetRewards())
   > n2 = For Each Loop(Array:Array of Quest UI Reward VM Object References=@n1) -> (Array Element:Quest UI Reward VM Object Reference)
   n3 = Branch(
    execute=@n2.Completed
    Condition=Is Not Empty!(TargetArray:Array of Quest UI Reward VM Object References=Get CustomWidgetRewards()) -> (ReturnValue:bool)
   )
   > Set HasContent(HasContent:bool=true) -> (Output_Get:bool)
   > n4 = DisplayRewardVMs!(execute, self) -> (then, FoundCinematic:bool) {is=title K2Node_CallFunction}
   > n5 = Branch(Condition=@n4)
   SetDesiredSizeOverride!(
    execute=[@n9.then, @n5.then]
    self:Image Object Reference=Get LazyIcon()
    DesiredSize:Vector 2D Structure=Get RewardIconSize() -> (RewardIconSize:Vector 2D Structure)
   ) -> (then)
   > SetSize!(
    execute
    self:Spacer Object Reference=Get Spacer() -> (Spacer:Spacer Object Reference)
    InSize:Vector 2D Structure=Get WidgetSize() -> (WidgetSize:Vector 2D Structure)
   ) -> (then)
   n6 = SlotAsOverlaySlot(Widget=Get LazyIcon())
   n7 = Get bCenterIconOverride() -> (bCenterIconOverride:bool)
   > SetHorizontalAlignment(
    self=@n6
    InHorizontalAlignment=Select(Option 0:EHorizontalAlignment Enum=HAlign_Fill, Option 1:EHorizontalAlignment Enum=HAlign_Center, Index=@n7) -> (ReturnValue:EHorizontalAlignment Enum)
   )
   > SetVerticalAlignment(
    self=@n6
    InVerticalAlignment=Select(Option 0:EVerticalAlignment Enum=VAlign_Fill, Option 1:EVerticalAlignment Enum=VAlign_Center, Index=@n7) -> (ReturnValue:EVerticalAlignment Enum)
   )
   n8 = Select(Option 0:int=0, Option 1:int=1, Index=Get HasContent() -> (HasContent:bool)) -> (ReturnValue:int)
   DisplayContainer(execute=[@n5.else, @n3.else], NumRewards=@n8)
   > n9 = DisplayPages(NumRewards=@n8)
   Branch(execute=@n2.LoopBody, Condition=IsEligibleForShowingInSmallReward!(self:Quest UI Reward VM Object Reference=@n2) -> (ReturnValue:bool))
   > Add(
    TargetArray:Array of Quest UI Reward VM Object References=Get CustomWidgetRewards()
    NewItem:"Quest UI Reward VM Object Reference (by ref)"=@n2
   ) {is=title K2Node_CallArrayFunction}
  ]}
 SetContentEnabled {nodes [
   n1 = SetContentEnabled!() -> (then, IsContentEnabled:bool=~) {is=title K2Node_FunctionEntry}
   > Set Is Content Enabled(Is Content Enabled:bool=@n1) -> (Output_Get:bool)
   > DisplayContentEnabled() {is=title K2Node_CallFunction}
   > SetStyle()
  ]}
 SetRewardVMs {nodes [
   n1 = SetRewardVMs!() -> (then, RewardVMs:Array of Quest UI Reward VM Object References) {is=title K2Node_FunctionEntry}
   > Set CurrentIndex(CurrentIndex:int=0) -> (Output_Get:int=0)
   > Clear(TargetArray=Get CustomWidgetRewards())
   > n2 = For Each Loop(Array=@n1)
   n3 = Length(TargetArray=Get CustomWidgetRewards())
   DisplayContainer(execute=@n2.Completed, NumRewards=@n3)
   > DisplayPages(NumRewards=@n3)
   > UpdateReward!(execute, self, Index:int=Get CurrentIndex() -> (CurrentIndex:int=0)) -> (then)
   Branch(execute=@n2.LoopBody, Condition=IsEligibleForShowingInSmallReward(self=@n2))
   > Add(TargetArray=Get CustomWidgetRewards(), NewItem=@n2)
  ]}
 SetStyle {nodes [
   SetStyle() {is=title K2Node_FunctionEntry}
   > SetPadding(
    self=Get Block_Text()
    InPadding_Left=0
    InPadding_Top=0
    InPadding_Right=0
    InPadding_Bottom=Select(Option 0:float=-15, Option 1:float=5, Index=Get QuestTextInside()) -> (ReturnValue:float)
   )
   > UpdateWidgetStyleTags!(
    execute
    InWidget:Widget Object Reference=Get TextBlock!(self:WBP UIKit Block Text Object Reference=Get Block_Text()) -> (TextBlock:Common Text Object Reference)
    InAddedTags:Array of Gameplay Tag Structures=Get TextStyleTags() -> (TextStyleTags:Array of Gameplay Tag Structures)
    InRemovedTags:Array of Gameplay Tag Structures=Make Array()
   ) -> (then, ReturnValue:bool)
   > SetSize(self=Get Spacer(), InSize=Get WidgetSize())
   > SetDesiredSizeOverride(self=Get LazyIcon(), DesiredSize=Get RewardIconSize())
   n1 = Get Pages()
   > Set PagesFont!(
    execute
    PagesFont:Slate Font Info Structure=Get Pages Font() -> (Pages Font:Slate Font Info Structure)
    self:WBP Quest Pages Object Reference=@n1
   ) -> (then, Output_Get:Slate Font Info Structure)
   > UpdatePagesFont(self:WBP Quest Pages Object Reference=@n1)
   > SetFont!(
    execute
    self:WBP UIKit Block Text Object Reference=Get Block_Text()
    Font:"Slate Font Info Structure (by ref)"=Get RewardAmountFont() -> (RewardAmountFont:Slate Font Info Structure)
   ) -> (then)
  ]}
 SetText {nodes [
   n1 = SetText!() -> (then, Text:Text) {is=title K2Node_FunctionEntry}
   > SetText(self:WBP UIKit Block Text Object Reference=Get Block_Text(), Text=@n1) {is=title K2Node_CallFunction}
   > SetBlockVisibility!(execute, self:WBP UIKit Block Base Object Reference=Get Block_Text(), Visible:bool=NOT Boolean(A=TextIsEmpty(InText=@n1))) -> (then)
  ]}
 UpdateIcon {nodes [
   n1 = UpdateIcon!() -> (then, LazyObject:"Soft Object Reference (by ref)") {is=title K2Node_FunctionEntry}
   > n2 = Branch(Condition=NOT Boolean(A=IsAssetNull!(Asset:Soft Object Reference=@n1) -> (ReturnValue:bool)))
   > SetBrushFromLazyDisplayAsset!(
    execute
    self:Common Lazy Image Object Reference=Get LazyIcon()
    LazyObject:"Soft Object Reference (by ref)"=Get LazyObject() -> (LazyObject:"Soft Object Reference (by ref)")
    bMatchTextureSize:bool=false
   ) -> (then)
   > SetVisibility(self=Get LazyIcon(), InVisibility=SelfHitTestInvisible)
   > n3 = Cast To QuestUIRewardCinematicVM(Object=Get QuestUIRewardVM() -> (QuestUIRewardVM:Quest UI Reward VM Object Reference))
   n4 = Get LazyIcon()
   > SetRenderScale!(execute, self:Widget Object Reference=@n4, Scale:Vector 2D Structure=(X=1.3,Y=1.3)) -> (then)
   SetRenderScale(execute=@n3.CastFailed, self=@n4, Scale=(X=1,Y=1))
   SetVisibility(execute=@n2.else, self=Get LazyIcon(), InVisibility=Collapsed)
  ]}
 UpdateIconAlignment {nodes [
   UpdateIconAlignment() {is=title K2Node_FunctionEntry}
   n1 = Get QuestUIRewardVM()
   > Is Valid(InputObject=@n1)
   n2 = SlotAsOverlaySlot(Widget=Get LazyIcon())
   n3 = OR Boolean(
    A="Equal (Enum)"(
     A:EFortItemType Enum=Get ItemType!(self:Quest UI Reward VM Object Reference=@n1) -> (ItemType:EFortItemType Enum=WorldItem)
     B:EFortItemType Enum=AccountResource
    )
    B=Get bCenterIconOverride()
   )
   > SetHorizontalAlignment(
    self=@n2
    InHorizontalAlignment=Select(Option 0:EHorizontalAlignment Enum=HAlign_Fill, Option 1:EHorizontalAlignment Enum=HAlign_Center, Index=@n3) -> (ReturnValue:EHorizontalAlignment Enum)
   )
   > SetVerticalAlignment(
    self=@n2
    InVerticalAlignment=Select(Option 0:EVerticalAlignment Enum=VAlign_Fill, Option 1:EVerticalAlignment Enum=VAlign_Center, Index=@n3) -> (ReturnValue:EVerticalAlignment Enum)
   )
  ]}
 UpdateItemType {nodes [
   UpdateItemType!() -> (then, ItemType:EFortItemType Enum=WorldItem) {is=title K2Node_FunctionEntry}
   > UpdateQuantityText()
   > UpdateIconAlignment() {is=title K2Node_CallFunction}
  ]}
 UpdateQuantity {nodes [
   UpdateQuantity!() -> (then, Quantity:int) {is=title K2Node_FunctionEntry}
   > UpdateQuantityText()
  ]}
 UpdateQuantityText {nodes [
   UpdateQuantityText() {is=title K2Node_FunctionEntry}
   n1 = Get QuestUIRewardVM()
   > Is Valid(InputObject=@n1)
   n2 = "Equal (Enum)"(A=Get ItemType(self=@n1) -> (ItemType=WorldItem), B=AccountResource)
   n3 = Get Quantity!(self:Quest UI Reward VM Object Reference=@n1) -> (Quantity:int=0)
   n4 = "Not Equal (Integer)"(A=@n3, B=1)
   n5 = "To Text (Integer)"(Value=@n3)
   n6 = int / int(A=@n3, B=1000)
   n7 = Format Text!(Format="LOCTEXT(\"{Amount}k\")", Amount:int=@n6) -> (Result)
   n8 = "integer >= integer"(A=@n6, B=1)
   > SetText(
    self:Self Object Reference
    Text=Select(Option 0:Text, Option 1:Text=Select(Option 0=@n5, Option 1=@n7, Index=@n8), Index=AND Boolean(A=@n4, B=@n2)) -> (ReturnValue:Text)
   )
  ]}
 UpdateReward {nodes [
   UpdateReward!() -> (then, Index:int) {is=title K2Node_FunctionEntry}
   n1 = Get CustomWidgetRewards()
   > n2 = Branch(Condition="Equal (Integer)"(A=Length(TargetArray=@n1)))
   DisplayRewardVM(execute=@n2.else, RewardVM="Get (a copy)"(Array=Get CustomWidgetRewards(), Dimension 1=Get Index() -> (Index:int=0)))
  ]}
 UpdateRewardRadius {nodes [
   n1 = UpdateRewardRadius!() -> (then, InPixels:bool=~, CornerRadius:double=~) {is=title K2Node_FunctionEntry}
   > n2 = GetDynamicMaterial(self=Get LazyIcon())
   > SetScalarParameterValue(self=@n2, ParameterName=UseValueInPixels, Value="To Float (Boolean)"(InBool=@n1))
   > SetScalarParameterValue(self=@n2, ParameterName=TopLeftCornerRadius, Value=@n1.CornerRadius)
   > SetScalarParameterValue(self=@n2, ParameterName=TopRightCornerRadius, Value=@n1.CornerRadius)
   > SetScalarParameterValue(self=@n2, ParameterName=BottomRightCornerRadius, Value=@n1.CornerRadius)
   > SetScalarParameterValue(self=@n2, ParameterName=BottomLeftCornerRadius, Value=@n1.CornerRadius)
  ]}
 UpdateShownReward {nodes [
   UpdateShownReward() {is=title K2Node_FunctionEntry}
   n1 = int - int(A=Length(TargetArray=Get CustomWidgetRewards()))
   n2 = Get CurrentIndex() -> (CurrentIndex=0)
   > n3 = Set CurrentIndex(CurrentIndex=Select(Option 0:int=int + int(A=@n2), Option 1:int=0, Index="integer >= integer"(A=@n2, B=@n1)) -> (ReturnValue:int)) -> (Output_Get=0)
   > UpdateReward(Index=@n3) {is=title K2Node_CallFunction}
   > SetSelectedPage!(execute, self:WBP Quest Pages Object Reference=Get Pages(), SelectedOption:int=@n3) -> (then)
  ]}
]
variables [
 bCenterIconOverride {type=boolean}
 bHideBackgroundOverride {type=boolean}
 bUseBlackBackground {type=boolean}
 CinematicRewardIcon {type=MaterialInterface, default=/QuestUI/QuestScreen/Materials/MI_UI_QuestIcon_CinematicReward_NoVFX}
 CinematicRewardIcon_VFX {type=MaterialInterface, default=/QuestUI/QuestScreen/Materials/MI_UI_QuestIcon_CinematicReward_VFX}
 CurrentIndex {type=integer}
 CustomWidgetRewards {type="array<QuestUIRewardVM>", default []}
 Is Content Enabled {type=boolean, default=true}
 IsCategory {type=boolean}
 OnRewardUpdated {}
 Pages Font {
  type=SlateFontInfo
  default {
   bForceMonospaced=false
   bMaterialIsStencil=false
   fontMaterial=null
   fontObject=&HeadingNow
   letterSpacing=0
   monospacedWidth=1
   size=15.75
   skewAmount=0
   typefaceFontName=75 Medium
   outlineSettings {
    bApplyOutlineToDropShadows=false
    bMiteredCorners=false
    bSeparateFillAlpha=false
    outlineMaterial=null
    outlineSize=0
    outlineColor=<rgba 0 0 0 1>
   }
  }
 }
 PagesStyleTags {type="array<GameplayTag>", default []}
 QuestTextInside {type=boolean, default=true}
 RewardAmountFont {
  type=SlateFontInfo
  default {
   bForceMonospaced=false
   bMaterialIsStencil=false
   fontMaterial=null
   fontObject=&HeadingNow
   letterSpacing=0
   monospacedWidth=1
   size=15.75
   skewAmount=0
   typefaceFontName=75 Medium
   outlineSettings {
    bApplyOutlineToDropShadows=false
    bMiteredCorners=false
    bSeparateFillAlpha=false
    outlineMaterial=null
    outlineSize=1
    outlineColor=<rgba 0 0 0 1>
   }
  }
 }
 RewardIconSize {type=Vector2D, default=<64 64>}
 ShowPageInEditor {type=boolean, default=true}
 TextStyleTags {type="array<GameplayTag>", default [{tagName=StyleSheet.Style.Subheading}, {tagName=StyleSheet.Style.20}]}
 TimerHandle {type=TimerHandle}
 WidgetSize {type=Vector2D, default=<112 112>}
]
widgets {
 info {inheritedWidgetCount=0, parentClass=&CommonUserWidget, rootWidgetClass=&VerticalBox, widgetCount=9}
 widgets [
  VerticalBox_0 {widgetClassPath=&VerticalBox, widget=VerticalBox_0}
  IconRoot {
   widgetClassPath=&Overlay
   bIsVariable=true
   parent=VerticalBox_0
   widget=IconRoot
   slot {class=VerticalBoxSlot, properties {padding=<ltrb 14 15 14 15>, size {sizeRule=Fill, value=1}}}
  }
  Background {
   widgetClassPath=&Image
   bIsVariable=true
   parent=IconRoot
   widget=Background
   properties {brush {
     bIsDynamicallyLoaded=false
     drawAs=Image
     imageType=FullColor
     mirroring=NoMirror
     resourceName=None
     resourceObject=&MI_UI_Quest_RewardBackground
     tiling=NoTile
     imageSize=<32 32>
     margin=<ltrb 0 0 0 0>
     outlineSettings {
      bUseBrushTransparency=false
      roundingType=HalfHeightRadius
      width=0
      color {colorUseRule=UseColor_Specified, specifiedColor=<rgba 0 0 0 0>}
      cornerRadii=<0 0 0 0>
     }
     tintColor {colorUseRule=UseColor_Specified, specifiedColor=<rgba 1 1 1 1>}
     uVRegion {bIsValid=false, max=<0 0>, min=<0 0>}
    }}
   slot {class=OverlaySlot, properties {horizontalAlignment=HAlign_Fill, verticalAlignment=VAlign_Fill}}
  }
  LazyIcon {
   widgetClassPath=&FortLazyImage
   bIsVariable=true
   parent=IconRoot
   widget=LazyIcon
   properties {brush {
     bIsDynamicallyLoaded=false
     drawAs=Image
     imageType=FullColor
     mirroring=NoMirror
     resourceName=None
     resourceObject=/QuestUI/QuestScreen/MaterialInstances/MI_UI_Quest_RewardImage
     tiling=NoTile
     imageSize=<64 64>
     margin=<ltrb 0 0 0 0>
     outlineSettings {
      bUseBrushTransparency=false
      roundingType=HalfHeightRadius
      width=0
      color {colorUseRule=UseColor_Specified, specifiedColor=<rgba 0 0 0 0>}
      cornerRadii=<0 0 0 0>
     }
     tintColor {colorUseRule=UseColor_Specified, specifiedColor=<rgba 1 1 1 1>}
     uVRegion {bIsValid=false, max=<0 0>, min=<0 0>}
    }}
   slot {class=OverlaySlot, properties {horizontalAlignment=HAlign_Fill, verticalAlignment=VAlign_Fill, padding=<ltrb 2 2 2 2>}}
  }
  Block_Text {
   widgetClassPath=/FortUILibrary/Systems/ModularButton/Blocks/WBP_UIKit_Block_Text
   bIsVariable=true
   parent=IconRoot
   widget=Block_Text
   slot {class=OverlaySlot, properties {horizontalAlignment=HAlign_Fill, verticalAlignment=VAlign_Bottom, padding=<ltrb 0 0 0 5>}}
  }
  Spacer {
   widgetClassPath=&Spacer
   bIsVariable=true
   parent=IconRoot
   widget=Spacer
   properties {size=<100 100>}
   slot {class=OverlaySlot, properties {horizontalAlignment=HAlign_Fill, verticalAlignment=VAlign_Fill}}
  }
  SeasonPassIcon {
   widgetClassPath=&/WBP_SeasonPassIcon
   bIsVariable=true
   parent=IconRoot
   widget=SeasonPassIcon
   slot {class=OverlaySlot, properties {padding=<ltrb -6 -7 0 0>}}
  }
  TextRoot {
   widgetClassPath=&Overlay
   bIsVariable=true
   parent=VerticalBox_0
   widget=TextRoot
   slot {class=VerticalBoxSlot, properties {horizontalAlignment=HAlign_Center, padding=<ltrb 5 -3.5 5 10>}}
  }
  Pages {
   widgetClassPath=/QuestUI/QuestScreen/RewardsPanel/WBP_QuestPages
   bIsVariable=true
   parent=TextRoot
   widget=Pages
   properties {bShowButtons=false, visibility=Collapsed}
   slot {class=OverlaySlot}
  }
 ]
}
