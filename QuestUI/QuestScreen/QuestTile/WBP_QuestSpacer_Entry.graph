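# WBP_QuestSpacer_Entry is a CommonUserWidget list entry that renders an empty, square gap in a quest UI list.
# When a list item is set, it unwraps the item, casts it to QuestUISpacerVM, stores that view model and reads its Size.
# SetSize stores the value and calls Display, which applies it to both X and Y of the root Spacer.
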WidgetBlueprint /QuestUI/QuestScreen/QuestTile/WBP_QuestSpacer_Entry
graphs [
 Display {nodes [
   Display() {is=title K2Node_FunctionEntry}
   n1 = Get Size() -> (Size:double)
   > SetSize!(execute, self:Spacer Object Reference=Get Spacer() -> (Spacer:Spacer Object Reference), InSize_X:double=@n1, InSize_Y:double=@n1) -> (then)
  ]}
 EventGraph {nodes [
   n1 = Event OnListItemObjectSet()
   > n2 = Cast To QuestUISpacerVM!(execute, Object=UnwrapItem!(Item:Object Reference=@n1.ListItemObject) -> (ReturnValue:Object Reference)) -> (then, CastFailed, AsQuest UISpacer VM:Quest UISpacer VM Object Reference) {is=title K2Node_DynamicCast}
   > n3 = Set QuestUISpacerVM(QuestUISpacerVM:Quest UISpacer VM Object Reference=@n2) -> (Output_Get:Quest UISpacer VM Object Reference)
   > SetSize!(execute, self:Self Object Reference, Size:double=Get Size!(self:Quest UISpacer VM Object Reference=@n3) -> (Size)) -> (then)
  ]}
 SetSize {nodes [
   n1 = SetSize!() -> (then, Size:double=~) {is=title K2Node_FunctionEntry}
   > Set Size(Size:double=@n1) -> (Output_Get:double)
   > Display() {is=title K2Node_CallFunction}
  ]}
]
variables [Size {type=number}]
widgets {
 info {inheritedWidgetCount=0, parentClass=&CommonUserWidget, rootWidgetClass=&Spacer, widgetCount=1}
 widgets [Spacer {widgetClassPath=&Spacer, bIsVariable=true, widget=Spacer, properties {size=<0 0>}, slot=~}]
}
