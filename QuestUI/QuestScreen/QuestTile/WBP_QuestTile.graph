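# WBP_QuestTile is a quest list tile button showing a quest's text, progress, stage, reward, avatar, badge, tracked pin and enabled or punch-card state.
# List item data sets the quest view model, unwrapping punch-card quests, and display functions update styling through materials and opacity.
# Dragging spawns a copy as the drag visual; dropping sets the dragged quest's tracked index.
# Animations handle hover, selection and move-mode dimming, with gamepad-specific branches.
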WidgetBlueprint /QuestUI/QuestScreen/QuestTile/WBP_QuestTile
animations [
 DragOver {
  binding_count=1
  length_seconds=0.1
  object=&self:DragOver
  sequence {bindings [DragOverArea {
     id=#1
     possessed_class=Overlay
     tracks [{
       type=MovieSceneFloatTrack
       display_name=Render Opacity
       sections [{type=MovieSceneFloatSection, end_frame=2, start_frame=0, channels [None {type=MovieSceneScriptingFloatChannel, keys [[0,0],[2,1]]}]}]
      }]
    }]}
 }
 DragStateDimmed {
  binding_count=5
  length_seconds=0.1
  object=&self:DragStateDimmed
  sequence {bindings [
    DragDropBar {
     id=#2
     possessed_class=Image
     tracks [{
       type=MovieSceneFloatTrack
       display_name=Render Opacity
       sections [{type=MovieSceneFloatSection, end_frame=2, start_frame=0, channels [None {type=MovieSceneScriptingFloatChannel, keys [[0,1],[2,4]]}]}]
      }]
    }
    WBP_QuestTile_C_5 {
     id=#3
     possessed_class=WBP_QuestTile_C
     tracks [{
       type=MovieSceneFloatTrack
       display_name=Render Opacity
       sections [{type=MovieSceneFloatSection, end_frame=2, start_frame=0, channels [None {type=MovieSceneScriptingFloatChannel, keys [[0,1],[2,0.5]]}]}]
      }]
    }
    LeftArrow_Img {
     id=#4
     possessed_class=Image
     tracks [{
       type=MovieSceneFloatTrack
       display_name=Render Opacity
       sections [{type=MovieSceneFloatSection, end_frame=2, start_frame=0, channels [None {type=MovieSceneScriptingFloatChannel, keys [[0,1],[2,4]]}]}]
      }]
    }
    RightArrow_Img {
     id=#5
     possessed_class=Image
     tracks [{
       type=MovieSceneFloatTrack
       display_name=Render Opacity
       sections [{type=MovieSceneFloatSection, end_frame=2, start_frame=0, channels [None {type=MovieSceneScriptingFloatChannel, keys [[0,1],[2,4]]}]}]
      }]
    }
    WBP_TrackedQuestPin {
     id=#6
     possessed_class=WBP_TrackedQuestPin_C
     tracks [{
       type=MovieSceneFloatTrack
       display_name=Render Opacity
       sections [{type=MovieSceneFloatSection, end_frame=2, start_frame=0, channels [None {type=MovieSceneScriptingFloatChannel, keys [[0,1],[2,1.25]]}]}]
      }]
    }
   ]}
 }
 DragStarted_Gamepad {
  binding_count=1
  length_seconds=0.1
  object=&self:DragStarted_Gamepad
  sequence {bindings [WBP_QuestTile_C_5 {
     id=#7
     possessed_class=WBP_QuestTile_C
     tracks [
      {
       type=MovieScene2DTransformTrack
       sections [{
         type=MovieScene2DTransformSection
         end_frame=2
         start_frame=0
         channels [
          Translation.X {type=MovieSceneScriptingFloatChannel, keys [[0,-50],[2,0]]}
          Translation.Y {type=MovieSceneScriptingFloatChannel, const [0,1]}
         ]
        }]
      }
      {
       type=MovieSceneFloatTrack
       display_name=Render Opacity
       sections [{type=MovieSceneFloatSection, end_frame=2, start_frame=0, channels [None {type=MovieSceneScriptingFloatChannel, keys [[0,0],[2,1]]}]}]
      }
     ]
    }]}
 }
 ShowSelection {
  binding_count=1
  length_seconds=0.1
  object=&self:ShowSelection
  sequence {
   bindings [SelectedStateOutline {
     id=#8
     possessed_class=Image
     tracks [
      {
       type=MovieSceneFloatTrack
       display_name=Render Opacity
       sections [{type=MovieSceneFloatSection, end_frame=6, start_frame=0, channels [None {type=MovieSceneScriptingFloatChannel, keys [[0,0],[6,1]]}]}]
      }
      {
       type=MovieSceneByteTrack
       display_name=Visibility
       sections [{type=MovieSceneByteSection, end_frame=6, start_frame=0, channels [None {type=MovieSceneScriptingByteChannel, keys [[0,1],[2,4],[6,4]]}]}]
      }
     ]
    }]
   display_rate {numerator=60}
  }
 }
 HoverIdle {
  binding_count=1
  length_seconds=0.5
  object=&self:HoverIdle
  sequence {bindings [Overlay_120 {
     id=#9
     possessed_class=Overlay
     tracks [{
       type=MovieSceneFloatTrack
       display_name=Render Opacity
       sections [{
         type=MovieSceneFloatSection
         end_frame=10
         start_frame=0
         channels [None {type=MovieSceneScriptingFloatChannel, keys [[0,1],[5,0.8],[10,1]]}]
        }]
      }]
    }]}
 }
 HoverOver {
  binding_count=1
  length_seconds=0.05
  object=&self:HoverOver
  sequence {bindings [HoverOverArea {
     id=#10
     possessed_class=FortVisualAttachment
     tracks [{
       type=MovieSceneFloatTrack
       display_name=Render Opacity
       sections [{type=MovieSceneFloatSection, end_frame=1, start_frame=0, channels [None {type=MovieSceneScriptingFloatChannel, keys [[0,0],[1,0.6]]}]}]
      }]
    }]}
 }
]
graphs [
 Display {nodes [
   Display() {is=title K2Node_FunctionEntry}
   > SetPinIndex!(
    execute
    self:WBP Tracked Quest Pin Object Reference=Get WBP_TrackedQuestPin() -> (WBP_TrackedQuestPin:WBP Tracked Quest Pin Object Reference)
    Index:int=Get PinIndex() -> (PinIndex:int=0)
   ) -> (then)
   > DisplayPunchcardState()
   > DisplayEnabledState()
   > DisplayProgress()
  ]}
 DisplayEnabledState {nodes [
   DisplayEnabledState() {is=title K2Node_FunctionEntry}
   > SetRenderOpacity(
    self=Get CharacterData() -> (CharacterData:Grid Panel Object Reference)
    InOpacity=Select(Option 0:float=0.5, Option 1:float=1, Index=Get IsQuestEnabled() -> (IsQuestEnabled:bool)) -> (ReturnValue:float)
   )
   > SetBrushFromMaterial(
    self=Get Img_QuestBackground() -> (Img_QuestBackground:Image Object Reference)
    Material=Select(
     Option 0:Material Instance Object Reference=/FortniteCoreUI/UI/Quests/MaterialInstances/MI_UI_QuestBackplate_Disabled
     Option 1:Material Instance Object Reference=/FortniteCoreUI/UI/Quests/MaterialInstances/MI_UI_QuestBackplate
     Index=Get IsQuestEnabled()
    ) -> (ReturnValue:Material Instance Object Reference)
   )
   > SetForegroundColor!(
    execute
    self:User Widget Object Reference
    InForegroundColor:Slate Color Structure=Select(
     Option 0:Slate Color Structure=Get DISABLED_COLOR() -> (DISABLED_COLOR:Slate Color Structure)
     Option 1:Slate Color Structure=Get WHITE() -> (WHITE:Slate Color Structure)
     Index=Get IsQuestEnabled()
    ) -> (ReturnValue:Slate Color Structure)
   ) -> (then)
   n1 = Get!() -> (DISABLED_COLOR_SpecifiedColor:Linear Color Structure, DISABLED_COLOR_ColorUseRule:ESlateColorStylingMode Enum)
   n2 = Get!() -> (WHITE_SpecifiedColor:Linear Color Structure, WHITE_ColorUseRule:ESlateColorStylingMode Enum)
   > SetColorAndOpacity(
    self:User Widget Object Reference=Get AdditionalText() -> (AdditionalText:WBP UIKit Block Text Object Reference)
    InColorAndOpacity:Linear Color Structure=Select(Option 0:Linear Color Structure=@n1, Option 1:Linear Color Structure=@n2, Index=Get IsQuestEnabled()) -> (ReturnValue:Linear Color Structure)
   )
   > SetRenderOpacity(
    self=Get CommonRichTextHeader() -> (CommonRichTextHeader:Common Rich Text Block Object Reference)
    InOpacity=Select(Option 0:float=BreakColor(InColor=@n1).A, Option 1:float=BreakColor(InColor=@n2).A, Index=Get IsQuestEnabled()) -> (ReturnValue:float)
   )
   > SetIsContentEnabled!(execute, self:WBP Tracked Quest Pin Object Reference=Get WBP_TrackedQuestPin(), IsEnabled:bool=Get IsQuestEnabled()) -> (then)
   > SetContentEnabled!(
    execute
    self:WBP Quest Reward Object Reference=Get QuestReward() -> (QuestReward:WBP Quest Reward Object Reference)
    IsContentEnabled:bool=Get IsQuestEnabled()
   ) -> (then)
  ]}
 DisplayProgress {nodes [
   DisplayProgress() {is=title K2Node_FunctionEntry}
   > SetProgressBarValue!(execute, self, Progress:double=Get ProgressValue() -> (ProgressValue:double)) -> (then)
   > SetStatusIndicatorToCompleted!(execute, self, Completed:bool=Get IsComplete() -> (IsComplete:bool)) -> (then)
   > SetProgressColor!(execute, self, Completed:bool=Get IsComplete(), QuestEnabled:bool=Get IsQuestEnabled()) -> (then)
   > SetText(
    self=Get Rich_Progress_Text() -> (Rich_Progress_Text:Rich Text Block Object Reference)
    InText=Get ProgressTextValue() -> (ProgressTextValue:Text)
   )
  ]}
 DisplayPunchcardState {nodes [
   DisplayPunchcardState() {is=title K2Node_FunctionEntry}
   > n1 = Is Valid(InputObject=Get As MVVM Punch Card Quest() -> (As MVVM Punch Card Quest:MVVM Punch Card Quest Object Reference))
   > n2 = Set IsLastElement(IsLastElement:bool=Get LastElement!(self:MVVM Punch Card Quest Object Reference=Get As MVVM Punch Card Quest()) -> (LastElement:bool)) -> (Output_Get:bool)
   SetVisibility(
    execute=[@n2.then, @n1.Is Not Valid]
    self=Get PunchCardSeparator() -> (PunchCardSeparator:Image Object Reference)
    InVisibility=Select(Option 0:ESlateVisibility Enum=Collapsed, Option 1:ESlateVisibility Enum=SelfHitTestInvisible, Index=Get IsPunchCard() -> (IsPunchCard:bool)) -> (ReturnValue:ESlateVisibility Enum)
   )
   > n3 = GetDynamicMaterial(self=Get Img_QuestBackground())
   n4 = Select(Option 0:bool=true, Option 1:bool=false, Index=Get IsLastElement() -> (IsLastElement:bool)) -> (ReturnValue:bool)
   n5 = Get IsPunchCard()
   > SetScalarParameterValue(self=@n3, ParameterName=IsSquaredOff, Value="To Float (Boolean)"(InBool=Select(Option 0=false, Option 1=@n4, Index=@n5)))
   n6 = Select(Option 0=false, Option 1=true, Index=Get IsLastElement())
   n7 = Get IsPunchCard()
   > SetScalarParameterValue(self=@n3, ParameterName=IsCut, Value="To Float (Boolean)"(InBool=Select(Option 0=false, Option 1=@n6, Index=@n7)))
   > Branch(Condition=Get IsPunchCard())
   > SetPadding!(
    execute
    self:User Widget Object Reference=Self-Reference()
    InPadding_Left:float=0
    InPadding_Top:float=0
    InPadding_Right:float=0
    InPadding_Bottom:float=0
   ) -> (then)
  ]}
 DoCustomNavigation {nodes [
   n1 = DoCustomNavigation!() -> (then, Navigation:EUINavigation Enum) {is=title K2Node_FunctionEntry}
   > Call On Quest Navigate!(
    execute:Exec
    self:WBP Quest Tile Object Reference
    QuestTile:User Widget Object Reference=Self-Reference()
    Direction:EUINavigation Enum=@n1
   ) -> (then:Exec) {is=title K2Node_CallDelegate}
   > Return Node!(execute, ReturnValue:Widget Object Reference)
  ]}
 EventGraph {nodes [
   SetText!(
    execute
    self:WBP UIKit Block Text Object Reference=Get AdditionalText()
    Text:Text="LOCTEXT(\"Next we\\'re giving you XP for... aiming? Sure. Why not. - Gwen\")"
   ) -> (then)
   > SetText(Text="LOCTEXT(\"Hit opponents while aiming down sights with the Monarch Pistol\")")
   > SetText!(execute, self:Rich Text Block Object Reference=Get Rich_Progress_Text(), InText:"Text (by ref)"="LOCTEXT(\"180/200\")") -> (then)
   n1 = SetDebugQuestID!(execute, self, QuestID:"String (by ref)"=MakeLiteralString()) -> (then)
   SetNavigationRuleCustom!(execute=[@n1.then, @n27.then], self:Widget Object Reference, Direction:EUINavigation Enum=Right, InCustomDelegate:Delegate=Create Event()) -> (then)
   > SetNavigationRuleCustom(Direction=Left, InCustomDelegate=Create Event())
   n2 = SetAvatarIcon!() -> (OutputDelegate, then, Avatar:"Soft Object Reference (by ref)") {is=title K2Node_CustomEvent}
   > n3 = Branch(Condition=IsValidSoftObjectReference(SoftObjectReference=@n2.Avatar))
   n4 = Get CharacterData()
   > SetVisibility(self=@n4, InVisibility=HitTestInvisible)
   > n5 = Async Load Asset(Asset=@n2.Avatar)
   n6 = Cast To Texture!(execute=@n5.Completed, Object=@n5) -> (then, CastFailed, AsTexture:Texture Object Reference) {is=title K2Node_DynamicCast}
   > n7 = GetDynamicMaterial(self=Get CharacterBackground() -> (CharacterBackground:Image Object Reference))
   > SetTextureParameterValue(self=@n7, ParameterName=Item, Value=@n6)
   SetVisibility(execute=[@n3.else, @n6.CastFailed], self=@n4, InVisibility=Collapsed)
   n8 = SetCustomBadge!() -> (OutputDelegate, then, CustomBadge:"Soft Object Reference (by ref)") {is=title K2Node_CustomEvent}
   > SetVisibility(
    self=Get CustomBadge() -> (CustomBadge:Image Object Reference)
    InVisibility=Select(
     Option 0:ESlateVisibility Enum=Collapsed
     Option 1:ESlateVisibility Enum=SelfHitTestInvisible
     Index=IsValidSoftObjectReference(SoftObjectReference=@n8.CustomBadge)
    ) -> (ReturnValue:ESlateVisibility Enum)
   )
   > n9 = Async Load Asset(Asset=@n8.CustomBadge)
   n10 = Cast To Texture2D(execute=@n9.Completed, Object=@n9)
   > SetBrushFromTexture!(execute, self:Image Object Reference=Get CustomBadge(), Texture:Texture 2D Object Reference=@n10, bMatchSize:bool=false) -> (then)
   n11 = Cast To MaterialInterface!(execute=@n10.CastFailed, Object=@n9) -> (then, CastFailed, AsMaterial Interface:Material Interface Object Reference) {is=title K2Node_DynamicCast}
   > SetBrushFromMaterial(self=Get CustomBadge(), Material=@n11)
   OverlayGamepadDragStartedVisualization() {is=title K2Node_CustomEvent}
   > PlayAnimation(InAnimation=Get DragStarted_Gamepad() -> (DragStarted_Gamepad:Widget Animation Object Reference))
   > SetBlockVisibility!(
    execute
    self:WBP UIKit Block Base Object Reference=Get Block_Outline() -> (Block_Outline:WBP UIKit Block Outline Object Reference)
    Visible:bool=false
   ) -> (then)
   n12 = MoveModeChangedVisualization!() -> (OutputDelegate, then, NeedsToAnimate:bool=~) {is=title K2Node_CustomEvent}
   > n13 = Sequence()
   > SetBlockVisibility(self=Get Block_Outline(), Visible=NOT Boolean(A=Get IsMoveMode() -> (IsMoveMode:bool)))
   > SetVisibility(
    self=Get HoverOverArea() -> (HoverOverArea:Fort Visual Attachment Object Reference)
    InVisibility=Select(Option 0=SelfHitTestInvisible, Option 1=Collapsed, Index=Get IsMoveMode())
   )
   > n14 = Branch(Condition=@n12.NeedsToAnimate)
   > PlayAnimation(
    InAnimation=Get DragStateDimmed() -> (DragStateDimmed:Widget Animation Object Reference)
    PlayMode=Select(Option 0:EUMGSequencePlayMode Enum=Reverse, Option 1:EUMGSequencePlayMode Enum, Index=Get IsMoveMode()) -> (ReturnValue:EUMGSequencePlayMode Enum)
   )
   RefreshMoveModeStyle(execute=@n14.else)
   > n15 = Branch(Condition=GetSelected!(self:Common Button Base Object Reference) -> (ReturnValue:bool))
   > SelectionChangedVisualization!(execute, self, IsSelected:bool=true) -> (then)
   n16 = Get CommonInputSubsystem!() -> (ReturnValue:Common Input Object Reference) {is=title K2Node_GetBlueprintContextEx}
   Branch(execute=@n15.else, Condition="Equal (Enum)"(A=GetCurrentInputType(self=@n16), B=Gamepad))
   > SetRenderOpacity(self=Get HoverOverArea(), InOpacity=0)
   n17 = Branch(execute=@n13.then_1, Condition=Get IsMoveMode())
   SetRenderOpacity(execute=@n17.else, self=Get DragOverArea() -> (DragOverArea:Overlay Object Reference), InOpacity=0)
   n18 = DraggedOverChangedVisualization!() -> (OutputDelegate, then, IsEntering:bool=~, DropPlacement:"EDragDropPlacementType Enum (by ref)"=NewEnumerator0) {is=title K2Node_CustomEvent}
   > SetDragDropPlacement!(execute, self, Placement:"EDragDropPlacementType Enum (by ref)"=@n18.DropPlacement) -> (then)
   > n19 = Branch(Condition=@n18.IsEntering)
   > Branch(Condition=Get IsDraggable() -> (IsDraggable:bool))
   > PlayAnimation(InAnimation=Get DragOver() -> (DragOver:Widget Animation Object Reference))
   StopAnimation(execute=@n19.else, InAnimation=Get DragOver())
   > SetRenderOpacity(self=Get DragOverArea(), InOpacity=0)
   n20 = SelectionChangedVisualization!() -> (OutputDelegate, then, IsSelected:bool=~) {is=title K2Node_CustomEvent}
   n21 = Get IsMoveMode()
   n22 = GetCurrentInputType(self=Get CommonInputSubsystem())
   > Branch(Condition=OR Boolean(A=NOT Boolean(A=@n21), B="Not Equal (Enum)"!(A:ECommonInputType Enum=@n22, B:ECommonInputType Enum=Gamepad) -> (ReturnValue:bool)))
   > PlayAnimation(
    InAnimation=Get ShowSelection() -> (ShowSelection:Widget Animation Object Reference)
    PlayMode=Select(Option 0=Reverse, Index=@n20.IsSelected)
   )
   > n23 = GetDynamicMaterial(self=Get Backplate!(self:WBP UIKit Backplate Base Object Reference=Get Stage_BackPlate() -> (Stage_BackPlate:WBP UIKit Backplate Object Reference)) -> (Backplate:Image Object Reference))
   > SetScalarParameterValue(self=@n23, ParameterName=Backplate Switch, Value=Select(Option 0:float=1, Option 1:float=4, Index=@n20.IsSelected) -> (ReturnValue:float))
   > SetColorAndOpacity!(
    execute
    self:Text Object Reference=Get StageCounterText() -> (StageCounterText:Text Block Object Reference)
    InColorAndOpacity_SpecifiedColor:Linear Color Structure=Select(Option 0:Linear Color Structure=(R=0.5776,G=0.5776,B=0.5776,A=1), Option 1:Linear Color Structure=(R=0,G=0,B=0,A=1), Index=@n20.IsSelected) -> (ReturnValue:Linear Color Structure)
    InColorAndOpacity_ColorUseRule:ESlateColorStylingMode Enum
   ) -> (then)
   n24 = Event PreConstruct()
   > Parent: PreConstruct(IsDesignTime=@n24.IsDesignTime)
   > n25 = Sequence()
   > RegisterBlocks()
   > SetTextStyle()
   > SetBlocksTiming!(
    execute
    self:UIKit Modular Button Object Reference
    Timing:"UIKit Block Timing Structure (by ref)"=Get BlockTiming() -> (BlockTiming:UIKit Block Timing Structure)
   ) -> (then)
   > SetSize!(
    execute
    self:Spacer Object Reference=Get Spacer() -> (Spacer:Spacer Object Reference)
    InSize_X:double=0
    InSize_Y:double=Get MinTileHeight() -> (MinTileHeight:double)
   ) -> (then)
   n26 = Get Pages!(self:WBP Quest Reward Object Reference=Get QuestReward()) -> (Pages:WBP Quest Pages Object Reference)
   > SetStyle!(execute, self:WBP Quest Pages Object Reference=@n26, RegularStyle:bool=false) -> (then)
   > SetBackColor!(execute, self:WBP Quest Pages Object Reference=@n26, Color:double=1) -> (then)
   Branch(execute=@n25.then_1, Condition=@n24.IsDesignTime)
   > Display() {is=title K2Node_CallFunction}
   n27 = Event Construct()
   Event On Entry Released() {is=title K2Node_Event}
   > InstantDeselect()
   Event On Clicked()
   > Parent: On Clicked() {is=title K2Node_CallParentFunction}
   n28 = Get CommonInputSubsystem()
   > Branch(Condition="Not Equal (Enum)"(A=GetCurrentInputType(self=@n28), B=MouseAndKeyboard))
   > Call On Quest Clicked!(
    execute:Exec
    self:WBP Quest Tile Object Reference
    ClickedQuest:WBP Quest Tile Object Reference=Self-Reference()
    WasDropped:bool=Get IsMoveMode()
   ) -> (then:Exec) {is=title K2Node_CallDelegate}
   Event OnFocusLost!() -> (OutputDelegate, then, InFocusEvent:Focus Event Structure) {is=title K2Node_Event}
   > SetTriggeringInputAction!(execute, self:Common Button Base Object Reference, InputActionRow_DataTable:Data Table Object Reference, InputActionRow_RowName:Name=None) -> (then)
   Event Destruct()
   > Call On Quest Tile Destroyed!(execute:Exec, self:WBP Quest Tile Object Reference) -> (then:Exec) {is=title K2Node_CallDelegate}
   > Parent: Destruct()
   n29 = Event OnListItemObjectSet()
   n30 = UnwrapItem!(Item:Object Reference=@n29.ListItemObject) -> (ReturnValue:Object Reference)
   > n31 = Cast To QuestUIQuestVM!(execute, Object=@n30) -> (then, CastFailed, AsQuest UI Quest View Model:Quest UI Quest View Model Object Reference) {is=title K2Node_DynamicCast}
   > Set QuestUIQuestVM(QuestUIQuestVM:Quest UI Quest View Model Object Reference=@n31) -> (Output_Get:Quest UI Quest View Model Object Reference)
   > Set IsPunchCard(IsPunchCard:bool=false) -> (Output_Get:bool)
   n32 = GetSelected()
   n33 = IsHovered()
   > n34 = SetRenderOpacity(self=Get AdditionalText(), InOpacity=Select(Option 0:float=0, Option 1:float=1, Index=OR Boolean(A=@n33, B=@n32)) -> (ReturnValue:float))
   ResetTileStyle(execute=[@n34.then, @n37.then])
   n35 = Cast To MVVM_PunchCardQuest!(execute=@n31.CastFailed, Object=@n30) -> (then, CastFailed, AsMVVM Punch Card Quest:MVVM Punch Card Quest Object Reference) {is=title K2Node_DynamicCast}
   > n36 = Set As MVVM Punch Card Quest(As MVVM Punch Card Quest:MVVM Punch Card Quest Object Reference=@n35) -> (Output_Get:MVVM Punch Card Quest Object Reference)
   > Set QuestUIQuestVM(QuestUIQuestVM=Get QuestData!(self:MVVM Punch Card Quest Object Reference=@n36) -> (QuestData:Quest UI Quest View Model Object Reference))
   > n37 = Set IsPunchCard(IsPunchCard=true)
   n38 = Event OnDragLeave!() -> (OutputDelegate, then, PointerEvent:Pointer Event Structure, Operation:Drag Drop Operation Object Reference) {is=title K2Node_Event}
   > n39 = Cast To QuestDragDrop!(execute, Object=@n38.Operation) -> (then, CastFailed, AsQuest Drag Drop:Quest Drag Drop Object Reference) {is=title K2Node_DynamicCast}
   > n40 = Cast To WBP_QuestTile!(execute, Object=Get DefaultDragVisual!(self:Drag Drop Operation Object Reference=@n39) -> (DefaultDragVisual:Widget Object Reference)) -> (then, CastFailed, AsWBP Quest Tile:WBP Quest Tile Object Reference) {is=title K2Node_DynamicCast}
   > DraggedOverChangedVisualization!(
    execute
    self
    IsEntering:bool=false
    DropPlacement:"EDragDropPlacementType Enum (by ref)"=GetDragDropPlacement!(SourceTile:WBP Quest Tile Object Reference=@n40) -> (Placement:EDragDropPlacementType Enum=NewEnumerator0)
   ) -> (then) {is=title K2Node_CallFunction}
   n41 = Event OnDragEnter!() -> (OutputDelegate, then, MyGeometry:Geometry Structure, PointerEvent:Pointer Event Structure, Operation:Drag Drop Operation Object Reference) {is=title K2Node_Event}
   > n42 = Cast To QuestDragDrop(Object=@n41.Operation)
   > n43 = Cast To WBP_QuestTile(Object=Get DefaultDragVisual(self=@n42))
   > DraggedOverChangedVisualization(IsEntering=true, DropPlacement=GetDragDropPlacement(SourceTile=@n43) -> (Placement=NewEnumerator0))
   Event On Hovered()
   > Parent: On Hovered() {is=title K2Node_CallParentFunction}
   > Call On Quest Hover!(execute:Exec, self:WBP Quest Tile Object Reference, QuestTile:WBP Quest Tile Object Reference=Self-Reference()) -> (then:Exec) {is=title K2Node_CallDelegate}
   n44 = Get IsMoveMode()
   > n45 = Branch(Condition=AND Boolean(A=Get IsDraggable(), B=NOT Boolean(A=@n44)))
   > n46 = PlayAnimationForward(InAnimation=Get HoverOver() -> (HoverOver:Widget Animation Object Reference))
   PlayAnimation(execute=[@n46.then, @n45.else], InAnimation=Get HoverIdle() -> (HoverIdle:Widget Animation Object Reference), NumLoopsToPlay=0)
   > Set IsFocused(IsFocused:bool=true) -> (Output_Get:bool)
   n47 = Get AdditionalText()
   > SetRenderOpacity(self=@n47, InOpacity=1)
   Event On Selected()
   > SelectionChangedVisualization(IsSelected=true) {is=title K2Node_CallFunction}
   n48 = Get AdditionalText()
   > SetRenderOpacity(self=@n48, InOpacity=1)
   Event On Deselected()
   > n49 = Is Valid(InputObject=GetListItemObject!(UserObjectListEntry:User Object List Entry Interface) -> (ReturnValue:Object Reference))
   InstantDeselect(execute=@n49.Is Not Valid)
   SelectionChangedVisualization(execute=@n49.Is Valid, IsSelected=false)
   > SetRenderOpacity(self=@n48, InOpacity=Select(Option 0=0, Option 1=1, Index=IsHovered()))
   Event On Unhovered()
   > Parent: On Unhovered() {is=title K2Node_CallParentFunction}
   n50 = Get IsMoveMode()
   > n51 = Branch(Condition=AND Boolean!(A=Get IsDraggable(), B=NOT Boolean(A=@n50), C:bool=Get IsFocused() -> (IsFocused:bool)) -> (ReturnValue))
   > n52 = PlayAnimationReverse(InAnimation=Get HoverOver())
   Set IsFocused(execute=[@n52.then, @n51.else], IsFocused=false)
   > SetRenderOpacity(self=@n47, InOpacity=Select(Option 0=0, Option 1=1, Index=GetSelected()))
   Event On Double Clicked() {is=title K2Node_Event}
   > Call On Quest Double Clicked!(execute:Exec, self:WBP Quest Tile Object Reference, QuestTile:WBP Quest Tile Object Reference=Self-Reference()) -> (then:Exec) {is=title K2Node_CallDelegate}
  ]}
 GetDragDropPlacement {nodes [
   n1 = GetDragDropPlacement!() -> (then, SourceTile:WBP Quest Tile Object Reference) {is=title K2Node_FunctionEntry}
   n2 = Get OrderedContentVMs!(self:Quest UI Tracker View Model Object Reference=Get TrackerVM!(self:Quest UI Player View Model Object Reference=Get QuestUIPlayerVM() -> (QuestUIPlayerVM:Quest UI Player View Model Object Reference)) -> (TrackerVM:Quest UI Tracker View Model Object Reference)) -> (OrderedContentVMs:Array of Quest UI Base Content View Model Object References)
   n3 = Find Item!(
    TargetArray:Array of Quest UI Base Content View Model Object References=@n2
    ItemToFind:"Quest UI Base Content View Model Object Reference (by ref)"=Get QuestUIQuestVM() -> (QuestUIQuestVM:Quest UI Quest View Model Object Reference)
   ) -> (ReturnValue)
   n4 = Find Item(TargetArray=@n2, ItemToFind=Get QuestUIQuestVM!(self:WBP Quest Tile Object Reference=@n1) -> (QuestUIQuestVM))
   n5 = "integer < integer"(A=@n4, B=@n3)
   > Return Node!(
    execute
    Placement:EDragDropPlacementType Enum=Select(
     Option 0:EDragDropPlacementType Enum=Select(Option 0=NewEnumerator0, Option 1=NewEnumerator1, Index=@n5)
     Option 1:EDragDropPlacementType Enum=NewEnumerator2
     Index="Equal (Integer)"(A=@n4, B=@n3)
    ) -> (ReturnValue:EDragDropPlacementType Enum)
   )
  ]}
 GetIsMnKDraggable {nodes [
   GetIsMnKDraggable() {is=title K2Node_FunctionEntry}
   n1 = GetCurrentInputType(self=Get CommonInputSubsystem() {is=title K2Node_GetSubsystem})
   > Return Node!(execute, ReturnValue:bool=AND Boolean!(A=Get IsDraggable(), B="Equal (Enum)"(A=@n1, B=MouseAndKeyboard)) -> (ReturnValue))
  ]}
 GetQuestText {nodes [
   n1 = GetQuestText!() -> (then, QuestName:Text, StageValue:int, StageMax:int) {is=title K2Node_FunctionEntry}
   n2 = "integer >= integer"(A=@n1.StageValue)
   n3 = AND Boolean(A=@n2, B="integer >= integer"(A=@n1.StageMax))
   > SetVisibility(
    self=Get Stage() -> (Stage:Overlay Object Reference)
    InVisibility=Select(Option 0:ESlateVisibility Enum=Collapsed, Option 1:ESlateVisibility Enum=SelfHitTestInvisible, Index=@n3) -> (ReturnValue:ESlateVisibility Enum)
   )
   > n4 = Branch(Condition=@n3)
   > Set Text(Text=Format Text!(Format="LOCTEXT(\" {0} of {1}\")", 0:int=@n1.StageValue, 1:int=@n1.StageMax) -> (Result), self=Get StageCounterText())
   > Return Node!(execute, QuestText:Text=@n1)
   SetVisibility(execute=@n4.else, self=Get Stage(), InVisibility=Select(Option 0=Collapsed, Option 1=SelfHitTestInvisible, Index=@n2))
   > n5 = Branch(Condition=@n2)
   > Set Text(Text=Format Text!(Format="LOCTEXT(\"Stage {0}\")", 0=@n1.StageValue) -> (Result), self=Get StageCounterText())
   > Return Node(QuestText=@n1)
   Return Node(execute=@n5.else, QuestText=@n1)
  ]}
 HandleDragDropDestroyed {nodes [
   HandleDragDropDestroyed() {is=title K2Node_FunctionEntry}
   > Call On Quest Clicked(WasDropped=true)
  ]}
 InitializeFromQuestTile {nodes [
   n1 = InitializeFromQuestTile!() -> (then, WBP_QuestTile:WBP Quest Tile Object Reference) {is=title K2Node_FunctionEntry}
   > SetQuestUIQuestVM!(execute, self, Viewmodel:Quest UI Quest View Model Object Reference=Get QuestUIQuestVM(self=@n1)) -> (then)
   > SetDragVisualWidth!(execute, self, SourceQuestTile:Widget Object Reference=@n1) -> (then)
   > SetDragVisualWidth(SourceQuestTile=@n1)
   > SetDragVisualStyle()
  ]}
 InstantDeselect {nodes [
   InstantDeselect() {is=title K2Node_FunctionEntry}
   > Set IsPunchCard(IsPunchCard=false)
   > Set IsLastElement(IsLastElement=false)
   > Set As MVVM Punch Card Quest()
   > SetVisibility(self=Get SelectedStateOutline() -> (SelectedStateOutline:Image Object Reference), InVisibility=Collapsed)
   > StopAllAnimations(self:User Widget Object Reference)
   > FlushAnimations(self:User Widget Object Reference)
   > n1 = GetDynamicMaterial(self=Get Background!(self:WBP Quest Reward Object Reference=Get QuestReward()) -> (Background:Image Object Reference))
   > TransitionToDeselectedIdle!(execute, MID:Material Instance Dynamic Object Reference=@n1) -> (then)
   > SetScalarParameterValue(self=@n1, ParameterName=Trans_SelectedIdle, Value=0)
   > n2 = For Each Loop(Array:Array of UIKit Block Object References=Get ModularBlocks() -> (ModularBlocks:Array of UIKit Block Object References)) -> (Array Element:UIKit Block Object Reference)
   SetSize(execute=@n2.Completed, self=Get Spacer(), InSize_X=0, InSize_Y=Get MinTileHeight())
   > SetRenderOpacity(self=[Get HoverOverArea(), Get DragOverArea()], InOpacity=0)
   n3 = Set CurrentBlock(execute=@n2.LoopBody, CurrentBlock:UIKit Block Object Reference=@n2) -> (Output_Get:UIKit Block Object Reference)
   > StopAllAnimations(self=@n3)
   > n4 = Cast To WBP_UIKit_Block_Text!(execute, Object=Get CurrentBlock() -> (CurrentBlock:UIKit Block Object Reference)) -> (then, CastFailed, AsWBP UIKit Block Text:WBP UIKit Block Text Object Reference) {is=title K2Node_DynamicCast}
   n5 = Get FontMID!(self:WBP UIKit Block Text Object Reference=@n4) -> (FontMID:Material Instance Dynamic Object Reference)
   > TransitionToDeselectedIdle(MID=@n5)
   > SetScalarParameterValue(self=@n5, ParameterName=Trans_SelectedIdle, Value=0)
   n6 = Cast To WBP_UIKit_Block_Outline!(execute=@n4.CastFailed, Object=Get CurrentBlock()) -> (then, CastFailed, AsWBP UIKit Block Outline:WBP UIKit Block Outline Object Reference) {is=title K2Node_DynamicCast}
   > RequestInstantTransition!(execute, self:UIKit Block Object Reference=@n6, State:EUIKitBlockInstantTransitionState Enum=InstantUnfocused) -> (then)
  ]}
 IsNew {nodes [
   n1 = IsNew!() -> (then, IsNew:bool=~) {is=title K2Node_FunctionEntry}
   > n2 = Branch(Condition=@n1)
   > SetStatusIndicatorToNew()
   n3 = Get QuestUIQuestVM!() -> (QuestUIQuestVM)
   n4 = Get RequiredCount!(self:Quest UI Quest View Model Object Reference=@n3) -> (RequiredCount:int=0)
   n5 = "integer >= integer"(A=Get AchievedCount!(self:Quest UI Quest View Model Object Reference=@n3) -> (AchievedCount:int=0), B=@n4)
   n6 = "integer > integer"(A=@n4)
   SetVisibility(
    execute=@n2.else
    self=Get StatusIndicator() -> (StatusIndicator:WBP UIKit Status Indicator Object Reference)
    InVisibility=Select(Option 0=Collapsed, Option 1=SelfHitTestInvisible, Index=AND Boolean(A=@n5, B=@n6))
   )
  ]}
 IsTracked {nodes [
   n1 = IsTracked!() -> (then, IsTracked:bool=~) {is=title K2Node_FunctionEntry}
   > SetIsTracked!(execute, self:WBP Tracked Quest Pin Object Reference=Get WBP_TrackedQuestPin(), IsTracked:bool=@n1) -> (then)
  ]}
 MarkSeen {nodes [
   MarkSeen() {is=title K2Node_FunctionEntry}
   > n1 = Is Valid(InputObject=Get QuestUIQuestVM())
   Is Valid(exec=@n1.Is Not Valid, InputObject=Get As MVVM Punch Card Quest())
   > MarkSeen!(execute, self:Quest UI Base View Model Object Reference=Get QuestData(self=Get As MVVM Punch Card Quest()), bCommitToMcp:bool=false) -> (then) {is=title K2Node_CallFunction}
   MarkSeen(execute=@n1.Is Valid, self=Get QuestUIQuestVM(), bCommitToMcp=false)
  ]}
 OnDragDetected {nodes [
   OnDragDetected!() -> (then, MyGeometry:Geometry Structure, PointerEvent:"Pointer Event Structure (by ref)") {is=title K2Node_FunctionEntry}
   > SetIsSelected!(execute, self:Common Button Base Object Reference, InSelected:bool=true, bGiveClickFeedback:bool=true) -> (then)
   > n1 = Create Widget!(
    execute:Exec
    Class:User Widget Class Reference=&/WBP_QuestTile.WBP_QuestTile_C
    OwningPlayer:Player Controller Object Reference=GetOwningPlayer()
    QuestUIQuestVM:Quest UI Quest View Model Object Reference
    IsDraggable:bool=False
    Style:Common Button Style Class Reference=/FortUILibrary/Foundation/Buttons/ButtonStyles/ButtonStyle-Empty.ButtonStyle-Empty_C
    bApplyAlphaOnDisable:bool=True
    bLocked:bool=False
    bSelectable:bool=True
    bShouldSelectUponReceivingFocus:bool=False
    bInteractableWhenSelected:bool=True
    bToggleable:bool=False
    bTriggerClickedAfterSelection:bool=False
    bDisplayInputActionWhenNotInteractable:bool=True
    bHideInputActionWithKeyboard:bool=False
    bShouldUseFallbackDefaultInputAction:bool=True
    bRequiresHold:bool=False
    HoldData:Common UIHold Data Class Reference
    ClickMethod:EButtonClickMethod Enum=DownAndUp
    InputPriority:int=0
    TriggeringInputAction:Data Table Row Handle Structure
   ) -> (then:Exec, ReturnValue:WBP Quest Tile Object Reference)
   > SetQuestUIQuestVM(self=@n1, Viewmodel=Get QuestUIQuestVM())
   > SetDragVisualWidth(self=@n1, SourceQuestTile=Self-Reference())
   > SetDragVisualStyle(self=@n1)
   > Bind Event to On Quest Tile Destroyed(self:WBP Quest Tile Object Reference=@n1, Delegate=Create Event()) {is=title K2Node_AddDelegate}
   > n2 = "Create Drag && Drop Operation"!(
    execute:Exec
    Class:Drag Drop Operation Class Reference=&/QuestDragDrop.QuestDragDrop_C
    Tag:String
    Payload:Object Reference
    DefaultDragVisual:Widget Object Reference=@n1
    Pivot:EDragPivot Enum=CenterLeft
    Offset:Vector 2D Structure=(X=-0.02,Y=0)
    InstigatorSlotIndex:int=0
   ) -> (then:Exec, ReturnValue:Quest Drag Drop Object Reference) {is=title K2Node_CreateDragDropOperation}
   > Return Node!(execute, Operation:Drag Drop Operation Object Reference=@n2)
  ]}
 OnDrop {nodes [
   n1 = OnDrop!() -> (then, MyGeometry:Geometry Structure, PointerEvent:Pointer Event Structure, Operation:Drag Drop Operation Object Reference) {is=title K2Node_FunctionEntry}
   > n2 = Cast To QuestDragDrop(Object=@n1.Operation)
   > n3 = Cast To WBP_QuestTile(Object=Get DefaultDragVisual(self=@n2))
   n4 = Get TrackerVM(self=Get QuestUIPlayerVM())
   n5 = Find Item(TargetArray=Get OrderedContentVMs(self=@n4), ItemToFind=Get QuestUIQuestVM())
   n6 = Get FortGlobalUIContext() -> (ReturnValue:Fort Global UIContext Object Reference) {is=title K2Node_GetSubsystem}
   > SetIsTracked!(
    execute
    self:Quest UI Quest View Model Object Reference=Get QuestUIQuestVM!(self:WBP Quest Tile Object Reference=@n3) -> (QuestUIQuestVM)
    TrackedIndex:int=Select(
     Option 0:int=@n5
     Option 1:int=GetCorrectedTrackedIndex!(self:Quest UI Tracker View Model Object Reference=@n4, QuestVMIndex:int=@n5) -> (ReturnValue:int=0)
     Index=IsInZone!(self:Fort Global UIContext Object Reference=@n6) -> (ReturnValue:bool)
    ) -> (ReturnValue:int)
   ) -> (then)
   > Call On Quest Clicked(ClickedQuest=@n3, WasDropped=true)
   > Return Node!(execute, ReturnValue:bool=true)
   Call On Quest Clicked(execute=[@n3.CastFailed, @n2.CastFailed])
   > Return Node(ReturnValue=false)
  ]}
 OnFocusReceived {nodes [
   OnFocusReceived!() -> (then, MyGeometry:Geometry Structure, InFocusEvent:Focus Event Structure) {is=title K2Node_FunctionEntry}
   > SetTriggeredInputAction!(
    execute
    self:Common Button Base Object Reference
    InputActionRow_DataTable:Data Table Object Reference=&InputActionDataTable
    InputActionRow_RowName:Name=Confirm
   ) -> (then)
   > Return Node(ReturnValue:Event Reply Structure=Make Event Reply!() -> (EventReply:Event Reply Structure) {is=title K2Node_MakeStruct})
  ]}
 OnMouseButtonDown_RootBorder {nodes [
   n1 = OnMouseButtonDown_RootBorder!() -> (then, MyGeometry:Geometry Structure, MouseEvent:"Pointer Event Structure (by ref)") {is=title K2Node_FunctionEntry}
   > Branch(Condition=Get IsDraggable())
   n2 = GetCurrentInputType(self=Get CommonInputSubsystem() {is=title K2Node_GetBlueprintContextEx})
   > n3 = Branch(Condition=OR Boolean(A="Equal (Enum)"(A=@n2, B=Gamepad), B=IsMobileGame()))
   > Return Node(ReturnValue=Unhandled())
   n4 = Unhandled()
   Return Node(
    execute=@n3.else
    ReturnValue=Select(
     Option 0:Event Reply Structure=Handled()
     Option 1:Event Reply Structure=DetectDrag!(Reply:"Event Reply Structure (by ref)"=@n4, DragKey:Key Structure=LeftMouseButton) -> (ReturnValue:Event Reply Structure)
     Index=Is Mouse Button Down!(Input:"Pointer Event Structure (by ref)"=@n1.MouseEvent, MouseButton:Key Structure=LeftMouseButton) -> (ReturnValue:bool)
    ) -> (ReturnValue:Event Reply Structure)
   )
  ]}
 OnQuestClicked {nodes [OnQuestClicked!() -> (then, ClickedQuest:WBP Quest Tile Object Reference, WasDropped:bool=~) {is=title K2Node_FunctionEntry}]}
 OnQuestDoubleClicked {nodes [OnQuestDoubleClicked!() -> (then, QuestTile:WBP Quest Tile Object Reference) {is=title K2Node_FunctionEntry}]}
 OnQuestHover {nodes [OnQuestHover!() -> (then, QuestTile:WBP Quest Tile Object Reference) {is=title K2Node_FunctionEntry}]}
 OnQuestNavigate {nodes [OnQuestNavigate!() -> (then, QuestTile:User Widget Object Reference, Direction:EUINavigation Enum) {is=title K2Node_FunctionEntry}]}
 OnQuestTileDestroyed {nodes [OnQuestTileDestroyed() {is=title K2Node_FunctionEntry}]}
 RefreshMoveModeStyle {nodes [
   RefreshMoveModeStyle() {is=title K2Node_FunctionEntry}
   n1 = AND Boolean(A=Get IsMoveMode(), B=Get IsDraggable())
   > SetRenderOpacity(
    self=[
     Get RightArrow_Img() -> (RightArrow_Img:Image Object Reference)
     Get LeftArrow_Img() -> (LeftArrow_Img:Image Object Reference)
     Get WBP_TrackedQuestPin()
     Get DragDropBar() -> (DragDropBar:Image Object Reference)
    ]
    InOpacity=Select(Option 0:float=1, Option 1:float=4, Index=@n1) -> (ReturnValue:float)
   )
   > SetRenderOpacity(self=Get Root_Border() -> (Root_Border:Border Object Reference), InOpacity=Select(Option 0=1, Option 1=0.3, Index=@n1))
  ]}
 RegisterBlocks {nodes [
   RegisterBlocks() {is=title K2Node_FunctionEntry}
   > RegisterModularBlock!(execute, self:UIKit Modular Button Object Reference, Block:UIKit Block Object Reference=Get Block_Outline()) -> (then)
   > RegisterModularBlock(Block=Get AdditionalText())
   > RegisterModularBlock(Block=Get Block_Text!(self:WBP Quest Reward Object Reference=Get QuestReward()) -> (Block_Text:WBP UIKit Block Text Object Reference))
   n1 = Get Pages_Style1!(self:WBP Quest Pages Object Reference=Get Pages(self=Get QuestReward())) -> (Pages_Style1:WBP UIKit Block Text Object Reference)
   > RegisterModularBlock(Block=@n1)
   > SetFontMaterial!(
    execute
    self:WBP UIKit Block Text Object Reference=@n1
    In Material:Material Instance Object Reference=/QuestUI/QuestScreen/MaterialInstances/MI_QuestNameTile_Text
   ) -> (then)
  ]}
 ResetTileStyle {nodes [
   ResetTileStyle() {is=title K2Node_FunctionEntry}
   > StopAllAnimations()
   > SetRenderOpacity(self=[Self-Reference(), Get Root_Border()], InOpacity=1)
   > SetRenderOpacity(self=[Get DragOverArea(), Get HoverOverArea()], InOpacity=0)
   > SetBlockVisibility(self=Get Block_Outline(), Visible=true)
   > RequestInstantTransitionOnBlocks!(execute, self:UIKit Modular Button Object Reference, State:EUIKitBlockInstantTransitionState Enum=InstantUnfocused) -> (then)
   > Set IsFocused(IsFocused=false)
   > SetTriggeringInputAction(InputActionRow_RowName=None)
   > DisplayPunchcardState() {is=title K2Node_CallFunction}
   > SetColorAndOpacity(self=Get StageCounterText(), InColorAndOpacity_SpecifiedColor=(R=0.5776,G=0.5776,B=0.5776,A=1))
   > n1 = GetDynamicMaterial(self=Get Backplate(self=Get Stage_BackPlate()))
   > SetScalarParameterValue(self=@n1, ParameterName=Backplate Switch, Value=1)
   n2 = Get Pages(self=Get QuestReward())
   > SetStyle(self=@n2, RegularStyle=false)
   > SetBackColor(self=@n2, Color=1)
  ]}
 SelectBlocks {nodes [
   n1 = SelectBlocks!() -> (then, IsSelected:bool=~) {is=title K2Node_FunctionEntry}
   > n2 = For Each Loop(Array=Get ModularBlocks())
   > n3 = Cast To WBP_UIKit_Block_Base!(execute, Object=@n2) -> (then, CastFailed, AsWBP UIKit Block Base:WBP UIKit Block Base Object Reference) {is=title K2Node_DynamicCast}
   > SetIsSelected!(execute, self:WBP UIKit Block Base Object Reference=@n3, IsSelected:bool=@n1) -> (then)
  ]}
 SetAdditionalText {nodes [
   n1 = SetAdditionalText!() -> (then, Text:Text) {is=title K2Node_FunctionEntry}
   n2 = TextIsEmpty(InText=@n1)
   > SetVisibility(
    self=Get AdditionalText()
    InVisibility=Select(Option 0:ESlateVisibility Enum=SelfHitTestInvisible, Option 1:ESlateVisibility Enum=Collapsed, Index=@n2) -> (ReturnValue:ESlateVisibility Enum)
   )
   > SetHeightOverride(
    self=Get SizeBox_QuestData() -> (SizeBox_QuestData:Size Box Object Reference)
    InHeightOverride=Select(Option 0:float=94, Option 1:float=66, Index=@n2) -> (ReturnValue:float)
   )
   > SetText!(execute, self:WBP UIKit Block Text Object Reference=Get AdditionalText(), Text:Text=Get Text!() -> (Text)) -> (then)
   > SetSize(self=Get Spacer(), InSize_X=0, InSize_Y=Get MinTileHeight())
  ]}
 SetContentEnabled {nodes [
   n1 = SetContentEnabled!() -> (then, Enabled:bool=~) {is=title K2Node_FunctionEntry}
   > Set IsQuestEnabled(IsQuestEnabled:bool=@n1) -> (Output_Get:bool)
   > DisplayEnabledState() {is=title K2Node_CallFunction}
  ]}
 SetDebugQuestID {nodes [
   n1 = SetDebugQuestID!() -> (then, QuestID:"String (by ref)") {is=title K2Node_FunctionEntry}
   > "SetText (Text)"(self=Get DebugQuestID() -> (DebugQuestID:Common Text Object Reference), InText="To Text (String)"(InString=@n1))
  ]}
 SetDragDropPlacement {nodes [
   n1 = SetDragDropPlacement!() -> (then, Placement:"EDragDropPlacementType Enum (by ref)"=NewEnumerator0) {is=title K2Node_FunctionEntry}
   > n2 = Cast To GridSlot!(execute, Object=Get Slot!(self:Widget Object Reference=Get DragOverArea()) -> (Slot:Panel Slot Object Reference)) -> (then, CastFailed, AsGrid Slot:Grid Slot Object Reference) {is=title K2Node_DynamicCast}
   > SetVerticalAlignment(
    self:Grid Slot Object Reference=@n2
    InVerticalAlignment=Select!(
     NewEnumerator0:EVerticalAlignment Enum=VAlign_Top
     NewEnumerator1:EVerticalAlignment Enum=VAlign_Bottom
     NewEnumerator2:EVerticalAlignment Enum=VAlign_Center
     Index:EDragDropPlacementType Enum=@n1
    ) -> (ReturnValue:EVerticalAlignment Enum)
   )
   > SetGridSlotNudge!(
    execute
    Widget:Widget Object Reference=Get DragOverArea()
    Nudge_X:double=0
    Nudge_Y:double=Select(
     NewEnumerator0:double=-28
     NewEnumerator1:double=28
     NewEnumerator2:double=0
     Index=Get Placement() -> (Placement:"EDragDropPlacementType Enum (by ref)"=NewEnumerator0)
    ) -> (ReturnValue:double)
   ) -> (then)
   > SetVisibility(
    self=Get DragDropBar()
    InVisibility=Select(
     NewEnumerator0:ESlateVisibility Enum=SelfHitTestInvisible
     NewEnumerator1:ESlateVisibility Enum=SelfHitTestInvisible
     NewEnumerator2:ESlateVisibility Enum=Hidden
     Index=Get Placement() -> (Placement=NewEnumerator0)
    ) -> (ReturnValue:ESlateVisibility Enum)
   )
  ]}
 SetDragVisualStyle {nodes [
   SetDragVisualStyle() {is=title K2Node_FunctionEntry}
   > n1 = Cast To GridSlot(Object=Get Slot(self=Get SelectedStateOutline()))
   > SetLayer!(execute, self:Grid Slot Object Reference=@n1, InLayer:int=2) -> (then)
   > SetBrushFromMaterial(self=Get SelectedStateOutline(), Material=/QuestUI/QuestScreen/QuestTile/Materials/MI_UI_DragVisualOverlay)
   > SetRenderOpacity(self=Get SelectedStateOutline(), InOpacity=1)
   > SetVisibility(self=Get SelectedStateOutline(), InVisibility=SelfHitTestInvisible)
   > SetColorAndOpacity!(execute, self:Image Object Reference=Get Img_QuestBackground(), InColorAndOpacity:Linear Color Structure=(R=0.5029,G=0.5029,B=0.5029,A=1)) -> (then)
   > SetVisibility(self=Get DragOverArea(), InVisibility=Collapsed)
  ]}
 SetDragVisualWidth {nodes [
   n1 = SetDragVisualWidth!() -> (then, SourceQuestTile:Widget Object Reference) {is=title K2Node_FunctionEntry}
   > SetWidthOverride(
    self=Get Root_SizeBox() -> (Root_SizeBox:Size Box Object Reference)
    InWidthOverride=GetLocalSize!(Geometry:"Geometry Structure (by ref)"=GetCachedGeometry!(self:Widget Object Reference=@n1) -> (ReturnValue:"Geometry Structure (by ref)")) -> (ReturnValue_X:double, ReturnValue_Y:double)
   )
   > n2 = Cast To BorderSlot!(execute, Object=Get Slot(self=Get Root_SizeBox())) -> (then, CastFailed, AsBorder Slot:Border Slot Object Reference) {is=title K2Node_DynamicCast}
   > SetHorizontalAlignment(self:Border Slot Object Reference=@n2, InHorizontalAlignment=HAlign_Left)
  ]}
 SetExpirationDateTime {nodes [
   n1 = SetExpirationDateTime!() -> (then, ExpirationDateTime:Date Time Structure) {is=title K2Node_FunctionEntry}
   n2 = BreakDateTime!(InDateTime:Date Time Structure=@n1) -> (Year:int=0, Month:int=0, Day:int=0, Hour:int=0, Minute:int=0, Second:int=0, Millisecond:int=0)
   > SetDateTimeValue!(
    execute
    self:Common Date Time Text Block Object Reference=Get ExpirationTime() -> (ExpirationTime:Athena Challenge Date Time Text Block Object Reference)
    InDateTime:Date Time Structure=MakeDateTime!(Year:int=@n2, Month:int=@n2.Month, Day:int=@n2.Day, Hour:int=@n2.Hour, Minute:int=0, Second:int=0, Millisecond:int=0) -> (ReturnValue:Date Time Structure)
    bShowAsCountdown:bool=true
    InRefreshDelay:float=1
   ) -> (then)
  ]}
 SetHasExpirationTimer {nodes [
   n1 = SetHasExpirationTimer!() -> (then, HasExpiration:bool=~) {is=title K2Node_FunctionEntry}
   > SetVisibility(
    self=Get TextTimerArea() -> (TextTimerArea:Overlay Object Reference)
    InVisibility=Select!(Option 0:ESlateVisibility Enum=Collapsed, Option 1:ESlateVisibility Enum=SelfHitTestInvisible, Index:bool=@n1) -> (ReturnValue)
   )
  ]}
 SetIsDraggable {nodes [
   n1 = SetIsDraggable!() -> (then, IsDraggable:bool=~) {is=title K2Node_FunctionEntry}
   > Set IsDraggable(IsDraggable:bool=@n1) -> (Output_Get:bool)
   > Branch(Condition=Get IsMoveMode())
   > RefreshMoveModeStyle() {is=title K2Node_CallFunction}
  ]}
 SetIsMoveMode {nodes [
   n1 = SetIsMoveMode!() -> (then, IsMoveMode:bool=~, NeedsToAnimate:bool=~) {is=title K2Node_FunctionEntry}
   > Set IsMoveMode(IsMoveMode:bool=@n1) -> (Output_Get:bool)
   > MoveModeChangedVisualization!(execute, self, NeedsToAnimate:bool=@n1.NeedsToAnimate) -> (then) {is=title K2Node_CallFunction}
  ]}
 SetPinIndex {nodes [
   n1 = SetPinIndex!() -> (then, PinIndex:int) {is=title K2Node_FunctionEntry}
   > n2 = Set PinIndex(PinIndex:int=@n1) -> (Output_Get:int=0)
   > SetPinIndex(self=Get WBP_TrackedQuestPin(), Index=@n2) {is=title K2Node_CallFunction}
  ]}
 SetProgressBarValue {nodes [
   n1 = SetProgressBarValue!() -> (then, Progress:double=~) {is=title K2Node_FunctionEntry}
   > n2 = GetDynamicMaterial(self=Get Progress() -> (Progress:Image Object Reference))
   > SetScalarParameterValue(
    self=@n2
    ParameterName=AnimateProgress
    Value=Select(Option 0:float=@n1, Option 1:float=-0.1, Index="float <= float"(A=@n1) -> (ReturnValue=~)) -> (ReturnValue:float)
   )
  ]}
 SetProgressColor {nodes [
   SetProgressColor!() -> (then, Completed:bool=~, QuestEnabled:bool=~) {is=title K2Node_FunctionEntry}
   > n1 = GetDynamicMaterial(self=Get Progress())
   n2 = Select(
    Option 0:Linear Color Structure=Get DISABLED_GRAY() -> (DISABLED_GRAY:Linear Color Structure)
    Option 1:Linear Color Structure=Select(Option 0=(R=0.1255,G=0.9373,B=0.9804,A=1), Option 1=(R=0.0395,G=1,B=0.2307,A=1), Index=Get Completed() -> (Completed:bool))
    Index=Get QuestEnabled() -> (QuestEnabled:bool)
   ) -> (ReturnValue:Linear Color Structure)
   > SetVectorParameterValue(self=@n1, ParameterName=ColorA, Value=@n2)
   > SetVectorParameterValue(self=@n1, ParameterName=ColorB, Value=@n2)
   > SetVectorParameterValue(self=@n1, ParameterName=ColorC, Value=@n2)
  ]}
 SetQuestText {nodes [
   n1 = SetQuestText!() -> (then, QuestText:Text) {is=title K2Node_FunctionEntry}
   > SetText!(execute, self:Rich Text Block Object Reference=Get CommonRichTextHeader(), InText:"Text (by ref)"=@n1) -> (then)
  ]}
 SetSharedQuestType {nodes [
   n1 = SetSharedQuestType!() -> (then, QuestType:ESharedQuestType Enum=Solo) {is=title K2Node_FunctionEntry}
   > SetVisibility(
    self=Get SharingQuest() -> (SharingQuest:Image Object Reference)
    InVisibility=Select!(
     Solo:ESlateVisibility Enum=Collapsed
     ShareableToOthers:ESlateVisibility Enum=SelfHitTestInvisible
     SharedFromOther:ESlateVisibility Enum=SelfHitTestInvisible
     Index:ESharedQuestType Enum=@n1
    ) -> (ReturnValue:ESlateVisibility Enum)
   )
   > SetBrushFromTexture(
    self=Get SharingQuest()
    Texture=Select(
     Solo:Texture 2D Object Reference
     ShareableToOthers:Texture 2D Object Reference=/Game/Athena/UI/Challenges/Art/T_UI_PartyAssist_64
     SharedFromOther:Texture 2D Object Reference=/Game/Athena/UI/Challenges/Art/T_UI_SquadAssist_64
     Index=Get QuestType() -> (QuestType:ESharedQuestType Enum=Solo)
    ) -> (ReturnValue:Texture 2D Object Reference)
    bMatchSize=false
   )
  ]}
 SetStatusIndicatorToCompleted {nodes [
   n1 = SetStatusIndicatorToCompleted!() -> (then, Completed:bool=~) {is=title K2Node_FunctionEntry}
   n2 = Get bIsNew!(self:Quest UI Quest View Model Object Reference=Get QuestUIQuestVM!() -> (QuestUIQuestVM)) -> (bIsNew:bool)
   > SetVisibility(
    self=Get StatusIndicator()
    InVisibility=Select!(Option 0:ESlateVisibility Enum=Collapsed, Option 1:ESlateVisibility Enum=SelfHitTestInvisible, Index:bool=OR Boolean(A=@n1, B=@n2)) -> (ReturnValue:ESlateVisibility Enum)
   )
   > n3 = Branch(Condition=Get Completed())
   n4 = Get StatusIndicator()
   > SetIndicatorType!(execute, self:WBP UIKit Status Indicator Base Object Reference=@n4, IndicatorType:E_UI_StatusIndicator_Type Enum=NewEnumerator0) -> (then)
   > SetIndicatorConfiguration!(
    execute
    self:WBP UIKit Status Indicator Base Object Reference=@n4
    InCount:int=0
    IndicatorConfig:E_UI_StatusIndicator_Configuration Enum=NewEnumerator0
   ) -> (then)
   Branch(execute=@n3.else, Condition=Get bIsNew(self=Get QuestUIQuestVM()))
   > SetStatusIndicatorToNew()
  ]}
 SetStatusIndicatorToNew {nodes [
   SetStatusIndicatorToNew() {is=title K2Node_FunctionEntry}
   > SetVisibility(self=Get StatusIndicator(), InVisibility=SelfHitTestInvisible)
   n1 = Get StatusIndicator()
   > SetIndicatorType(self=@n1, IndicatorType=NewEnumerator2)
   > SetIndicatorConfiguration(self=@n1, InCount=0, IndicatorConfig=NewEnumerator2)
  ]}
 SetTextStyle {nodes [
   SetTextStyle() {is=title K2Node_FunctionEntry}
   > SetFont!(
    execute
    self:WBP UIKit Block Text Object Reference=Get AdditionalText()
    Font:"Slate Font Info Structure (by ref)"=Get AdditionalFont() -> (AdditionalFont:Slate Font Info Structure)
   ) -> (then)
   > SetTextOverflowPolicy!(execute, self:WBP UIKit Block Text Object Reference=Get AdditionalText(), OverflowPolicy:ETextOverflowPolicy Enum=Clip) -> (then)
   > SetWrapping!(execute, self:WBP UIKit Block Text Object Reference=Get AdditionalText(), IsAutoWrap:bool=false, WrapTextAt:int=0) -> (then)
   > SetJustification!(execute, self:WBP UIKit Block Text Object Reference=Get AdditionalText(), Justification:ETextJustify Enum=Left) -> (then)
   > SetTransformPolicy!(execute, self:WBP UIKit Block Text Object Reference=Get AdditionalText(), Transform Policy:ETextTransformPolicy Enum=None) -> (then)
   > SetFont(self=Get Block_Text(self=Get QuestReward()), Font=Get RewardFont() -> (RewardFont:Slate Font Info Structure))
   > SetLineHeightPercentage!(
    execute
    self:Common Text Object Reference=Get TextBlock!(self:WBP UIKit Block Text Object Reference=Get AdditionalText()) -> (TextBlock:Common Text Object Reference)
    InLineHeightPercentage:float=Get AdditionalText_LineHeightPercentage() -> (AdditionalText_LineHeightPercentage:float)
   ) -> (then)
  ]}
 UpdateProgressAmount {nodes [
   UpdateProgressAmount!() -> (then, Amount:int) {is=title K2Node_FunctionEntry}
   > UpdateProgressValues()
   > DisplayProgress() {is=title K2Node_CallFunction}
  ]}
 UpdateProgressValues {nodes [
   UpdateProgressValues() {is=title K2Node_FunctionEntry}
   n1 = Get QuestUIQuestVM()
   n2 = "To Float (Integer)"(InInt=Get RequiredCount(self=@n1) -> (RequiredCount=0))
   n3 = "To Float (Integer)"(InInt=Get AchievedCount(self=@n1) -> (AchievedCount=0))
   > Set ProgressValue(ProgressValue:double="Clamp (Float)"(Value=SafeDivide(A=@n3, B=@n2), Min=0, Max=1)) -> (Output_Get:double)
   > Set IsComplete(IsComplete:bool="Equal (Float)"(A=Get ProgressValue(), B=1)) -> (Output_Get:bool)
   > Set IsZero(IsZero:bool="Equal (Float)"(A=Get ProgressValue())) -> (Output_Get:bool)
   n4 = Get QuestUIQuestVM()
   > Is Valid(InputObject=@n4)
   > n5 = Switch on EQuestProgressStyle!(
    execute
    Selection:EQuestProgressStyle Enum=Get QuestProgressStyle!(self:Quest UI Quest View Model Object Reference=@n4) -> (QuestProgressStyle:EQuestProgressStyle Enum=Default)
   ) -> (Default:Exec, Percentage:Exec) {is=title K2Node_SwitchEnum}
   n6 = Select(
    Option 0:Text=Select(
     Option 0="LOCTEXT(\"<Progress>{Achieved}</>/{Required}\")"
     Option 1="LOCTEXT(\"<Completed>{Achieved}/{Required}</>\")"
     Index=Get IsComplete()
    )
    Option 1:Text="LOCTEXT(\"{Achieved}/{Required}\")"
    Index=Get IsZero() -> (IsZero:bool)
   ) -> (ReturnValue:Text)
   n7 = Get bContentEnabled!(self:Quest UI Base Content View Model Object Reference=Get QuestUIQuestVM()) -> (bContentEnabled:bool)
   n8 = Get QuestUIQuestVM()
   > Set ProgressTextValue(ProgressTextValue:Text=Format Text!(
     Format=Select(Option 0="LOCTEXT(\"<Disabled>{Achieved}/{Required}</>\")", Option 1=@n6, Index=@n7)
     Achieved:int=Get AchievedCount(self=@n8) -> (AchievedCount=0)
     Required:int=Get RequiredCount(self=@n8) -> (RequiredCount=0)
    ) -> (Result)) -> (Output_Get:Text)
   n9 = float * float(A=Get ProgressValue(), B=100) {is=title K2Node_PromotableOperator}
   n10 = Select(Option 0="LOCTEXT(\"<Progress>{Percent}</>%\")", Option 1="LOCTEXT(\"<Completed>{Percent}</>%\")", Index=Get IsComplete())
   n11 = Get IsZero()
   Set ProgressTextValue(
    execute=@n5.Percentage
    ProgressTextValue=Format Text!(Format=Select(Option 0=@n10, Option 1="LOCTEXT(\"{Percent}%\")", Index=@n11), Percent:int=Floor(A=@n9) {is=title K2Node_CallFunction}) -> (Result)
   )
  ]}
]
variables [
 AdditionalFont {
  type=SlateFontInfo
  default {
   bForceMonospaced=false
   bMaterialIsStencil=false
   fontMaterial=/QuestUI/QuestScreen/MaterialInstances/MI_QuestFlavorTile_Text
   fontObject=&HeadingNow
   letterSpacing=0
   monospacedWidth=1
   size=14.25
   skewAmount=0
   typefaceFontName=74 Regular
   outlineSettings {
    bApplyOutlineToDropShadows=false
    bMiteredCorners=false
    bSeparateFillAlpha=false
    outlineMaterial=null
    outlineSize=0
    outlineColor=<rgba 0 0 0 1>
   }
  }
 }
 AdditionalText_LineHeightPercentage {type=number, default=1.15}
 As MVVM Punch Card Quest {type=MVVM_PunchCardQuest_C, default=null}
 BlockTiming {
  type=UIKitBlockTiming
  default {
   deselectingFocused=0.2
   deselectingIdle=0.2
   disabling=0
   enabling=0
   focusing=0.2
   hovering=0.2
   pressing=0.2
   releasing=0.2
   selecting=0.2
   unfocusing=0.2
   unhovering=0.2
  }
 }
 Days {type=integer}
 IsComplete {type=boolean}
 IsDraggable {type=boolean}
 IsFocused {type=boolean}
 IsLastElement {type=boolean}
 IsMoveMode {type=boolean}
 IsPunchCard {type=boolean}
 IsQuestEnabled {type=boolean, default=true}
 IsZero {type=boolean}
 MinTileHeight {type=number}
 OnQuestClicked {}
 OnQuestDoubleClicked {}
 OnQuestHover {}
 OnQuestNavigate {}
 OnQuestTileDestroyed {}
 PinIndex {type=integer}
 ProgressFont {
  type=SlateFontInfo
  default {
   bForceMonospaced=false
   bMaterialIsStencil=false
   fontMaterial=&MI_QuestNameTile_Text
   fontObject=&HeadingNow
   letterSpacing=0
   monospacedWidth=1
   size=14.25
   skewAmount=0
   typefaceFontName=75 Medium
   outlineSettings {
    bApplyOutlineToDropShadows=false
    bMiteredCorners=false
    bSeparateFillAlpha=false
    outlineMaterial=null
    outlineSize=0
    outlineColor=<rgba 0 0 0 1>
   }
  }
 }
 ProgressTextValue {type=string, default=0/0}
 ProgressValue {type=number, default=0.5}
 QuestFont {
  type=SlateFontInfo
  default {
   bForceMonospaced=false
   bMaterialIsStencil=false
   fontMaterial=&MI_QuestNameTile_Text
   fontObject=&HeadingNow
   letterSpacing=0
   monospacedWidth=1
   size=15.75
   skewAmount=0
   typefaceFontName=75 Medium
   outlineSettings {
    bApplyOutlineToDropShadows=false
    bMiteredCorners=false
    bSeparateFillAlpha=false
    outlineMaterial=null
    outlineSize=0
    outlineColor=<rgba 0 0 0 1>
   }
  }
 }
 QuestFont_Selected {
  type=SlateFontInfo
  default {
   bForceMonospaced=false
   bMaterialIsStencil=false
   fontMaterial=&MI_QuestNameTile_Text
   fontObject=&HeadingNow
   letterSpacing=0
   monospacedWidth=1
   size=15.75
   skewAmount=0
   typefaceFontName=75 Medium
   outlineSettings {
    bApplyOutlineToDropShadows=false
    bMiteredCorners=false
    bSeparateFillAlpha=false
    outlineMaterial=null
    outlineSize=0
    outlineColor=<rgba 0 0 0 1>
   }
  }
 }
 QuestText_LineHeightPercentage {type=number, default=1.15}
 RewardFont {
  type=SlateFontInfo
  default {
   bForceMonospaced=false
   bMaterialIsStencil=false
   fontMaterial=/FortUILibrary/UIKit/Buttons/MaterialInstances/SelectableButtons/MI_UIKit_SelectableText_Quiet
   fontObject=&HeadingNow
   letterSpacing=0
   monospacedWidth=1
   size=14.25
   skewAmount=0
   typefaceFontName=75 Medium
   outlineSettings {
    bApplyOutlineToDropShadows=false
    bMiteredCorners=false
    bSeparateFillAlpha=false
    outlineMaterial=null
    outlineSize=0
    outlineColor=<rgba 0 0 0 1>
   }
  }
 }
 Weeks {type=integer}
]
widgets {
 info {
  inheritedWidgetCount=1
  parentClass=&/WBP_QuestTile_Base.WBP_QuestTile_Base_C
  rootWidgetClass=&Border
  widgetCount=44
 }
 widgets [
  Root_Border {
   widgetClassPath=&Border
   bIsVariable=true
   widget=Root_Border
   properties {
    background {
     bIsDynamicallyLoaded=false
     drawAs=NoDrawType
     imageType=NoImage
     mirroring=NoMirror
     resourceName=None
     resourceObject=null
     tiling=NoTile
     imageSize=<32 32>
     margin=<ltrb 0 0 0 0>
     outlineSettings {
      bUseBrushTransparency=false
      roundingType=HalfHeightRadius
      width=0
      color {colorUseRule=UseColor_Specified, specifiedColor=<rgba 0 0 0 0>}
      cornerRadii=<0 0 0 0>
     }
     tintColor {colorUseRule=UseColor_Specified, specifiedColor=<rgba 1 1 1 1>}
     uVRegion {bIsValid=false, max=<0 0>, min=<0 0>}
    }
    padding=<ltrb 0 0 0 0>
   }
  }
  Root_SizeBox {
   widgetClassPath=&SizeBox
   bIsVariable=true
   parent=Root_Border
   widget=Root_SizeBox
   properties {bOverride_MaxDesiredWidth=true, heightOverride=145, maxDesiredWidth=1000, minDesiredWidth=95}
   slot {class=BorderSlot, properties {padding=<ltrb 0 0 0 0>}}
  }
  Root {
   widgetClassPath=&GridPanel
   parent=Root_SizeBox
   widget=Root
   properties {columnFill [0,0,0,1], rowFill [1]}
   slot {class=SizeBoxSlot, properties {verticalAlignment=VAlign_Center}}
  }
  Img_QuestBackground {
   widgetClassPath=&Image
   bIsVariable=true
   parent=Root
   widget=Img_QuestBackground
   properties {brush {
     bIsDynamicallyLoaded=false
     drawAs=Image
     imageType=FullColor
     mirroring=NoMirror
     resourceName=None
     resourceObject=&MI_UI_QuestBackplate
     tiling=NoTile
     imageSize=<32 32>
     margin=<ltrb 0 0 0 0>
     outlineSettings {
      bUseBrushTransparency=false
      roundingType=HalfHeightRadius
      width=0
      color {colorUseRule=UseColor_Specified, specifiedColor=<rgba 0 0 0 0>}
      cornerRadii=<0 0 0 0>
     }
     tintColor {colorUseRule=UseColor_Specified, specifiedColor=<rgba 1 1 1 1>}
     uVRegion {bIsValid=false, max=<0 0>, min=<0 0>}
    }}
   slot {class=GridSlot, properties {column=1, columnSpan=4}}
  }
  Overlay_Reward {
   widgetClassPath=&Overlay
   parent=Root
   widget=Overlay_Reward
   slot {
    class=GridSlot
    properties {column=4, horizontalAlignment=HAlign_Center, layer=11, verticalAlignment=VAlign_Center, padding=<ltrb 20 8 20 15>}
   }
  }
  QuestReward {
   widgetClassPath=&/WBP_QuestReward
   bIsVariable=true
   parent=Overlay_Reward
   widget=QuestReward
   properties {bHideBackgroundOverride=true, questTextInside=false, padding=<ltrb -10 -10 -10 -10>, rewardIconSize=<65 65>, widgetSize=<65 65>}
   slot {class=OverlaySlot, properties {horizontalAlignment=HAlign_Fill, verticalAlignment=VAlign_Fill, padding=<ltrb 0 0 0 -8>}}
  }
  Spacer_26 {
   widgetClassPath=&Spacer
   parent=Overlay_Reward
   widget=Spacer_26
   properties {size=<78 78>}
   slot {class=OverlaySlot, properties {horizontalAlignment=HAlign_Fill, verticalAlignment=VAlign_Fill}}
  }
  VerticalBox_0 {
   widgetClassPath=&VerticalBox
   parent=Root
   widget=VerticalBox_0
   slot {
    class=GridSlot
    properties {column=1, horizontalAlignment=HAlign_Left, layer=5, verticalAlignment=VAlign_Top, nudge=<0 4>, padding=<ltrb 8 8 0 10>}
   }
  }
  WBP_TrackedQuestPin {
   widgetClassPath=/QuestUI/QuestScreen/Map/WBP_TrackedQuestPin
   bIsVariable=true
   parent=VerticalBox_0
   widget=WBP_TrackedQuestPin
   slot {class=VerticalBoxSlot, properties {horizontalAlignment=HAlign_Center, verticalAlignment=VAlign_Center}}
  }
  SharingQuest {
   widgetClassPath=&Image
   bIsVariable=true
   parent=VerticalBox_0
   widget=SharingQuest
   properties {brush {
     bIsDynamicallyLoaded=false
     drawAs=Image
     imageType=FullColor
     mirroring=NoMirror
     resourceName=None
     resourceObject=&T_UI_PartyAssist_64
     tiling=NoTile
     imageSize=<32 32>
     margin=<ltrb 0 0 0 0>
     outlineSettings {
      bUseBrushTransparency=false
      roundingType=HalfHeightRadius
      width=0
      color {colorUseRule=UseColor_Specified, specifiedColor=<rgba 0 0 0 0>}
      cornerRadii=<0 0 0 0>
     }
     tintColor {colorUseRule=UseColor_Foreground, specifiedColor=<rgba 1 1 1 1>}
     uVRegion {bIsValid=false, max=<0 0>, min=<0 0>}
    }}
   slot {class=VerticalBoxSlot, properties {horizontalAlignment=HAlign_Center, verticalAlignment=VAlign_Center}}
  }
  CustomBadge {
   widgetClassPath=&Image
   bIsVariable=true
   parent=VerticalBox_0
   widget=CustomBadge
   properties {
    visibility=Collapsed
    brush {
     bIsDynamicallyLoaded=false
     drawAs=Image
     imageType=FullColor
     mirroring=NoMirror
     resourceName=None
     resourceObject=/FortniteCoreUI/UI/Quests/MaterialInstances/MI_UI_CustomBadge
     tiling=NoTile
     imageSize=<32 32>
     margin=<ltrb 0 0 0 0>
     outlineSettings {
      bUseBrushTransparency=false
      roundingType=HalfHeightRadius
      width=0
      color {colorUseRule=UseColor_Specified, specifiedColor=<rgba 0 0 0 0>}
      cornerRadii=<0 0 0 0>
     }
     tintColor {colorUseRule=UseColor_Foreground, specifiedColor=<rgba 1 1 1 1>}
     uVRegion {bIsValid=false, max=<0 0>, min=<0 0>}
    }
   }
   slot {class=VerticalBoxSlot, properties {horizontalAlignment=HAlign_Center, verticalAlignment=VAlign_Center}}
  }
  Spacer_171 {widgetClassPath=&Spacer, parent=VerticalBox_0, widget=Spacer_171, properties {size=<32 0>}, slot {class=VerticalBoxSlot}}
  CharacterData {
   widgetClassPath=&GridPanel
   bIsVariable=true
   parent=Root
   widget=CharacterData
   properties {visibility=Collapsed, columnFill [1], rowFill [1]}
   slot {class=GridSlot}
  }
  CommonBorder_0 {
   widgetClassPath=&CommonBorder
   parent=CharacterData
   widget=CommonBorder_0
   properties {horizontalAlignment=HAlign_Left, padding=<ltrb 0 0 4 0>}
   slot {class=GridSlot, properties {layer=5}}
  }
  CharacterBackground {
   widgetClassPath=&Image
   bIsVariable=true
   parent=CommonBorder_0
   widget=CharacterBackground
   properties {brush {
     bIsDynamicallyLoaded=false
     drawAs=Image
     imageType=FullColor
     mirroring=NoMirror
     resourceName=None
     resourceObject=/FortniteCoreUI/UI/Quests/MaterialInstances/MI_UI_QuestGiverBackground
     tiling=NoTile
     imageSize=<116 130>
     margin=<ltrb 0 0 0 0>
     outlineSettings {
      bUseBrushTransparency=false
      roundingType=HalfHeightRadius
      width=0
      color {colorUseRule=UseColor_Specified, specifiedColor=<rgba 0 0 0 0>}
      cornerRadii=<0 0 0 0>
     }
     tintColor {colorUseRule=UseColor_Specified, specifiedColor=<rgba 1 1 1 1>}
     uVRegion {bIsValid=false, max=<0 0>, min=<0 0>}
    }}
   slot {class=BorderSlot, properties {horizontalAlignment=HAlign_Left, padding=<ltrb 0 0 4 0>}}
  }
  Description {
   widgetClassPath=&GridPanel
   parent=Root
   widget=Description
   properties {columnFill [1], rowFill [1,0]}
   slot {class=GridSlot, properties {column=3, horizontalAlignment=HAlign_Left, layer=11}}
  }
  SizeBox_QuestData {
   widgetClassPath=&SizeBox
   bIsVariable=true
   parent=Description
   widget=SizeBox_QuestData
   properties {bOverride_HeightOverride=true, heightOverride=94, maxDesiredHeight=94, minDesiredHeight=95}
   slot {class=GridSlot}
  }
  QuestData {
   widgetClassPath=&StackBox
   parent=SizeBox_QuestData
   widget=QuestData
   properties {orientation=Orient_Vertical}
   slot {class=SizeBoxSlot}
  }
  CommonRichTextHeader {
   widgetClassPath=&CommonRichTextBlock
   bIsVariable=true
   parent=QuestData
   widget=CommonRichTextHeader
   properties {
    autoWrapText=true
    bOverrideDefaultStyle=true
    scrollOrientation=Orient_Vertical
    scrollStyle=&TextScrollStyle-Base_C
    text="TextBlock_Rich\r\n33\r\n33\r\n3"
    textStyleSet=/Game/Athena/HUD/QuestIcons/DT_RichText_QuestTile
    decoratorClasses [/Game/Athena/HUD/QuestIcons/BP_QuestTile_IconDecorator.BP_QuestTile_IconDecorator_C]
    defaultTextStyleOverride {
     overflowPolicy=Clip
     transformPolicy=None
     colorAndOpacity {colorUseRule=UseColor_Specified, specifiedColor=<rgba 0.6172 0.6172 0.7682 1>}
     font {
      bForceMonospaced=false
      bMaterialIsStencil=false
      fontMaterial=null
      fontObject=&HeadingNow
      letterSpacing=10
      monospacedWidth=1
      size=14.25
      skewAmount=0
      typefaceFontName=75 Medium
      outlineSettings {
       bApplyOutlineToDropShadows=false
       bMiteredCorners=false
       bSeparateFillAlpha=false
       outlineMaterial=null
       outlineSize=0
       outlineColor=<rgba 1 1 1 1>
      }
     }
     highlightColor {colorUseRule=UseColor_Specified, specifiedColor=<rgba 0 0 0 1>}
     highlightShape {
      bIsDynamicallyLoaded=false
      drawAs=Image
      imageType=NoImage
      mirroring=NoMirror
      resourceName=None
      resourceObject=null
      tiling=NoTile
      imageSize=<32 32>
      margin=<ltrb 0 0 0 0>
      outlineSettings {
       bUseBrushTransparency=false
       roundingType=HalfHeightRadius
       width=0
       color {colorUseRule=UseColor_Specified, specifiedColor=<rgba 0 0 0 0>}
       cornerRadii=<0 0 0 0>
      }
      tintColor {colorUseRule=UseColor_Specified, specifiedColor=<rgba 1 1 1 1>}
      uVRegion {bIsValid=false, max=<0 0>, min=<0 0>}
     }
     selectedBackgroundColor {colorUseRule=UseColor_Foreground, specifiedColor=<rgba 1 0 1 1>}
     shadowColorAndOpacity=<rgba 0 0 0 1>
     shadowOffset=<0 0>
     strikeBrush {
      bIsDynamicallyLoaded=false
      drawAs=Image
      imageType=NoImage
      mirroring=NoMirror
      resourceName=None
      resourceObject=null
      tiling=NoTile
      imageSize=<32 32>
      margin=<ltrb 0 0 0 0>
      outlineSettings {
       bUseBrushTransparency=false
       roundingType=HalfHeightRadius
       width=0
       color {colorUseRule=UseColor_Specified, specifiedColor=<rgba 0 0 0 0>}
       cornerRadii=<0 0 0 0>
      }
      tintColor {colorUseRule=UseColor_Specified, specifiedColor=<rgba 1 1 1 1>}
      uVRegion {bIsValid=false, max=<0 0>, min=<0 0>}
     }
     underlineBrush {
      bIsDynamicallyLoaded=false
      drawAs=Image
      imageType=NoImage
      mirroring=NoMirror
      resourceName=None
      resourceObject=null
      tiling=NoTile
      imageSize=<32 32>
      margin=<ltrb 0 0 0 0>
      outlineSettings {
       bUseBrushTransparency=false
       roundingType=HalfHeightRadius
       width=0
       color {colorUseRule=UseColor_Specified, specifiedColor=<rgba 0 0 0 0>}
       cornerRadii=<0 0 0 0>
      }
      tintColor {colorUseRule=UseColor_Specified, specifiedColor=<rgba 1 1 1 1>}
      uVRegion {bIsValid=false, max=<0 0>, min=<0 0>}
     }
    }
   }
   slot {class=StackBoxSlot, properties {padding=<ltrb 0 14 0 0>, size {sizeRule=Fill, value=1}}}
  }
  AdditionalText {
   widgetClassPath=/FortUILibrary/Systems/ModularButton/Blocks/WBP_UIKit_Block_Text
   bIsVariable=true
   parent=QuestData
   widget=AdditionalText
   properties {
    renderOpacity=0
    transitionTiming {
     deselectingFocused=0.2
     deselectingIdle=0.2
     disabling=0
     enabling=0
     focusing=0.2
     hovering=0.2
     pressing=0.2
     releasing=0.2
     selecting=0.2
     unfocusing=0.2
     unhovering=0.2
    }
   }
   slot {class=StackBoxSlot}
  }
  Horizontal_QuestProgressBar {
   widgetClassPath=&HorizontalBox
   parent=Description
   widget=Horizontal_QuestProgressBar
   properties {bOverride_Cursor=true}
   slot {class=GridSlot, properties {row=1, verticalAlignment=VAlign_Bottom, padding=<ltrb 0 0 0 10>}}
  }
  Progress {
   widgetClassPath=&Image
   bIsVariable=true
   parent=Horizontal_QuestProgressBar
   widget=Progress
   properties {brush {
     bIsDynamicallyLoaded=false
     drawAs=Image
     imageType=FullColor
     mirroring=NoMirror
     resourceName=None
     resourceObject=/FortniteCoreUI/UI/Quests/MaterialInstances/MI_UI_QuestProgressBar
     tiling=NoTile
     imageSize=<1000 10>
     margin=<ltrb 0.5 0.5 0.5 0.5>
     outlineSettings {
      bUseBrushTransparency=false
      roundingType=HalfHeightRadius
      width=0
      color {colorUseRule=UseColor_Specified, specifiedColor=<rgba 0 0 0 0>}
      cornerRadii=<0 0 0 0>
     }
     tintColor {colorUseRule=UseColor_Specified, specifiedColor=<rgba 1 1 1 1>}
     uVRegion {bIsValid=false, max=<0 0>, min=<0 0>}
    }}
   slot {class=HorizontalBoxSlot, properties {verticalAlignment=VAlign_Center, size {sizeRule=Fill, value=1}}}
  }
  Rich_Progress_Text {
   widgetClassPath=/Script/UMG.RichTextBlock
   bIsVariable=true
   parent=Horizontal_QuestProgressBar
   widget=Rich_Progress_Text
   properties {text=180/200, textStyleSet=&/QuestTileProgressTextStyle}
   slot {class=HorizontalBoxSlot, properties {padding=<ltrb 10 0 0 0>}}
  }
  StatusIndicator {
   widgetClassPath=/FortUILibrary/UIKit/StatusIndicator/WBP_UIKit_StatusIndicator
   bIsVariable=true
   parent=Root
   widget=StatusIndicator
   properties {icon=/FortUILibrary/Textures/Icons/T_UI_Icon_UIKit_Check, iconSize=Compact, visibility=Collapsed}
   slot {class=GridSlot, properties {column=4, horizontalAlignment=HAlign_Right, layer=10, verticalAlignment=VAlign_Top, padding=<ltrb 0 8 8 0>}}
  }
  Spacer {
   widgetClassPath=&Spacer
   bIsVariable=true
   parent=Root
   widget=Spacer
   properties {size=<0 0>}
   slot {class=GridSlot, properties {column=1, horizontalAlignment=HAlign_Left, verticalAlignment=VAlign_Top}}
  }
  Stage {
   widgetClassPath=&Overlay
   bIsVariable=true
   parent=Root
   widget=Stage
   slot {class=GridSlot, properties {column=3, horizontalAlignment=HAlign_Left, layer=20, verticalAlignment=VAlign_Top, nudge=<0 -10.6>}}
  }
  Stage_BackPlate {
   widgetClassPath=&WBP_UIKit_Backplate
   bIsVariable=true
   parent=Stage
   widget=Stage_BackPlate
   properties {
    brightness=Medium
    cornerRadius=Circle
    visibility=SelfHitTestInvisible
    foregroundColor {colorUseRule=UseColor_Specified, specifiedColor=<rgba 1 0 1 1>}
   }
   slot {class=OverlaySlot, properties {horizontalAlignment=HAlign_Fill, verticalAlignment=VAlign_Fill}}
  }
  StageCounterText {
   widgetClassPath=&UEFN_TextBlock
   bIsVariable=true
   parent=Stage
   widget=StageCounterText
   properties {
    text=3 of 7
    colorAndOpacity {colorUseRule=UseColor_Specified, specifiedColor=<rgba 0.5776 0.5776 0.5776 1>}
    font {
     bForceMonospaced=false
     bMaterialIsStencil=false
     fontMaterial=null
     fontObject=&HeadingNow
     letterSpacing=0
     monospacedWidth=1
     size=12.75
     skewAmount=0
     typefaceFontName=75 Medium
     outlineSettings {
      bApplyOutlineToDropShadows=false
      bMiteredCorners=false
      bSeparateFillAlpha=false
      outlineMaterial=null
      outlineSize=0
      outlineColor=<rgba 0 0 0 1>
     }
    }
   }
   slot {class=OverlaySlot, properties {horizontalAlignment=HAlign_Center, verticalAlignment=VAlign_Center, padding=<ltrb 8 0 8 0>}}
  }
  PunchCardSeparator {
   widgetClassPath=&Image
   bIsVariable=true
   parent=Root
   widget=PunchCardSeparator
   properties {
    visibility=Collapsed
    brush {
     bIsDynamicallyLoaded=false
     drawAs=Image
     imageType=NoImage
     mirroring=NoMirror
     resourceName=None
     resourceObject=null
     tiling=NoTile
     imageSize=<2 2>
     margin=<ltrb 0 0 0 0>
     outlineSettings {
      bUseBrushTransparency=false
      roundingType=HalfHeightRadius
      width=0
      color {colorUseRule=UseColor_Specified, specifiedColor=<rgba 0 0 0 0>}
      cornerRadii=<0 0 0 0>
     }
     tintColor {colorUseRule=UseColor_Specified, specifiedColor=<rgba 1 1 1 1>}
     uVRegion {bIsValid=false, max=<0 0>, min=<0 0>}
    }
    colorAndOpacity=<rgba 0.1248 0.1441 0.4793 1>
    renderTransformPivot=<0.5 0>
   }
   slot {class=GridSlot, properties {column=1, columnSpan=4, layer=10, verticalAlignment=VAlign_Top, padding=<ltrb 0 -20 0 50>}}
  }
  TextTimerArea {
   widgetClassPath=&Overlay
   bIsVariable=true
   parent=Root
   widget=TextTimerArea
   properties {renderTransform {angle=0, scale=<0.8 0.8>, shear=<0 0>, translation=<0 0>}}
   slot {class=GridSlot, properties {column=3, horizontalAlignment=HAlign_Right, layer=11, verticalAlignment=VAlign_Top, padding=<ltrb 0 -15 -125 0>}}
  }
  WBP_UIKit_Backplate {
   widgetClassPath=&WBP_UIKit_Backplate
   bIsVariable=true
   parent=TextTimerArea
   widget=WBP_UIKit_Backplate
   properties {brightness=Dark, cornerRadius=Circle}
   slot {class=OverlaySlot, properties {horizontalAlignment=HAlign_Fill, verticalAlignment=VAlign_Fill}}
  }
  ExpirationTime {
   widgetClassPath=/Script/FortniteUI.AthenaChallengeDateTimeTextBlock
   bIsVariable=true
   parent=TextTimerArea
   widget=ExpirationTime
   properties {
    bAllowSeconds=false
    completionText=EXPIRED
    labelFormat="Ends: {0}"
    lineHeightPercentage=1.15
    text="Ends: 31 Dec 1, 23:58:45"
    colorAndOpacity {colorUseRule=UseColor_Foreground, specifiedColor=<rgba 1 1 1 1>}
    font {
     bForceMonospaced=false
     bMaterialIsStencil=false
     fontMaterial=null
     fontObject=&HeadingNow
     letterSpacing=10
     monospacedWidth=1
     size=15
     skewAmount=0
     typefaceFontName=75 Medium
     outlineSettings {
      bApplyOutlineToDropShadows=false
      bMiteredCorners=false
      bSeparateFillAlpha=false
      outlineMaterial=null
      outlineSize=0
      outlineColor=<rgba 0 0 0 1>
     }
    }
    shadowOffset=<0 0>
   }
   slot {class=OverlaySlot, properties {padding=<ltrb 10 5 10 5>}}
  }
  Block_Outline {
   widgetClassPath=/FortUILibrary/UIKit/Buttons/Blocks/WBP_UIKit_Block_Outline
   bIsVariable=true
   parent=Root
   widget=Block_Outline
   properties {bottomLeftCornerRadius=24, bottomRightCornerRadius=24, radiusInPixels=true, renderOpacity=0, topLeftCornerRadius=24, topRightCornerRadius=24}
   slot {class=GridSlot, properties {column=1, columnSpan=4, layer=10}}
  }
  SelectedStateOutline {
   widgetClassPath=&Image
   bIsVariable=true
   parent=Root
   widget=SelectedStateOutline
   properties {
    visibility=Collapsed
    brush {
     bIsDynamicallyLoaded=false
     drawAs=Image
     imageType=FullColor
     mirroring=NoMirror
     resourceName=None
     resourceObject=/QuestUI/QuestScreen/QuestTile/Materials/MI_UI_Outline_Selected
     tiling=NoTile
     imageSize=<32 32>
     margin=<ltrb 0 0 0 0>
     outlineSettings {
      bUseBrushTransparency=false
      roundingType=HalfHeightRadius
      width=0
      color {colorUseRule=UseColor_Specified, specifiedColor=<rgba 0 0 0 0>}
      cornerRadii=<0 0 0 0>
     }
     tintColor {colorUseRule=UseColor_Specified, specifiedColor=<rgba 1 1 1 1>}
     uVRegion {bIsValid=false, max=<0 0>, min=<0 0>}
    }
   }
   slot {class=GridSlot, properties {column=1, columnSpan=4, layer=10}}
  }
  DebugQuestID {
   widgetClassPath=&CommonTextBlock
   bIsVariable=true
   parent=Root
   widget=DebugQuestID
   properties {
    bAutoCollapseWithEmptyText=true
    text=Debug Quest ID
    colorAndOpacity {colorUseRule=UseColor_Specified, specifiedColor=<rgba 0.5 0.5 0.5 1>}
    font {
     bForceMonospaced=false
     bMaterialIsStencil=false
     fontMaterial=null
     fontObject=&HeadingNow
     letterSpacing=0
     monospacedWidth=1
     size=10.5
     skewAmount=0
     typefaceFontName=74 Regular
     outlineSettings {
      bApplyOutlineToDropShadows=false
      bMiteredCorners=false
      bSeparateFillAlpha=false
      outlineMaterial=null
      outlineSize=0
      outlineColor=<rgba 0 0 0 1>
     }
    }
    shadowOffset=<0 0>
    strikeBrush {
     bIsDynamicallyLoaded=false
     drawAs=NoDrawType
     imageType=NoImage
     mirroring=NoMirror
     resourceName=None
     resourceObject=null
     tiling=NoTile
     imageSize=<32 32>
     margin=<ltrb 0 0 0 0>
     outlineSettings {
      bUseBrushTransparency=false
      roundingType=HalfHeightRadius
      width=0
      color {colorUseRule=UseColor_Specified, specifiedColor=<rgba 0 0 0 0>}
      cornerRadii=<0 0 0 0>
     }
     tintColor {colorUseRule=UseColor_Specified, specifiedColor=<rgba 1 1 1 1>}
     uVRegion {bIsValid=false, max=<0 0>, min=<0 0>}
    }
   }
   slot {class=GridSlot, properties {column=3, verticalAlignment=VAlign_Bottom, padding=<ltrb 4 5 0 0>}}
  }
  DragOverArea {
   widgetClassPath=&Overlay
   bIsVariable=true
   parent=Root
   widget=DragOverArea
   properties {renderOpacity=0}
   slot {class=GridSlot, properties {columnSpan=5, layer=50, verticalAlignment=VAlign_Bottom, nudge=<0 28>, padding=<ltrb 4 2 4 2>}}
  }
  DragDropBar {
   widgetClassPath=&Image
   bIsVariable=true
   parent=DragOverArea
   widget=DragDropBar
   properties {
    visibility=SelfHitTestInvisible
    brush {
     bIsDynamicallyLoaded=false
     drawAs=RoundedBox
     imageType=NoImage
     mirroring=NoMirror
     resourceName=None
     resourceObject=null
     tiling=NoTile
     imageSize=<4 4>
     margin=<ltrb 0 0 0 0>
     outlineSettings {
      bUseBrushTransparency=false
      roundingType=HalfHeightRadius
      width=0
      color {colorUseRule=UseColor_Specified, specifiedColor=<rgba 0 0 0 0>}
      cornerRadii=<0 0 0 0>
     }
     tintColor {colorUseRule=UseColor_Specified, specifiedColor=<rgba 1 1 1 1>}
     uVRegion {bIsValid=false, max=<0 0>, min=<0 0>}
    }
   }
   slot {class=OverlaySlot, properties {horizontalAlignment=HAlign_Fill, verticalAlignment=VAlign_Center, padding=<ltrb 10 0 10 0>}}
  }
  LeftArrow_Img {
   widgetClassPath=&Image
   bIsVariable=true
   parent=DragOverArea
   widget=LeftArrow_Img
   properties {brush {
     bIsDynamicallyLoaded=false
     drawAs=Image
     imageType=FullColor
     mirroring=NoMirror
     resourceName=None
     resourceObject=/QuestUI/QuestScreen/QuestTile/Materials/MI_UI_Quest_MoveQuestArrow
     tiling=NoTile
     imageSize=<30 30>
     margin=<ltrb 0 0 0 0>
     outlineSettings {
      bUseBrushTransparency=false
      roundingType=HalfHeightRadius
      width=0
      color {colorUseRule=UseColor_Specified, specifiedColor=<rgba 0 0 0 0>}
      cornerRadii=<0 0 0 0>
     }
     tintColor {colorUseRule=UseColor_Specified, specifiedColor=<rgba 1 1 1 1>}
     uVRegion {bIsValid=false, max=<0 0>, min=<0 0>}
    }}
   slot {class=OverlaySlot, properties {verticalAlignment=VAlign_Center, padding=<ltrb -20 0 0 0>}}
  }
  RightArrow_Img {
   widgetClassPath=&Image
   bIsVariable=true
   parent=DragOverArea
   widget=RightArrow_Img
   properties {
    brush {
     bIsDynamicallyLoaded=false
     drawAs=Image
     imageType=FullColor
     mirroring=NoMirror
     resourceName=None
     resourceObject=&MI_UI_Quest_MoveQuestArrow
     tiling=NoTile
     imageSize=<30 30>
     margin=<ltrb 0 0 0 0>
     outlineSettings {
      bUseBrushTransparency=false
      roundingType=HalfHeightRadius
      width=0
      color {colorUseRule=UseColor_Specified, specifiedColor=<rgba 0 0 0 0>}
      cornerRadii=<0 0 0 0>
     }
     tintColor {colorUseRule=UseColor_Specified, specifiedColor=<rgba 1 1 1 1>}
     uVRegion {bIsValid=false, max=<0 0>, min=<0 0>}
    }
    renderTransform {angle=0, scale=<-1 1>, shear=<0 0>, translation=<0 0>}
   }
   slot {class=OverlaySlot, properties {horizontalAlignment=HAlign_Right, padding=<ltrb 0 0 -20 0>}}
  }
  HoverOverArea {
   widgetClassPath=&FortVisualAttachment
   bIsVariable=true
   parent=Root
   widget=HoverOverArea
   properties {renderOpacity=0, contentAnchor=<1 0.5>}
   slot {class=GridSlot, properties {layer=1000, verticalAlignment=VAlign_Center}}
  }
  Overlay_120 {widgetClassPath=&Overlay, parent=HoverOverArea, widget=Overlay_120, slot {class=SizeBoxSlot}}
  Image_129 {
   widgetClassPath=&Image
   bIsVariable=true
   parent=Overlay_120
   widget=Image_129
   properties {
    brush {
     bIsDynamicallyLoaded=false
     drawAs=Image
     imageType=FullColor
     mirroring=NoMirror
     resourceName=None
     resourceObject=/Game/UI/Frontend/MissionSelect/OutpostShieldPowerBar/Icon-NextArrow-L
     tiling=NoTile
     imageSize=<25 25>
     margin=<ltrb 0 0 0 0>
     outlineSettings {
      bUseBrushTransparency=false
      roundingType=HalfHeightRadius
      width=0
      color {colorUseRule=UseColor_Specified, specifiedColor=<rgba 0 0 0 0>}
      cornerRadii=<0 0 0 0>
     }
     tintColor {colorUseRule=UseColor_Specified, specifiedColor=<rgba 1 1 1 1>}
     uVRegion {bIsValid=false, max=<0 0>, min=<0 0>}
    }
    renderTransform {angle=90, scale=<1 1>, shear=<0 0>, translation=<0 0>}
   }
   slot {class=OverlaySlot, properties {padding=<ltrb 0 25 0 0>}}
  }
  Image {
   widgetClassPath=&Image
   bIsVariable=true
   parent=Overlay_120
   widget=Image
   properties {
    brush {
     bIsDynamicallyLoaded=false
     drawAs=Image
     imageType=FullColor
     mirroring=NoMirror
     resourceName=None
     resourceObject=&Icon-NextArrow-L
     tiling=NoTile
     imageSize=<25 25>
     margin=<ltrb 0 0 0 0>
     outlineSettings {
      bUseBrushTransparency=false
      roundingType=HalfHeightRadius
      width=0
      color {colorUseRule=UseColor_Specified, specifiedColor=<rgba 0 0 0 0>}
      cornerRadii=<0 0 0 0>
     }
     tintColor {colorUseRule=UseColor_Specified, specifiedColor=<rgba 1 1 1 1>}
     uVRegion {bIsValid=false, max=<0 0>, min=<0 0>}
    }
    renderTransform {angle=-90, scale=<1 1>, shear=<0 0>, translation=<0 0>}
   }
   slot {class=OverlaySlot}
  }
  InputActionWidget {widgetClassPath=&CommonActionWidget, bInherited=true, widget=null}
 ]
}
